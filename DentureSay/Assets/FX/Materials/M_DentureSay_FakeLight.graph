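# Unlit translucent Niagara sprite material that fakes a point light at the particle position on surfaces behind it.
# Sphere masks over reconstructed world position shape falloff, and particle colour screen-blends with desaturated scene colour.
# A node-built GGX specular with source radius reads GBuffer normal, roughness and metallic; platform switches drop it on some targets.
# Edge and depth fades soften opacity; a Custom HLSL node duplicates the GGX maths without feeding other nodes.
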Material /DentureSay/Assets/FX/Materials/M_DentureSay_FakeLight
graph {
 properties {
  bAutomaticallySetUsageInEditor=false
  bUsedWithNiagaraSprites=true
  blendMode=BLEND_Translucent
  shadingModel=MSM_Unlit
  translucencyPass=MTP_BeforeDOF
  usedShadingModels=Unlit
 }
 nodes [
  n1 = DynamicParameter() {defaultValue=<rgba 300 3.5 1 1>, paramNames [Radius,Exponent,Param3,Param4]}
  n2 = Power(
   Base=SphereMask(
    A=MaterialFunctionCall() {materialFunction=&WorldPositionBehindTranslucency}.Result
    B=ParticlePositionWS()
    Radius=@n1.Radius
   ) {hardnessPercent=0}
   Exp=@n1.Exponent
  )
  n3 = SphereMask(
   A=MaterialFunctionCall() {materialFunction=&WorldPositionBehindTranslucency}.Result
   B=ParticlePositionWS()
   Radius=Multiply(A=DynamicParameter() {defaultValue=<rgba 300 3.5 1 1>, paramNames [Radius,Exponent,Param3,Param4]}.Radius) {constB=1.3333}
  ) {hardnessPercent=0}
  n4 = Multiply(A=@n3, B=@n3)
  n5 = Ceil(None=Add(A=@n4, B=@n2))
  n6 = SceneColor!(UVs)
  n7 = OneMinus(None=@n6.RGB)
  n8 = ParticleColor()
  n9 = MaterialFunctionCall() {materialFunction=&WorldPositionBehindTranslucency}
  MES_SubLW = Subtract(A=ParticlePositionWS(), B=@n9.Result)
  MEL_LenLW = Length(None=@MES_SubLW) {is=class MaterialExpressionLength}
  MEM_Mul2Dist = Multiply(A=@MEL_LenLW) {constB=2}
  MED_DivSR = Divide(A=Constant() {r=10}, B=@MEM_Mul2Dist)
  n10 = SceneTexture() {sceneTextureId=PPI_Roughness}
  MEM_MaxR = Max(A=Clamp(None=ComponentMask(None=@n10.Color) {r=true}, Max=ScalarParameter() {defaultValue=0.5, parameterName=RoughnessMaxV})) {constB=0.045}
  MEM_Alpha = Multiply(A=@MEM_MaxR, B=@MEM_MaxR)
  MEA_AddSrAlpha = Add(A=@MED_DivSR, B=@MEM_Alpha)
  MESat_AlphaP = Saturate(None=@MEA_AddSrAlpha)
  MEM_A2 = Multiply(A=@MESat_AlphaP, B=@MESat_AlphaP)
  MEN_NormN = Normalize(VectorInput=SceneTexture() {sceneTextureId=PPI_WorldNormal}.Color)
  MES_SubVW = Subtract(A=CameraPositionWS(), B=@n9.Result)
  MEN_NormV = Normalize(VectorInput=@MES_SubVW)
  MEN_NormL = Normalize(VectorInput=@MES_SubLW)
  MELI_LerpVL = LinearInterpolate(A=@MEN_NormV, B=@MEN_NormL, Alpha=Constant() {r=0.48})
  MEN_NormH = Normalize(VectorInput=@MELI_LerpVL)
  MEDP_DotNH = DotProduct(A=@MEN_NormN, B=@MEN_NormH)
  MESat_SatNH = Saturate(None=@MEDP_DotNH)
  MEM_NHa2 = Multiply(A=@MESat_SatNH, B=@MEM_A2)
  MES_SubNH = Subtract(A=@MEM_NHa2, B=@MESat_SatNH)
  MEM_MulD3 = Multiply(A=@MES_SubNH, B=@MESat_SatNH)
  MEA_AddOne = Add(A=@MEM_MulD3)
  MEM_MulDD = Multiply(A=@MEA_AddOne, B=@MEA_AddOne)
  MEM_PiDD = Multiply(A=@MEM_MulDD) {constB=3.1416}
  MED_D = Divide(A=@MEM_A2, B=@MEM_PiDD)
  MEDP_DotNL = DotProduct(A=@MEN_NormN, B=@MEN_NormL)
  MESat_NoL = Saturate(None=@MEDP_DotNL)
  MEDP_DotNV = DotProduct(A=@MEN_NormN, B=@MEN_NormV)
  MESat_SatNV = Saturate(None=@MEDP_DotNV)
  MEA_NoV = Add(A=@MESat_SatNV) {constB=0}
  MEM_NV2 = Multiply(A=@MEA_NoV, B=@MEA_NoV)
  MES_OneA2 = Subtract(B=@MEM_A2)
  MEM_NV2OneA2 = Multiply(A=@MEM_NV2, B=@MES_OneA2)
  MEA_AddVa2 = Add(A=@MEM_NV2OneA2, B=@MEM_A2)
  MESQ_SqrtV = SquareRoot(None=@MEA_AddVa2)
  MEM_GGXV = Multiply(A=@MESat_NoL, B=@MESQ_SqrtV)
  MEM_NL2 = Multiply(A=@MESat_NoL, B=@MESat_NoL)
  MEM_NL2OneA2 = Multiply(A=@MEM_NL2, B=@MES_OneA2)
  MEA_AddLa2 = Add(A=@MEM_NL2OneA2, B=@MEM_A2)
  MESQ_SqrtL = SquareRoot(None=@MEA_AddLa2)
  MEM_GGXL = Multiply(A=@MEA_NoV, B=@MESQ_SqrtL)
  MEA_AddGGX = Add(A=@MEM_GGXV, B=@MEM_GGXL)
  MEM_MaxGGX = Max(A=@MEA_AddGGX) {constB=0}
  MED_Vis = Divide(B=@MEM_MaxGGX) {constA=0.5}
  MEM_MulDVis = Multiply(A=@MED_D, B=@MED_Vis)
  MEM_Mul008Spec = Multiply(A=Constant() {r=1}) {constB=0.08}
  MELI_LerpF0 = LinearInterpolate(
   A=@MEM_Mul008Spec
   B=Constant4Vector() {constant=<rgba 0.5 0.5 0.5 1>}.RGB
   Alpha=ComponentMask(None=SceneTexture() {sceneTextureId=PPI_Metallic}.Color) {r=true}
  )
  MES_OneF0 = Subtract(B=@MELI_LerpF0)
  MEDP_DotVH = DotProduct(A=@MEN_NormV, B=@MEN_NormH)
  MESat_VoH = Saturate(None=@MEDP_DotVH)
  MES_OneVH = Subtract(B=@MESat_VoH)
  MEP_PowVH5 = Power(Base=@MES_OneVH) {constExponent=5}
  MEM_OneMF0Pow = Multiply(A=@MES_OneF0, B=@MEP_PowVH5)
  MEA_F = Add(A=@MELI_LerpF0, B=@MEM_OneMF0Pow)
  MEM_MulDVisF = Multiply(A=@MEM_MulDVis, B=@MEA_F)
  MEM_MulFNL = Multiply(A=@MEM_MulDVisF, B=@MESat_NoL)
  MEM_MaxAprEps = Max(A=@MESat_AlphaP) {constB=0}
  MED_DivAlphaAp = Divide(A=@MEM_Alpha, B=@MEM_MaxAprEps)
  MEM_SphereNorm = Multiply(A=@MED_DivAlphaAp, B=@MED_DivAlphaAp)
  MEM_MulFinal = Multiply(A=@MEM_MulFNL, B=@MEM_SphereNorm)
  n11 = Multiply(A=Multiply(A=@n4, B=@MEM_MulFinal), B=Multiply(A=Length(None=@n8.RGB), B=Constant() {r=1}))
  n12 = Desaturation(None=@n6.RGB, Fraction=Constant() {r=1})
  n13 = Add(A=@n11, B=@n12)
  n14 = OneMinus(None=Multiply(
    A=@n8.RGB
    B=MaterialFunctionCall("Xb1,PS4"=@n12, "PC-High&&Epic,XSX,PS5"=@n13, "PC-Perf,Android"=@n12, Switch=@n12, "PC-Low(NotPerf)&&Medium"=@n13) {materialFunction=&PlatformSwitch_1D}.Result
   ))
  Add(
   A=Multiply(A=MaterialFunctionCall!(MaskIn=@n5) {materialFunction=/Game/Effects/Niagara/MasterMaterials/Functions/Common/TranslucentClip}.ToNormalOrEmissive) {constB=0}
   B=OneMinus(None=Multiply(A=@n7, B=@n14))
  )
  MaterialExpressionSceneTexture_1 = SceneTexture() {sceneTextureId=PPI_WorldNormal}
  MaterialExpressionClamp_0 = Clamp(
   None=ComponentMask(None=SceneTexture() {sceneTextureId=PPI_Roughness}.Color) {r=true}
   Max=ScalarParameter() {defaultValue=0.5, parameterName=RoughnessMaxV}
  )
  MaterialExpressionConstant_1 = Constant() {r=1}
  MaterialExpressionConstant4Vector_1 = Constant4Vector() {constant=<rgba 0.5 0.5 0.5 1>}
  MaterialExpressionComponentMask_1 = ComponentMask(None=SceneTexture() {sceneTextureId=PPI_Metallic}.Color) {r=true}
  MaterialExpressionMaterialFunctionCall_1 = MaterialFunctionCall() {materialFunction=&WorldPositionBehindTranslucency}
  MaterialExpressionCameraPositionWS_0 = CameraPositionWS()
  MaterialExpressionParticlePositionWS_3 = ParticlePositionWS()
  MaterialExpressionConstant_9 = Constant() {r=0.48}
  MaterialExpressionConstant_8 = Constant() {r=10}
  Custom!(
   N=@MaterialExpressionSceneTexture_1.Color
   Roughness=@MaterialExpressionClamp_0
   Specular=@MaterialExpressionConstant_1
   BaseColor=@MaterialExpressionConstant4Vector_1.RGB
   Metallic=@MaterialExpressionComponentMask_1
   WorldPos=@MaterialExpressionMaterialFunctionCall_1.Result
   CameraPos=@MaterialExpressionCameraPositionWS_0
   LightPos=@MaterialExpressionParticlePositionWS_3
   HalfBias=@MaterialExpressionConstant_9
   SourceRadius=@MaterialExpressionConstant_8
  ) {
   code="float3 nN = normalize(N);\r\nfloat3 V  = normalize(CameraPos - WorldPos);\r\nfloat3 L  = normalize(LightPos  - WorldPos);\r\nfloat3 H  = normalize(lerp(V, L, HalfBias));\r\n\r\nfloat NoH = saturate(dot(nN, H));\r\nfloat NoV = saturate(dot(nN, V)) + 1e-5;\r\nfloat NoL = saturate(dot(nN, L));\r\nfloat VoH = saturate(dot(V,  H));\r\n\r\n// ---- F0 ----\r\nfloat3 F0 = lerp((0.08 * Specular).xxx, BaseColor, Metallic);\r\n\r\n// ---- Roughness with Source Radius----\r\nfloat r     = max(Roughness, 0.045);\r\nfloat dist  = length(LightPos - WorldPos);\r\nfloat alpha = r * r;\r\nfloat alphaPrime = saturate(SourceRadius / (dist * 2.0) + alpha);\r\nfloat a2    = alphaPrime * alphaPrime;\r\n\r\n// \r\nfloat sphereNorm = alpha / max(alphaPrime, 1e-5);\r\nsphereNorm = sphereNorm * sphereNorm;\r\n\r\n// ---- D: GGX ----\r\nfloat d = (NoH * a2 - NoH) * NoH + 1.0;\r\nfloat D = a2 / (3.14159265 * d * d);\r\n\r\n// ---- V: Smith ----\r\nfloat GGXV = NoL * sqrt(NoV * NoV * (1.0 - a2) + a2);\r\nfloat GGXL = NoV * sqrt(NoL * NoL * (1.0 - a2) + a2);\r\nfloat Vis  = 0.5 / max(GGXV + GGXL, 1e-5);\r\n\r\n// ---- F: Schlick ----\r\nfloat3 F = F0 + (1.0 - F0) * pow(1.0 - VoH, 5.0);\r\n\r\n// ---- Mask ----\r\nreturn D * Vis * F * NoL * sphereNorm;"
   showCode=true
   inputs [
    {
     inputName=N
     input {
      expression=&self:MaterialExpressionSceneTexture_1
      mask=1
      maskA=1
      maskB=1
      maskG=1
      maskR=1
     }
    }
    {inputName=Roughness, input {expression=&self:MaterialExpressionClamp_0}}
    {inputName=Specular, input {expression=&self:MaterialExpressionConstant_1}}
    {
     inputName=BaseColor
     input {
      expression=&self:MaterialExpressionConstant4Vector_1
      mask=1
      maskB=1
      maskG=1
      maskR=1
      outputIndex=5
     }
    }
    {inputName=Metallic, input {expression=&self:MaterialExpressionComponentMask_1}}
    {inputName=WorldPos, input {expression=&self:MaterialExpressionMaterialFunctionCall_1}}
    {inputName=CameraPos, input {expression=&self:MaterialExpressionCameraPositionWS_0}}
    {inputName=LightPos, input {expression=&self:MaterialExpressionParticlePositionWS_3}}
    {inputName=HalfBias, input {expression=&self:MaterialExpressionConstant_9}}
    {inputName=SourceRadius, input {expression=&self:MaterialExpressionConstant_8}}
   ]
  }
  MaterialFunctionCall(Offset Amount=ScalarParameter() {defaultValue=500, parameterName=Offset}, Clamp Padding=Constant() {r=50}) {materialFunction=/Engine/Functions/Engine_MaterialFunctions02/WorldPositionOffset/CameraOffset}
  n15 = TextureCoordinate()
  n16 = OneMinus(None=@n15)
  MaterialExpressionSaturate_3 = Saturate(None=Multiply(A=Min(A=@n16, B=@n15), B=ScalarParameter() {defaultValue=5, parameterName=EdgeHardness}) {constB=10})
  n17 = Convert(None=@MaterialExpressionSaturate_3) {
   nodeName=Break Float2
   convertInputs [{
     type=Vector2
     expressionInput {
      expression=&self:MaterialExpressionSaturate_3
      inputName=None
      mask=0
      maskA=0
      maskB=0
      maskG=0
      maskR=0
      outputIndex=0
     }
    }]
   convertOutputs [{type=Scalar}, {type=Scalar}]
  }
  n18 = Min(A=@n17, B=@n17)
  n19 = DepthFade(FadeDistance=ScalarParameter() {defaultValue=100, parameterName=DF})
  n20 = Constant()
  n21 = Saturate(None=@n11)
  n22 = MaterialFunctionCall("Xb1,PS4"=@n20, "PC-High&&Epic,XSX,PS5"=@n21, "PC-Perf,Android"=@n20, Switch=@n20, "PC-Low(NotPerf)&&Medium"=@n21) {materialFunction=&PlatformSwitch_1D}
  Multiply(A=Multiply(A=@n18, B=@n18), B=Multiply(A=Multiply(A=@n19, B=@n19), B=Max(A=@n22.Result, B=@n2)))
 ]
}
