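# Translucent, unlit, two-sided Niagara material for mesh and sprite wind gust effects.
# Two panning noise textures, with flippable spin direction and dynamic-parameter control, are combined and sharpened by a noise-distorted sphere mask.
# That shape drives opacity with vertex-color, Fresnel edge and depth fades, plus optional gradient emissive glow with hue shift.
# On the deferred, non-low-quality path, particle color gets fake lighting from a panned normal map and sky light direction.
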Material /DentureSay/Assets/FX/Materials/M_DentureSay_Gust
graph {
 properties {
  bAutomaticallySetUsageInEditor=false
  bUsedWithNiagaraMeshParticles=true
  bUsedWithNiagaraSprites=true
  blendMode=BLEND_Translucent
  pixelDepthOffsetMode=PDOM_Legacy
  shadingModel=MSM_Unlit
  twoSided=true
  usedShadingModels=Unlit
 }
 nodes [
  n1 = DynamicParameter() {parameterIndex=1, defaultValue=<rgba 1 1 0 0>, paramNames [Color Index, HueShift, "", ""]}
  n2 = VectorParameter() {parameterName=TileAndPanSpeed1, defaultValue=<rgba 4 2.5 -1.7 2.5>}
  n3 = AppendVector(A=@n2.R, B=@n2.R)
  n4 = TextureCoordinate()
  n5 = MaterialFunctionCall(Float2=@n4) {materialFunction=&BreakOutFloat2Components}
  n6 = LinearInterpolate(A=AppendVector(A=OneMinus(None=@n5.R), B=@n5.G), B=@n4, Alpha=ScalarParameter() {parameterName=SpinDirection})
  n7 = DynamicParameter() {defaultValue=<rgba 1 1 0 0>, paramNames [PanAdd, PanSpeed, PanAnim, SphereMaskCenterOffset]}
  n8 = Time() {period=333}
  n9 = Add(A=Multiply(A=@n7.PanSpeed, B=Add(A=@n7.PanAdd, B=@n8)), B=ParticleRandom())
  n10 = AppendVector(A=Constant(), B=DynamicParameter() {defaultValue=<rgba 1 1 0 0>, paramNames [PanAdd, PanSpeed, PanAnim, SphereMaskCenterOffset]}.PanAnim)
  n11 = Add(A=Panner(Coordinate=Multiply(A=@n3, B=@n6) {constB=0.5}, Time=@n9, Speed=AppendVector(A=@n2.B, B=@n2.B)) {speedX=0.05, speedY=0.3}, B=@n10)
  n12 = TextureSampleParameter2D(UVs=@n11) {parameterName=Texture1, samplerType=SAMPLERTYPE_LinearGrayscale, texture=/Game/Effects/Fort_Effects/Textures/Tile/T_Diagonal_Blur_Noise_M}
  n13 = ScalarParameter() {defaultValue=0.75, parameterName=Texture1Mult}
  n14 = VectorParameter() {parameterName=TileAndPanSpeed2, defaultValue=<rgba 2 1 -0.8 1>}
  n15 = Multiply(A=@n6, B=AppendVector(A=@n14.R, B=@n14.R))
  n16 = AppendVector(A=@n14.B, B=@n14.B)
  n17 = TextureSampleParameter2D(UVs=Add(A=Panner(Coordinate=@n15, Time=@n9, Speed=@n16) {speedX=0.05, speedY=0.3}, B=@n10)) {parameterName=Texture2, samplerType=SAMPLERTYPE_LinearGrayscale, texture=/Game/Effects/Fort_Effects/Textures/Tile/T_Caustic_Wind_Diagonal}
  n18 = Add(
   A=Multiply(A=Subtract(A=@n17.G) {constB=0.5}, B=ScalarParameter() {defaultValue=0.25, parameterName=SphereMaxDistort}) {constB=0.2}
   B=TextureCoordinate()
  )
  n19 = SphereMask(
   A=Power(Base=ComponentMask(None=@n18) {g=true}, Exp=ScalarParameter() {defaultValue=1.1, parameterName=SphereMaskShift})
   B=Add(
    A=ScalarParameter() {defaultValue=0.5, group=SphereMask, parameterName=SphereMaskCenter}
    B=DynamicParameter() {defaultValue=<rgba 1 1 0 0>, paramNames [PanAdd, PanSpeed, PanAnim, SphereMaskCenterOffset]}.SphereMaskCenterOffset
   )
   Radius=ScalarParameter() {defaultValue=0.42, group=SphereMask, parameterName=SphereMaskSize}
   Hardness=ScalarParameter() {group=SphereMask, parameterName=SphereMaskHardness}
  ) {attenuationRadius=0.42, hardnessPercent=0}
  n20 = ScalarParameter() {defaultValue=1, parameterName=SphereMaskPower}
  n21 = Power(
   Base=Max(A=Multiply(A=@n12.G, B=@n13) {constB=0.75}, B=@n17.G)
   Exp=LinearInterpolate(
    A=ScalarParameter() {defaultValue=7, parameterName=PowerMax}
    B=ScalarParameter() {defaultValue=0.7, parameterName=PowerMin}
    Alpha=Power(Base=@n19, Exp=@n20) {constExponent=1}
   ) {constA=7, constB=0.7}
  ) {constExponent=6}
  n22 = ComponentMask(None=TextureCoordinate()) {g=true}
  n23 = ScalarParameter() {defaultValue=5, parameterName=EmissiveGradientPower}
  n24 = Multiply(A=Multiply(A=@n21, B=Power(Base=@n22, Exp=@n23) {constExponent=3}) {constA=1}, B=VertexColor().R)
  n25 = LinearInterpolate(
   A=VectorParameter() {parameterName=EmissiveGlowColorDark, defaultValue=<rgba 1 0.137 0 1>}.RGB
   B=VectorParameter() {parameterName=EmissiveGlowColor, defaultValue=<rgba 2.5 1.2857 0.0612 1>}.RGB
   Alpha=@n24
  )
  n26 = Multiply(
   A=StaticSwitchParameter(
    True=MaterialFunctionCall(Hue Shift Percentage=@n1.HueShift, Texture=@n25) {materialFunction=&HueShift}.Result
    False=@n25
   ) {parameterName=HueShift}
   B=@n24
  )
  n27 = QualitySwitch(
   Default=MaterialFunctionCall(LDR Input=Clamp(None=@n26), HDR Tint=@n26, HDR Intensity Multiplier=ScalarParameter() {defaultValue=1, parameterName=HDRMult}) {materialFunction=&DeriveHDRfromLDR}.RemappedRange
   Low=@n26
  )
  n28 = TextureSampleParameter2D(UVs=@n11) {parameterName=Normal, samplerType=SAMPLERTYPE_Normal, texture=/Game/Effects/Fort_Effects/Textures/Tile/T_Diagonal_Blur_Noise_N}
  n29 = Saturate(None=DotProduct(A=Transform(None=@n28.RGB), B=SkyAtmosphereLightDirection()))
  n30 = Multiply(A=LinearInterpolate(A=VectorParameter() {parameterName=ShadowColor, defaultValue=<rgba 0.4036 0.4277 0.5 1>}.RGB, Alpha=Multiply(A=@n29, B=@n29)) {constB=1.1})
  n31 = If(
   A=ScalarParameter() {defaultValue=1, group=TimeOfDay, parameterName=Use Juno TOD Color}
   "A > B"=@n30
   "A == B"=@n30
   "A < B"=Constant() {r=1}
  ) {constB=0.99}
  n32 = ParticleColor()
  n33 = ShadingPathSwitch(Default=@n32.RGB, Deferred=QualitySwitch(Default=Multiply(A=@n31, B=@n32.RGB), Low=@n32.RGB))
  n34 = StaticSwitchParameter(True=Add(A=ShadingPathSwitch(Default=@n26, Deferred=@n27), B=@n33), False=@n33) {parameterName=EmissiveGlow}
  MaterialFunctionCall(Emissive=StaticSwitchParameter(True=@n34, False=@n34) {parameterName=UseBrickColor}) {materialFunction=&MF_FX_EmissiveAdjustments}
  n35 = Fresnel(ExponentIn=ScalarParameter() {defaultValue=2, parameterName=EdgeFadeExp}, Normal=Multiply(A=VertexNormalWS(), B=TwoSidedSign())) {baseReflectFraction=0}
  n36 = Multiply(
   A=@n19
   B=MaterialFunctionCall(In=OneMinus(None=@n35), Contrast=Constant() {r=0.2}) {materialFunction=&CheapContrast}.Result
  )
  n37 = VertexColor()
  n38 = Power(Base=Multiply(A=Multiply(A=@n32.A, B=@n21), B=Multiply(A=@n36, B=@n37.R)), Exp=ScalarParameter() {defaultValue=1.6, parameterName=Power}) {constExponent=0.5}
  QualitySwitch(Default=DepthFade(Opacity=@n38, FadeDistance=ScalarParameter() {defaultValue=40, parameterName=DepthFade}), Low=@n38)
 ]
}
