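# Unlit translucent Niagara ribbon material for the DentureSay lightning effect.
# Particle color RGB feeds an emissive-adjustment function, while particle alpha is multiplied by three rerouted masks.
# Named masks are a cosine edge fade across V, a sharpened U gradient offset by the EnergyProgress dynamic parameter, and stretched Gaussian noise panned by speed parameters.
# EnergyProgress lets the particle system drive the visible energy along the ribbon.
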Material /DentureSay/Assets/FX/Materials/M_DentureSay_Lightning
graph {
 properties {
  bAutomaticallySetUsageInEditor=false
  bUsedWithNiagaraRibbons=true
  bUsedWithStaticMesh=false
  blendMode=BLEND_Translucent
  shadingModel=MSM_Unlit
  usedShadingModels=Unlit
 }
 nodes [
  n1 = ParticleColor()
  MaterialFunctionCall(Emissive=@n1.RGB) {materialFunction=&MF_FX_EmissiveAdjustments}
  n2 = NamedRerouteUsage()
  n3 = NamedRerouteUsage()
  Multiply(A=@n1.A, B=Multiply(A=NamedRerouteUsage(), B=Multiply(A=@n2, B=@n3)))
  n4 = ComponentMask(None=TextureCoordinate()) {g=true}
  NamedRerouteDeclaration(None=OneMinus(None=Cosine(None=@n4))) {name=EdgeMask, nodeColor=<rgba 1 0.3281 0 1>}
  n5 = TextureCoordinate()
  n6 = Subtract(A=OneMinus(None=ComponentMask(None=@n5) {r=true}), B=DynamicParameter() {paramNames [EnergyProgress, Param2, Param3, Param4]}.EnergyProgress)
  NamedRerouteDeclaration(None=Saturate(None=Power(Base=@n6) {constExponent=3})) {name=EnergyProgress, nodeColor=<rgba 0.0078 0 1 1>}
  n7 = ScalarParameter() {defaultValue=5, parameterName=Pan Speed X}
  n8 = ScalarParameter() {defaultValue=1, parameterName=Pan Speed Y}
  n9 = TextureSample(UVs=Panner(Coordinate=TextureCoordinate() {uTiling=4, vTiling=0.1}, Speed=AppendVector(A=@n7, B=@n8))) {texture=/Game/Effects/Fort_Effects/Textures/Tile/T_GaussianNoise_01_M}
  NamedRerouteDeclaration(None=Saturate(None=Power(Base=@n9.G))) {name=Noise, nodeColor=<rgba 0.9922 0 1 1>}
 ]
}
