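# Translucent, unlit, two-sided material for Niagara mesh particles, by name the dark slash effect of DentureSay.
# SplineThicken widens the mesh by a Width parameter scaled by dynamic Param2, and a CameraOffset term nudges the result along the view direction toward the camera.
# Particle color feeds emissive through MF_FX_EmissiveAdjustments.
# Two UV-based TipTail edge masks multiply with particle alpha to shape the fade.
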Material /DentureSay/Assets/FX/Materials/M_DentureSay_SlashDark
graph {
 properties {
  bAutomaticallySetUsageInEditor=false
  bUsedWithNiagaraMeshParticles=true
  blendMode=BLEND_Translucent
  shadingModel=MSM_Unlit
  twoSided=true
  usedShadingModels=Unlit
 }
 nodes [
  n1 = Multiply(A=DynamicParameter() {paramNames [Dissolve,Param2,Param3,Param4]}.Param2, B=ScalarParameter() {defaultValue=135.0596, parameterName=Width})
  n2 = MaterialFunctionCall(WidthBase=@n1, WidthTip=@n1, Expand U or V UV Channel=StaticBool()) {materialFunction=&SplineThicken}
  n3 = Subtract(A=Add(A=WorldPosition().XYZ, B=@n2.WorldPositionOffset), B=CameraPositionWS())
  Add(A=@n2.WorldPositionOffset, B=Multiply(A=Normalize(VectorInput=@n3), B=ScalarParameter() {defaultValue=-25, parameterName=CameraOffset}))
  n4 = ParticleColor()
  MaterialFunctionCall(Emissive=Divide(A=@n4.RGB)) {materialFunction=&MF_FX_EmissiveAdjustments}
  MaterialExpressionTextureCoordinate_0 = TextureCoordinate()
  n5 = Convert(None=@MaterialExpressionTextureCoordinate_0) {
   nodeName=Break Float2
   convertInputs [{
     type=Vector2
     expressionInput {
      expression=&self:MaterialExpressionTextureCoordinate_0
      inputName=None
      mask=0
      maskA=0
      maskB=0
      maskG=0
      maskR=0
      outputIndex=0
     }
    }]
   convertOutputs [{type=Scalar}, {type=Scalar}]
  }
  n6 = VectorParameter() {group=TipTail, parameterName=TipTail_EdgeMask_InvertControl, channelNames {a="", b=Power, g=Multiply, r=Offset}, defaultValue=<rgba 0 4 1 1>}
  n7 = VectorParameter() {group=TipTail, parameterName=TipTail_EdgeMask_Control, channelNames {a="", b=Power, g=Multiply, r=Offset}, defaultValue=<rgba -0.0768 5 1 1>}
  n8 = VectorParameter() {
   group=TipTail
   parameterName=TipTailV_EdgeMask_Control
   channelNames {a="", b=Power, g=Multiply, r=Offset}
   defaultValue=<rgba 0 1.5 0.7622 1>
  }
  Multiply(
   A=MaterialFunctionCall!(UorV=@n5, Invert_EdgeControl=@n6.RGB, NoInvert_EdgeControl=@n7.RGB) {materialFunction=/Game/Effects/Niagara/MasterMaterials/Functions/Common/FXCommon_EdgeMask_Ex}.Mask
   B=Multiply(
    A=MaterialFunctionCall!(UorV=@n5, EdgeControl=@n8.RGB) {materialFunction=/Game/Effects/Niagara/MasterMaterials/Functions/Common/FXCommon_EdgeMask_Simple}.Mask
    B=@n4.A
   )
  )
 ]
}
