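# Additive, unlit, two-sided material for Niagara mesh particles, drawing a dissolving, glowing slash ring.
# SplineThicken expands the mesh by a particle-driven width, with an offset along the camera vector.
# Perlin noise, edge masks and the particle Dissolve value reshape centre-scaled UVs, whose sphere masks index the CLA_DentureSay curve atlas for tinted core and glow.
# A separate noise dissolve is multiplied by edge mask and sharpened near-camera fade.
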Material /DentureSay/Assets/FX/Materials/M_DentureSay_SlashRing
graph {
 expressions {
  MaterialExpressionAdd_0 {
   class=MaterialExpressionAdd
   inputs [
    A {from=MaterialExpressionMaterialFunctionCall_5, output=R}
    B {from=MaterialExpressionAdd_1}
   ]
  }
  MaterialExpressionAdd_1 {
   class=MaterialExpressionAdd
   inputs [
    A {from=MaterialExpressionLinearInterpolate_2}
    B {from=MaterialExpressionScalarParameter_2}
   ]
  }
  MaterialExpressionAdd_2 {
   class=MaterialExpressionAdd
   inputs [
    A {from=MaterialExpressionMaterialFunctionCall_5, output=R}
    B {from=MaterialExpressionLinearInterpolate_2}
   ]
  }
  MaterialExpressionAdd_3 {
   class=MaterialExpressionAdd
   inputs [
    A {from=MaterialExpressionMultiply_8}
    B {from=MaterialExpressionMultiply_11}
   ]
  }
  MaterialExpressionAdd_4 {
   class=MaterialExpressionAdd
   inputs [
    A {from=MaterialExpressionMaterialFunctionCall_5, output=R}
    B {from=MaterialExpressionAdd_5}
   ]
  }
  MaterialExpressionAdd_5 {
   class=MaterialExpressionAdd
   inputs [
    A {from=MaterialExpressionLinearInterpolate_2}
    B {from=MaterialExpressionScalarParameter_5}
   ]
  }
  MaterialExpressionAdd_6 {
   class=MaterialExpressionAdd
   inputs [
    A {from=MaterialExpressionLinearInterpolate_2}
    B {from=""}
   ]
  }
  MaterialExpressionAdd_7 {
   class=MaterialExpressionAdd
   inputs [
    A {from=MaterialExpressionWorldPosition_0, output=XYZ}
    B {from=MaterialExpressionMaterialFunctionCall_0, output=WorldPositionOffset}
   ]
  }
  MaterialExpressionAdd_8 {
   class=MaterialExpressionAdd
   inputs [
    A {from=MaterialExpressionMaterialFunctionCall_0, output=WorldPositionOffset}
    B {from=MaterialExpressionMultiply_2}
   ]
  }
  MaterialExpressionCameraPositionWS_0 {class=MaterialExpressionCameraPositionWS, inputs [], properties {}}
  MaterialExpressionConstant_0 {class=MaterialExpressionConstant}
  MaterialExpressionConstant_2 {class=MaterialExpressionConstant}
  MaterialExpressionConstant_3 {class=MaterialExpressionConstant, properties {r=2}}
  MaterialExpressionConstant_4 {class=MaterialExpressionConstant, properties {r=0.5}}
  MaterialExpressionConstant_5 {class=MaterialExpressionConstant, properties {r=1}}
  MaterialExpressionConstant_7 {class=MaterialExpressionConstant, properties {r=0.5}}
  MaterialExpressionConstant_8 {class=MaterialExpressionConstant, properties {r=1}}
  MaterialExpressionConvert_1 {
   class=MaterialExpressionConvert
   inputs [
    None {from=""}
    None {from=MaterialExpressionMultiply_3}
   ]
   properties {
    nodeName=Make Vector2
    convertInputs [
     {
      type=Scalar
      defaultValue=<rgba 1 0 0 1>
      expressionInput {expression=null, inputName=None, mask=0, maskA=0, maskB=0, maskG=0, maskR=0, outputIndex=0}
     }
     {
      type=Scalar
      expressionInput {
       expression=&self:MaterialExpressionMultiply_3
       inputName=None
       mask=0
       maskA=0
       maskB=0
       maskG=0
       maskR=0
       outputIndex=0
      }
     }
    ]
    convertOutputs [{type=Vector2}]
   }
  }
  MaterialExpressionConvert_11 {
   class=MaterialExpressionConvert
   inputs [None {from=MaterialExpressionMaterialFunctionCall_3, output=UVs}]
   properties {
    nodeName=Break Float2
    convertInputs [{
      type=Vector2
      expressionInput {
       expression=&self:MaterialExpressionMaterialFunctionCall_3
       inputName=None
       mask=0
       maskA=0
       maskB=0
       maskG=0
       maskR=0
       outputIndex=0
      }
     }]
    convertOutputs [{type=Scalar}, {type=Scalar}]
   }
  }
  MaterialExpressionConvert_2 {
   class=MaterialExpressionConvert
   inputs [None {from=MaterialExpressionMaterialFunctionCall_6, output=UVs}]
   properties {
    nodeName=Break Float2
    convertInputs [{
      type=Vector2
      expressionInput {
       expression=&self:MaterialExpressionMaterialFunctionCall_6
       inputName=None
       mask=0
       maskA=0
       maskB=0
       maskG=0
       maskR=0
       outputIndex=0
      }
     }]
    convertOutputs [{type=Scalar}, {type=Scalar}]
   }
  }
  MaterialExpressionConvert_3 {
   class=MaterialExpressionConvert
   inputs [None {from=MaterialExpressionTextureCoordinate_0}]
   properties {
    nodeName=Break Float2
    convertInputs [{
      type=Vector2
      expressionInput {
       expression=&self:MaterialExpressionTextureCoordinate_0
       inputName=None
       mask=0
       maskA=0
       maskB=0
       maskG=0
       maskR=0
       outputIndex=0
      }
     }]
    convertOutputs [{type=Scalar}, {type=Scalar}]
   }
  }
  MaterialExpressionConvert_9 {
   class=MaterialExpressionConvert
   inputs [
    None {from=""}
    None {from=MaterialExpressionMultiply_1}
   ]
   properties {
    nodeName=Make Vector2
    convertInputs [
     {
      type=Scalar
      defaultValue=<rgba 1 0 0 1>
      expressionInput {expression=null, inputName=None, mask=0, maskA=0, maskB=0, maskG=0, maskR=0, outputIndex=0}
     }
     {
      type=Scalar
      expressionInput {
       expression=&self:MaterialExpressionMultiply_1
       inputName=None
       mask=0
       maskA=0
       maskB=0
       maskG=0
       maskR=0
       outputIndex=0
      }
     }
    ]
    convertOutputs [{type=Vector2}]
   }
  }
  MaterialExpressionDivide_0 {
   class=MaterialExpressionDivide
   inputs [
    A {from=MaterialExpressionMultiply_0}
    B {from=""}
   ]
  }
  MaterialExpressionDynamicParameter_0 {class=MaterialExpressionDynamicParameter, inputs [], properties {paramNames [Dissolve,Param2,Param3,Param4]}}
  MaterialExpressionLinearInterpolate_1 {
   class=MaterialExpressionLinearInterpolate
   inputs [
    A {from=MaterialExpressionConstant_8}
    B {from=MaterialExpressionScalarParameter_10}
    Alpha {from=MaterialExpressionPower_2}
   ]
  }
  MaterialExpressionLinearInterpolate_2 {
   class=MaterialExpressionLinearInterpolate
   inputs [
    A {from=""}
    B {from=MaterialExpressionDynamicParameter_0, output=Dissolve}
    Alpha {from=MaterialExpressionMaterialFunctionCall_10, output=Mask}
   ]
   properties {constA=-1}
  }
  MaterialExpressionMaterialFunctionCall_0 {
   class=MaterialExpressionMaterialFunctionCall
   inputs [
    WidthBase {from=MaterialExpressionMultiply_9}
    WidthTip {from=MaterialExpressionMultiply_9}
    WorldPosition {from=""}
    UVs for Projection {from=""}
    UVs for Thickness {from=""}
    UVs for texturing {from=""}
    Expand U or V UV Channel {from=MaterialExpressionStaticBool_0}
    DeriveNormalZ {from=""}
    AngleCorrectedNormals {from=""}
    AdditionalNormal {from=""}
    FlattenPixelNormal {from=""}
    UVs For NormalShading {from=""}
   ]
   properties {materialFunction=&SplineThicken}
  }
  MaterialExpressionMaterialFunctionCall_1 {
   class=MaterialExpressionMaterialFunctionCall
   inputs [Emissive {from=MaterialExpressionDivide_0}]
   properties {materialFunction=&MF_FX_EmissiveAdjustments}
  }
  MaterialExpressionMaterialFunctionCall_10 {
   class=MaterialExpressionMaterialFunctionCall
   inputs [
    UorV {from=MaterialExpressionConvert_3}
    Invert_EdgeControl {from=MaterialExpressionVectorParameter_10, output=RGB}
    NoInvert_EdgeControl {from=MaterialExpressionVectorParameter_9, output=RGB}
   ]
   properties {materialFunction=/Game/Effects/Niagara/MasterMaterials/Functions/Common/FXCommon_EdgeMask_Ex}
  }
  MaterialExpressionMaterialFunctionCall_2 {
   class=MaterialExpressionMaterialFunctionCall
   inputs [
    CurveIndex {from=MaterialExpressionConstant_0}
    CurveNum {from=MaterialExpressionReroute_1}
    GreyMapIn {from=MaterialExpressionSphereMask_1}
   ]
   properties {materialFunction=/Game/Effects/Niagara/MasterMaterials/Functions/Common/FXCommon_CurveAtlasUV}
  }
  MaterialExpressionMaterialFunctionCall_3 {
   class=MaterialExpressionMaterialFunctionCall
   inputs [
    UVs {from=MaterialExpressionTextureCoordinate_0}
    Texture Scale {from=MaterialExpressionConvert_9}
   ]
   properties {materialFunction=&ScaleUVsByCenter}
  }
  MaterialExpressionMaterialFunctionCall_4 {
   class=MaterialExpressionMaterialFunctionCall
   inputs [
    CurveIndex {from=MaterialExpressionConstant_2}
    CurveNum {from=MaterialExpressionConstant_3}
    GreyMapIn {from=MaterialExpressionTextureCoordinate_0}
   ]
   properties {materialFunction=&FXCommon_CurveAtlasUV}
  }
  MaterialExpressionMaterialFunctionCall_5 {
   class=MaterialExpressionMaterialFunctionCall
   inputs [
    Texture {from=MaterialExpressionTextureObjectParameter_0}
    AdditionalUVOffset {from=""}
    UVControl_RGB {from=MaterialExpressionVectorParameter_1, output=RGB}
    UVControl_A {from=MaterialExpressionVectorParameter_1, output=RGB}
    "[Override]TimeInput" {from=""}
    "[Override]UVInput" {from=""}
   ]
   properties {materialFunction=&FXCommon_TextureSample}
  }
  MaterialExpressionMaterialFunctionCall_6 {
   class=MaterialExpressionMaterialFunctionCall
   inputs [
    UVs {from=MaterialExpressionTextureCoordinate_0}
    Texture Scale {from=MaterialExpressionConvert_1}
   ]
   properties {materialFunction=&ScaleUVsByCenter}
  }
  MaterialExpressionMaterialFunctionCall_7 {
   class=MaterialExpressionMaterialFunctionCall
   inputs [
    CurveIndex {from=MaterialExpressionConstant_5}
    CurveNum {from=MaterialExpressionReroute_1}
    GreyMapIn {from=MaterialExpressionSphereMask_0}
   ]
   properties {materialFunction=&FXCommon_CurveAtlasUV}
  }
  MaterialExpressionMaterialFunctionCall_8 {
   class=MaterialExpressionMaterialFunctionCall
   inputs [
    UorV {from=MaterialExpressionConvert_3}
    Invert_EdgeControl {from=MaterialExpressionVectorParameter_3, output=RGB}
    NoInvert_EdgeControl {from=MaterialExpressionVectorParameter_2, output=RGB}
   ]
   properties {materialFunction=&FXCommon_EdgeMask_Ex}
  }
  MaterialExpressionMaterialFunctionCall_9 {
   class=MaterialExpressionMaterialFunctionCall
   inputs [Distance {from=MaterialExpressionScalarParameter_13}]
   properties {materialFunction=&Near_Camera_Fade}
  }
  MaterialExpressionMultiply_0 {
   class=MaterialExpressionMultiply
   inputs [
    A {from=MaterialExpressionAdd_3}
    B {from=MaterialExpressionVectorParameter_0, output=RGB}
   ]
  }
  MaterialExpressionMultiply_1 {
   class=MaterialExpressionMultiply
   inputs [
    A {from=MaterialExpressionTextureSample_0, output=R}
    B {from=MaterialExpressionSaturate_3}
   ]
  }
  MaterialExpressionMultiply_10 {
   class=MaterialExpressionMultiply
   inputs [
    A {from=MaterialExpressionAdd_2}
    B {from=MaterialExpressionScalarParameter_7}
   ]
  }
  MaterialExpressionMultiply_11 {
   class=MaterialExpressionMultiply
   inputs [
    A {from=MaterialExpressionTextureSample_3, output=RGB}
    B {from=MaterialExpressionScalarParameter_12}
   ]
  }
  MaterialExpressionMultiply_12 {
   class=MaterialExpressionMultiply
   inputs [
    A {from=MaterialExpressionMaterialFunctionCall_10, output=Mask}
    B {from=MaterialExpressionTextureSample_0, output=R}
   ]
  }
  MaterialExpressionMultiply_13 {
   class=MaterialExpressionMultiply
   inputs [
    A {from=MaterialExpressionMaterialFunctionCall_10, output=Mask}
    B {from=MaterialExpressionTextureSample_2, output=R}
   ]
  }
  MaterialExpressionMultiply_14 {
   class=MaterialExpressionMultiply
   inputs [
    A {from=MaterialExpressionMaterialFunctionCall_9, output=Result}
    B {from=MaterialExpressionMaterialFunctionCall_9, output=Result}
   ]
  }
  MaterialExpressionMultiply_15 {
   class=MaterialExpressionMultiply
   inputs [
    A {from=MaterialExpressionMultiply_16}
    B {from=MaterialExpressionMultiply_5}
   ]
  }
  MaterialExpressionMultiply_16 {
   class=MaterialExpressionMultiply
   inputs [
    A {from=MaterialExpressionMultiply_14}
    B {from=MaterialExpressionMultiply_14}
   ]
  }
  MaterialExpressionMultiply_2 {
   class=MaterialExpressionMultiply
   inputs [
    A {from=MaterialExpressionNormalize_1}
    B {from=MaterialExpressionScalarParameter_1}
   ]
  }
  MaterialExpressionMultiply_3 {
   class=MaterialExpressionMultiply
   inputs [
    A {from=MaterialExpressionMultiply_6}
    B {from=MaterialExpressionSaturate_1}
   ]
  }
  MaterialExpressionMultiply_4 {
   class=MaterialExpressionMultiply
   inputs [
    A {from=MaterialExpressionAdd_0}
    B {from=MaterialExpressionScalarParameter_3}
   ]
  }
  MaterialExpressionMultiply_5 {
   class=MaterialExpressionMultiply
   inputs [
    A {from=MaterialExpressionMaterialFunctionCall_8, output=Mask}
    B {from=MaterialExpressionAdd_4}
   ]
  }
  MaterialExpressionMultiply_6 {
   class=MaterialExpressionMultiply
   inputs [
    A {from=MaterialExpressionTextureSample_2, output=R}
    B {from=MaterialExpressionScalarParameter_9}
   ]
  }
  MaterialExpressionMultiply_7 {
   class=MaterialExpressionMultiply
   inputs [
    A {from=MaterialExpressionAdd_6}
    B {from=""}
   ]
   properties {constB=0.5}
  }
  MaterialExpressionMultiply_8 {
   class=MaterialExpressionMultiply
   inputs [
    A {from=MaterialExpressionTextureSample_1, output=G}
    B {from=MaterialExpressionLinearInterpolate_1}
   ]
  }
  MaterialExpressionMultiply_9 {
   class=MaterialExpressionMultiply
   inputs [
    A {from=MaterialExpressionDynamicParameter_0, output=Param2}
    B {from=MaterialExpressionScalarParameter_0}
   ]
  }
  MaterialExpressionNormalize_1 {class=MaterialExpressionNormalize, inputs [VectorInput {from=MaterialExpressionSubtract_0, output=""}]}
  MaterialExpressionOneMinus_2 {class=MaterialExpressionOneMinus, inputs [None {from=MaterialExpressionMultiply_7}]}
  MaterialExpressionPower_1 {
   class=MaterialExpressionPower
   inputs [
    Base {from=MaterialExpressionSaturate_0}
    Exp {from=MaterialExpressionScalarParameter_4}
   ]
  }
  MaterialExpressionPower_2 {
   class=MaterialExpressionPower
   inputs [
    Base {from=MaterialExpressionOneMinus_2}
    Exp {from=MaterialExpressionScalarParameter_11}
   ]
  }
  MaterialExpressionPower_6 {
   class=MaterialExpressionPower
   inputs [
    Base {from=MaterialExpressionSaturate_2}
    Exp {from=MaterialExpressionScalarParameter_8}
   ]
  }
  MaterialExpressionReroute_1 {class=MaterialExpressionReroute, inputs [None {from=MaterialExpressionConstant_3}]}
  MaterialExpressionSaturate_0 {class=MaterialExpressionSaturate, inputs [None {from=MaterialExpressionMultiply_4}]}
  MaterialExpressionSaturate_1 {class=MaterialExpressionSaturate, inputs [None {from=MaterialExpressionPower_1}]}
  MaterialExpressionSaturate_2 {class=MaterialExpressionSaturate, inputs [None {from=MaterialExpressionMultiply_10}]}
  MaterialExpressionSaturate_3 {class=MaterialExpressionSaturate, inputs [None {from=MaterialExpressionPower_6}]}
  MaterialExpressionScalarParameter_0 {class=MaterialExpressionScalarParameter, properties {defaultValue=87.6774, parameterName=Width}}
  MaterialExpressionScalarParameter_1 {class=MaterialExpressionScalarParameter, properties {defaultValue=-25, parameterName=CameraOffset}}
  MaterialExpressionScalarParameter_10 {class=MaterialExpressionScalarParameter, properties {defaultValue=2, parameterName=HL}}
  MaterialExpressionScalarParameter_11 {class=MaterialExpressionScalarParameter, properties {defaultValue=16.0882, parameterName=HLPowe}}
  MaterialExpressionScalarParameter_12 {class=MaterialExpressionScalarParameter, properties {defaultValue=0.5, parameterName=GS}}
  MaterialExpressionScalarParameter_13 {class=MaterialExpressionScalarParameter, properties {defaultValue=200, parameterName=Near_Camera_Fade}}
  MaterialExpressionScalarParameter_2 {class=MaterialExpressionScalarParameter, properties {defaultValue=0.392, group=Dissolve, parameterName=Dissolve2_GrayAdjust_Add}}
  MaterialExpressionScalarParameter_3 {class=MaterialExpressionScalarParameter, properties {defaultValue=1, group=Dissolve, parameterName=Dissolve_GrayAdjust_Mul}}
  MaterialExpressionScalarParameter_4 {class=MaterialExpressionScalarParameter, properties {defaultValue=0.8778, group=Dissolve, parameterName=Dissolve2_GrayAdjust_Power}}
  MaterialExpressionScalarParameter_5 {class=MaterialExpressionScalarParameter, properties {defaultValue=0.0955, group=Dissolve, parameterName=DissolveO_GrayAdjust_Add}}
  MaterialExpressionScalarParameter_6 {class=MaterialExpressionScalarParameter, properties {defaultValue=1, group=Dissolve, parameterName=Dissolve_GrayAdjust_Add}}
  MaterialExpressionScalarParameter_7 {class=MaterialExpressionScalarParameter, properties {defaultValue=1, group=Dissolve, parameterName=Dissolve_GrayAdjust_Mul}}
  MaterialExpressionScalarParameter_8 {class=MaterialExpressionScalarParameter, properties {defaultValue=0.3, group=Dissolve, parameterName=Dissolve_GrayAdjust_Power}}
  MaterialExpressionScalarParameter_9 {class=MaterialExpressionScalarParameter, properties {defaultValue=1.2, parameterName=GlowWScale}}
  MaterialExpressionSphereMask_0 {
   class=MaterialExpressionSphereMask
   inputs [
    A {from=MaterialExpressionConvert_2, output=""}
    B {from=MaterialExpressionConstant_4, output=""}
    Radius {from="", output=""}
    Hardness {from="", output=""}
   ]
   properties {attenuationRadius=0.5, hardnessPercent=0}
  }
  MaterialExpressionSphereMask_1 {
   class=MaterialExpressionSphereMask
   inputs [
    A {from=MaterialExpressionConvert_11, output=""}
    B {from=MaterialExpressionConstant_7, output=""}
    Radius {from="", output=""}
    Hardness {from="", output=""}
   ]
   properties {attenuationRadius=0.5, hardnessPercent=0}
  }
  MaterialExpressionStaticBool_0 {class=MaterialExpressionStaticBool}
  MaterialExpressionSubtract_0 {
   class=MaterialExpressionSubtract
   inputs [
    A {from=MaterialExpressionAdd_7}
    B {from=MaterialExpressionCameraPositionWS_0}
   ]
  }
  MaterialExpressionTextureCoordinate_0 {class=MaterialExpressionTextureCoordinate}
  MaterialExpressionTextureObjectParameter_0 {
   class=MaterialExpressionTextureObjectParameter
   inputs [
    UVs {from="", output=""}
    Apply View MipBias {from="", output=""}
   ]
   properties {
    group=Dissolve
    parameterName=Dissolve_Texture
    samplerType=SAMPLERTYPE_LinearGrayscale
    sortPriority=0
    texture=&T_SimplePerlin_Greyscale
   }
  }
  MaterialExpressionTextureSample_0 {
   class=MaterialExpressionTextureSample
   inputs [
    UVs {from=MaterialExpressionMaterialFunctionCall_4, output=UV}
    Tex {from=""}
    Apply View MipBias {from=""}
   ]
   properties {samplerType=SAMPLERTYPE_LinearColor, texture=&/CLA_DentureSay}
  }
  MaterialExpressionTextureSample_1 {
   class=MaterialExpressionTextureSample
   inputs [
    UVs {from=MaterialExpressionMaterialFunctionCall_2, output=UV}
    Tex {from=""}
    Apply View MipBias {from=""}
   ]
   properties {samplerType=SAMPLERTYPE_LinearColor, texture=&CLA_DentureSay}
  }
  MaterialExpressionTextureSample_2 {
   class=MaterialExpressionTextureSample
   inputs [
    UVs {from=MaterialExpressionMaterialFunctionCall_4, output=UV}
    Tex {from=""}
    Apply View MipBias {from=""}
   ]
   properties {samplerType=SAMPLERTYPE_LinearColor, texture=&CLA_DentureSay}
  }
  MaterialExpressionTextureSample_3 {
   class=MaterialExpressionTextureSample
   inputs [
    UVs {from=MaterialExpressionMaterialFunctionCall_7, output=UV}
    Tex {from=""}
    Apply View MipBias {from=""}
   ]
   properties {samplerType=SAMPLERTYPE_LinearColor, texture=&CLA_DentureSay}
  }
  MaterialExpressionVectorParameter_0 {class=MaterialExpressionVectorParameter, properties {defaultValue=<rgba 1 0.4775 0.05 1>}}
  MaterialExpressionVectorParameter_1 {
   class=MaterialExpressionVectorParameter
   properties {group=Dissolve, parameterName=Dissolve_UVControl, channelNames {a=VPan, b=UPan, g=VScale, r=UScale}, defaultValue=<rgba 3 0 0 0>}
  }
  MaterialExpressionVectorParameter_10 {
   class=MaterialExpressionVectorParameter
   properties {
    group=EdgeF
    parameterName=EdgeF_EdgeMask_InvertControl
    channelNames {a="", b=Power, g=Multiply, r=Offset}
    defaultValue=<rgba 0.128 2.1835 1 1>
   }
  }
  MaterialExpressionVectorParameter_2 {
   class=MaterialExpressionVectorParameter
   properties {
    group=EdgeOpacityFade
    parameterName=EdgeOpacityFade_EdgeMask_Control
    channelNames {a="", b=Power, g=Multiply, r=Offset}
    defaultValue=<rgba 1 1 1 1>
   }
  }
  MaterialExpressionVectorParameter_3 {
   class=MaterialExpressionVectorParameter
   properties {
    group=EdgeOpacityFade
    parameterName=EdgeOpacityFade_EdgeMask_InvertControl
    channelNames {a="", b=Power, g=Multiply, r=Offset}
    defaultValue=<rgba 0 6 2 1>
   }
  }
  MaterialExpressionVectorParameter_9 {
   class=MaterialExpressionVectorParameter
   properties {group=EdgeF, parameterName=EdgeF_EdgeMask_Control, channelNames {a="", b=Power, g=Multiply, r=Offset}, defaultValue=<rgba -0.0704 3 1 1>}
  }
  MaterialExpressionWorldPosition_0 {class=MaterialExpressionWorldPosition, inputs [], properties {}}
 }
 properties {
  bAutomaticallySetUsageInEditor=false
  bUsedWithNiagaraMeshParticles=true
  blendMode=BLEND_Additive
  previewMesh=/NarrowFleaCollectible/Assets/FX/Ring/SM_NarrowFlea_Rings
  shadingModel=MSM_Unlit
  twoSided=true
  usedShadingModels=Unlit
 }
}
