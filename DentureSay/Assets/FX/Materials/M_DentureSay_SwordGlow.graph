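# Additive, unlit, two-sided Niagara material for a sword glow effect, usable on sprites, ribbons and mesh particles.
# A color parameter is hue-shifted along V, tinted by particle color and passed through shared emissive adjustments.
# S-curved U and V edge masks, a V safe-bound mask, color and particle alpha, and powered depth fade form the soft alpha.
# SplineThicken sets width, optionally scaled by a dynamic parameter.
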Material /DentureSay/Assets/FX/Materials/M_DentureSay_SwordGlow
graph {
 properties {
  bAutomaticallySetUsageInEditor=false
  bUsedWithNiagaraMeshParticles=true
  bUsedWithNiagaraRibbons=true
  bUsedWithNiagaraSprites=true
  blendMode=BLEND_Additive
  shadingModel=MSM_Unlit
  twoSided=true
  usedShadingModels=Unlit
 }
 nodes [
  n1 = TextureCoordinate()
  n2 = VectorParameter() {parameterName=Color, defaultValue=<rgba 2 0.6715 0.0073 0.6508>}
  n3 = MaterialFunctionCall(
   Hue Shift Percentage=Multiply(
    A=MaterialFunctionCall(Float2=@n1) {materialFunction=&BreakOutFloat2Components}.G
    B=ScalarParameter() {defaultValue=0.07, parameterName=HueShift}
   )
   Texture=@n2.RGB
  ) {materialFunction=&HueShift}
  n4 = ParticleColor()
  MaterialFunctionCall(Emissive=Multiply(A=Multiply(A=@n3.Result) {constB=0.6}, B=@n4.RGB)) {materialFunction=&MF_FX_EmissiveAdjustments}
  n5 = Multiply(
   A=ScalarParameter() {parameterName=Width}
   B=LinearInterpolate(B=DynamicParameter() {paramNames [WidthScale,Param2,Param3,Param4]}.WidthScale, Alpha=ScalarParameter() {parameterName=EnableDPWidth}) {constA=1}
  )
  MaterialFunctionCall(WidthBase=@n5, WidthTip=@n5) {materialFunction=&SplineThicken}
  n6 = MaterialFunctionCall(Float2=TextureCoordinate()) {materialFunction=&BreakOutFloat2Components}
  n7 = VectorParameter() {group=SafeBound, parameterName=SafeBound_EdgeMask_Control, channelNames {a="", b=Power, g=Multiply, r=Offset}, defaultValue=<rgba 1 1 1 1>}
  n8 = VectorParameter() {
   group=SwordGlow
   parameterName=SwordGlow_EdgeMask_Control
   channelNames {a="", b=Power, g=Multiply, r=Offset}
   defaultValue=<rgba 0 1 1 4.077>
  }
  n9 = MaterialFunctionCall!(UorV=@n6.R, EdgeControl=@n8) {materialFunction=/Game/Effects/Niagara/MasterMaterials/Functions/Common/FXCommon_EdgeMask_Simple}
  n10 = VectorParameter() {
   group=SwordGlowV
   parameterName=SwordGlowV_EdgeMask_InvertControl
   channelNames {a="", b=Power, g=Multiply, r=Offset}
   defaultValue=<rgba 0 3 2 2.8997>
  }
  n11 = VectorParameter() {
   group=SwordGlowV
   parameterName=SwordGlowV_EdgeMask_Control
   channelNames {a="", b=Power, g=Multiply, r=Offset}
   defaultValue=<rgba 0 3.3464 3 2.8997>
  }
  n12 = MaterialFunctionCall!(UorV=@n6.G, Invert_EdgeControl=@n10.RGB, NoInvert_EdgeControl=@n11.RGB) {materialFunction=/Game/Effects/Niagara/MasterMaterials/Functions/Common/FXCommon_EdgeMask_Ex}
  n13 = ScalarParameter() {defaultValue=2.8997, parameterName=SwordGlowV_Scurve}
  n14 = Multiply(
   A=MaterialFunctionCall(UorV=@n6.G, EdgeControl=@n7.RGB) {materialFunction=&FXCommon_EdgeMask_Simple}.Mask
   B=Multiply(
    A=MaterialFunctionCall(In=@n9.Mask, Power=@n8.A) {materialFunction=&SCurve}.Result
    B=MaterialFunctionCall(In=@n12.Mask, Power=@n13) {materialFunction=&SCurve}.Result
   )
  )
  n15 = DepthFade(FadeDistance=ScalarParameter() {defaultValue=20, parameterName=DF})
  Multiply(A=@n4.A, B=Multiply(A=Multiply(A=@n2.A, B=@n14), B=Power(Base=@n15)))
 ]
}
