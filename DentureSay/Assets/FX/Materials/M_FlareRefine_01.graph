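# Unlit additive Niagara sprite and mesh-particle material for a flare, using a star-shaped texture by default.
# A center-scaled second sample is subtracted from the first, and the Edge dynamic parameter blends between solid shape and outline.
# That mask, scaled by Flare Mask and particle alpha, feeds a DepthFade.
# Color runs the unmodified first sample through MF_FX_Refine, tints it by particle color, then applies MF_FX_EmissiveAdjustments.
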Material /DentureSay/Assets/FX/Materials/M_FlareRefine_01
graph {
 properties {
  bUsedWithNiagaraMeshParticles=true
  bUsedWithNiagaraSprites=true
  blendMode=BLEND_Additive
  pixelDepthOffsetMode=PDOM_Legacy
  shadingModel=MSM_Unlit
  usedShadingModels=Unlit
 }
 nodes [
  n1 = ParticleColor()
  n2 = TextureObjectParameter() {
   group=Flare
   parameterName=Flare Tex
   samplerType=SAMPLERTYPE_LinearColor
   texture=/Game/Effects/Niagara/MasterMaterials/Shape/Texture/T_RoundCorner_Star
  }
  n3 = ChannelMaskParameter(None=TextureSample(Tex=@n2).RGBA) {group=Flare, parameterName=FlareMaskChannel}
  n4 = TextureCoordinate()
  n5 = ScalarParameter() {defaultValue=0.9, parameterName=ScaleSecondLayer}
  n6 = ChannelMaskParameter(None=TextureSample(
    UVs=MaterialFunctionCall(UVs=@n4, Texture Scale=@n5) {materialFunction=&ScaleUVsByCenter}.UVs
    Tex=@n2
   ).RGBA) {group=Flare, parameterName=FlareMaskChannel}
  n7 = Saturate(None=LinearInterpolate(A=@n3.R, B=Subtract(A=@n3.R, B=@n6.R), Alpha=DynamicParameter() {defaultValue=<rgba 0.9 1 1 1>, paramNames [Edge,Param2,Param3,Param4]}.Edge))
  n8 = ScalarParameter() {defaultValue=1, group=Flare, parameterName=Flare Mask}
  DepthFade(Opacity=Multiply(A=@n1.A, B=Multiply(A=@n7, B=@n8)), FadeDistance=ScalarParameter() {parameterName=FadeDistance})
  n9 = VectorParameter() {group=Flare, parameterName=Flare Refine, defaultValue=<rgba 1 1 1 0.5>}
  MaterialFunctionCall(Emissive=Multiply(
    A=MaterialFunctionCall(Texture=@n3.R, RGB=@n9.RGB, A=@n9.RGB) {materialFunction=&MF_FX_Refine}.Result
    B=@n1.RGB
   )) {materialFunction=&MF_FX_EmissiveAdjustments}
 ]
}
