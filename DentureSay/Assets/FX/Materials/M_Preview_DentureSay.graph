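# Translucent, unlit, two-sided preview material that draws a soft orange glowing volume, fading toward the top of its bounds.
# Bounding-box UVW feeds an inverted sphere mask, whose tunable power shapes the color from hollow center to bright edges.
# An inverted depth fade adds extra glow where the surface meets scene geometry.
# Opacity multiplies the edge mask by an inverted vertical gradient, then depth-fades to soften intersections.
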Material /DentureSay/Assets/FX/Materials/M_Preview_DentureSay
graph {
 properties {bAutomaticallySetUsageInEditor=false, blendMode=BLEND_Translucent, shadingModel=MSM_Unlit, twoSided=true, usedShadingModels=Unlit}
 nodes [
  n1 = DepthFade(FadeDistance=ScalarParameter() {parameterName=Param_5})
  n2 = VectorParameter() {defaultValue=<rgba 4 1.6589 0 2>}
  n3 = OneMinus(None=SphereMask(
    A=MaterialFunctionCall() {materialFunction=&BoundingBoxBased_0-1_UVW}.RGB
    B=ScalarParameter() {defaultValue=0.5}
    Radius=ScalarParameter() {defaultValue=0.8945, parameterName=Param_1}
    Hardness=ScalarParameter() {defaultValue=-0.1118, parameterName=Param_2}
   ))
  n4 = ScalarParameter() {defaultValue=2, parameterName=Param_3}
  Add(A=Multiply(A=OneMinus(None=@n1), B=@n2.RGB), B=LinearInterpolate(B=@n2.RGB, Alpha=Power(Base=@n3, Exp=@n4)))
  n5 = OneMinus(None=MaterialFunctionCall() {materialFunction=&BoundingBoxBased_0-1_UVW}.B)
  n6 = ScalarParameter() {defaultValue=2, parameterName=Param_4}
  DepthFade(Opacity=Multiply(A=Power(Base=@n3), B=Power(Base=@n5, Exp=@n6)))
 ]
}
