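# Unlit, translucent, two-sided material for Niagara mesh particles that draws animated light-ray streaks for the WaywardRebel skin star effect.
# One ray texture is sampled twice with separate scale and pan controls; red times green is boosted and contrasted.
# That result, a single-axis UV edge fade and particle alpha multiply into an alpha mask.
# Color is particle RGB times an Emissive parameter, passed through the shared emissive-adjustment function.
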Material /DentureSay/Assets/FX/Materials/M_WaywardRebel_Skin_Star
graph {
 properties {
  bUsedWithNiagaraMeshParticles=true
  blendMode=BLEND_Translucent
  pixelDepthOffsetMode=PDOM_Legacy
  shadingModel=MSM_Unlit
  twoSided=true
  usedShadingModels=Unlit
 }
 nodes [
  n1 = ParticleColor()
  MaterialFunctionCall(Emissive=Multiply(A=@n1.RGB, B=ScalarParameter() {defaultValue=1, parameterName=Emissive}) {constB=100}) {materialFunction=&MF_FX_EmissiveAdjustments}
  n2 = TextureObjectParameter() {
   group=FlareT_A
   parameterName=FlareT_A_Texture
   sortPriority=0
   texture=/Game/Effects/Fort_Effects/Textures/Outlander/T_InTheZone_LightRays_01
  }
  n3 = VectorParameter() {group=FlareT_A, parameterName=FlareT_A_UVControl, channelNames {a=VPan, b=UPan, g=VScale, r=UScale}, defaultValue=<rgba 6 0.01 0.05 0>}
  n4 = MaterialFunctionCall(Texture=@n2, UVControl_RGB=@n3.RGB, UVControl_A=@n3.RGB) {materialFunction=&FXCommon_TextureSample}
  n5 = VectorParameter() {group=FlareTB, parameterName=FlareTB_UVControl, channelNames {a=VPan, b=UPan, g=VScale, r=UScale}, defaultValue=<rgba 4 0.01 -0.1 0>}
  n6 = MaterialFunctionCall(Texture=@n2, UVControl_RGB=@n5.RGB, UVControl_A=@n5.RGB) {materialFunction=&FXCommon_TextureSample}
  n7 = MaterialFunctionCall(
   In=Multiply(A=Multiply(A=@n4.R, B=@n6.G), B=ScalarParameter() {defaultValue=2, parameterName=LightMul})
   Contrast=ScalarParameter() {defaultValue=0.0632, parameterName=LightContrast}
  ) {materialFunction=&CheapContrast}
  MaterialExpressionTextureCoordinate_0 = TextureCoordinate()
  n8 = Convert(None=@MaterialExpressionTextureCoordinate_0) {
   nodeName=Break Float2
   convertInputs [{
     type=Vector2
     expressionInput {
      expression=&self:MaterialExpressionTextureCoordinate_0
      inputName=None
      mask=0
      maskA=0
      maskB=0
      maskG=0
      maskR=0
      outputIndex=0
     }
    }]
   convertOutputs [{type=Scalar}, {type=Scalar}]
  }
  n9 = VectorParameter() {
   group=Fade
   parameterName=Fade_EdgeMask_InvertControl
   channelNames {a="", b=Power, g=Multiply, r=Offset}
   defaultValue=<rgba 0 19.4834 2.7169 1>
  }
  n10 = VectorParameter() {group=Fade, parameterName=Fade_EdgeMask_Control, channelNames {a="", b=Power, g=Multiply, r=Offset}, defaultValue=<rgba 0 1 5.1849 1>}
  Multiply(
   A=@n1.A
   B=Multiply(
    A=Multiply(A=@n7.Result)
    B=MaterialFunctionCall!(UorV=@n8, Invert_EdgeControl=@n9.RGB, NoInvert_EdgeControl=@n10.RGB) {materialFunction=/Game/Effects/Niagara/MasterMaterials/Functions/Common/FXCommon_EdgeMask_Ex}.Mask
   )
  )
 ]
}
