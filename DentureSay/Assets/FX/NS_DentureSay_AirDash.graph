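# Niagara effect for the DentureSay air dash: a short burst of spinning spiral meshes, expanding wind cones and twinkling star sprites.
# WindCone and Stars use spawn groups for orange and blue variants, while Spirals takes its tint from Blockbreak and custom-color user parameters.
# PP spawns a component-rendered particle only for the local player, driving a speed curve faded by distance.
# Several older emitters, including speedlines, shock cone and exhaust flame, are disabled.
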NiagaraSystem /DentureSay/Assets/FX/NS_DentureSay_AirDash
dependencies {
 usedDataInterfaces [
  &NiagaraDataInterfaceCurve
  &NiagaraDataInterfaceVectorCurve
  &NiagaraDataInterfaceColorCurve
  &NiagaraDataInterfaceCamera
 ]
 usedDynamicInputs [&MakeVector4FromVectorAndFloat, /Game/Effects/Niagara/Modules/Math/FN_CameraDistance]
 usedModules [
  &SystemState
  &self:SystemSpawnScript.SystemScriptSource.SystemScriptGraph.NiagaraNodeAssignment_0.SetVariables_#1
 ]
}
emitters {
 Endflare {rendererClasses [], renderers []}
 ExhaustFlame001 {
  bEnabled=false
  emitterUpdateScript [
   &EmitterState()
   &SpawnBurst_Instantaneous(Use Spawn Probability=false, UseLoopCountLimit=false)
  ]
  particleSpawnScript [
   &InitializeParticle(
    Color=<rgba 20 12.35 3 1>
    Lifetime=0.28
    Mesh Scale=&Multiply_VectorByFloat(Vector=<0.6 0.35 0.35>, Float=3.5)
    Mesh Scale Mode=::NewEnumerator4
    Position Offset=<0 56.1348 -0.32>
    UsePositionOffset=true
    Write Material Random=true
   )
   &AddVelocity(Velocity=&Multiply_VectorByFloat(Vector=<0 -1 0>, Float=100))
   &CameraOffset()
  ]
  particleUpdateScript [
   &ParticleState()
   &SolveForcesAndVelocity(Clamp Velocity=false, Limit Acceleration=false)
   &ScaleColor(
    Scale Alpha=&FloatFromCurve(FloatCurve=DataInterface() {propertyValues {Curve {keys [
         {interpMode=RCIM_Cubic, leaveTangent=9.8521, tangentMode=RCTM_Break, time=0, value=0.1}
         {interpMode=RCIM_Cubic, tangentMode=RCTM_User, time=0.2, value=1}
         {interpMode=RCIM_Cubic, tangentMode=RCTM_Break, time=1, value=0}
        ]}}})
    Scale RGB=&VectorFromFloat(Value=&Multiply_Float(
      A=&RandomRangeFloat(Minimum=3, Maximum=20)
      B=&FloatFromCurve(
       FloatCurve=DataInterface() {propertyValues {Curve {keys [
           {interpMode=RCIM_Cubic, tangentMode=RCTM_User, time=0.3, value=1}
           {arriveTangentWeight=0.6424, interpMode=RCIM_Cubic, tangentMode=RCTM_User, time=1, value=0.25}
          ]}}}
       CurveIndex=$Emitter.NormalizedLoopAge
       Scale Curve=3
      )
     ))
    ScaleA=true
    ScaleRGB=true
   )
   &ScaleMeshSize(Scale Factor=&VectorFromCurve(
     VectorCurve=DataInterface() {propertyValues {
       bOptimizeLUT=false
       XCurve {keys [{arriveTangent=1.5411, interpMode=RCIM_Cubic, leaveTangent=1.5411, tangentMode=RCTM_User, time=0, value=0.75}, <at 1 2>]}
       YCurve {keys [{arriveTangent=2.1071, interpMode=RCIM_Cubic, leaveTangent=2.1071, tangentMode=RCTM_User, time=0, value=0.75}, <at 1 2>]}
       ZCurve {keys [
         {arriveTangent=1.9563, interpMode=RCIM_Cubic, leaveTangent=1.9563, tangentMode=RCTM_User, time=0.0018, value=0.7464}
         {arriveTangent=-0.105, interpMode=RCIM_Cubic, leaveTangent=-0.105, tangentMode=RCTM_User, time=0.9987, value=1.9917}
        ]}
      }}
     Scale Curve=&VectorFromFloat(Value=&Lerp_Float(
       A=1.3
       B=2
       Alpha=&FloatFromCurve(
        FloatCurve=DataInterface() {propertyValues {Curve {keys [
            {interpMode=RCIM_Cubic, tangentMode=RCTM_User, time=0, value=1}
            {arriveTangentWeight=1.1078, interpMode=RCIM_Cubic, tangentMode=RCTM_User, time=1, value=0}
           ]}}}
        CurveIndex=$Emitter.NormalizedLoopAge
       )
      ))
    ))
   /Game/Effects/Niagara/Modules/Materials/MaterialPanner(
    Jitter Pan by SpawnGroups=false
    Offset:NiagaraFloat=&Add_Float(A=$Emitter.Age, B=$Particles.Initial.MaterialRandom)
    Speed=-5
   )
  ]
  rendererClasses [&NiagaraMeshRendererProperties]
  renderers [{rendererClass=&NiagaraMeshRendererProperties, rendererIndex=0}]
 }
 ExhaustFlame001_Lightweight {rendererClasses [], renderers []}
 FlashMesh001 {rendererClasses [], renderers []}
 FlashMesh002 {rendererClasses [], renderers []}
 FlashMesh004 {rendererClasses [], renderers []}
 Light {rendererClasses [], renderers []}
 PP {
  emitterUpdateScript [
   &EmitterState()
   &SpawnBurst_Instantaneous(
    Spawn Count=&MakeCustomIntFromBool(Bool=$User.LocalPlayer, True INT=1, False INT=0)
    Use Spawn Probability=false
    UseLoopCountLimit=false
   )
  ]
  particleSpawnScript [&InitializeParticle(
    Color Mode=::NewEnumerator0
    Lifetime=$System.CurrentLoopDuration
    Sprite Size Mode=::NewEnumerator3
    Uniform Sprite Size=1
    UsePositionOffset=false
    Write Material Random=true
   )]
  particleUpdateScript [
   /SetVariables(Particles.SpeedNormalizedB:NiagaraFloat=&FloatFromCurve(
     FloatCurve=DataInterface() {propertyValues {Curve {keys [<at 0 0 RCIM_Cubic>,<at 0.15 1 RCIM_Cubic>,<at 1 0 RCIM_Cubic>]}}}
     CurveIndex=$System.NormalizedLoopAge
    )) {enabled=false, isSetParameters=true}
   /SetVariables(Particles.SpeedNormalizedB:NiagaraFloat=&FloatFromCurve(
     FloatCurve=DataInterface() {propertyValues {Curve {keys [<at 0 0 RCIM_Cubic>,<at 0.15 1 RCIM_Cubic>,<at 1 0 RCIM_Cubic>]}}}
     CurveIndex=$System.NormalizedLoopAge
     Scale Curve=&Multiply_Float(
      A=1
      B=&FN_LOD_Distance_Remap(Start Distance=2500, End Distance=3250, Minimum Output=1, Maximum Output=0, Clamp Output=true)
     )
    )) {isSetParameters=true}
   &ParticleState()
  ]
  rendererClasses [&NiagaraComponentRendererProperties]
  renderers [{rendererClass=&NiagaraComponentRendererProperties, rendererIndex=0}]
 }
 Shock_Cone {
  bEnabled=false
  emitterUpdateScript [
   &EmitterState()
   &SpawnBurst_Instantaneous(Use Spawn Probability=false, UseLoopCountLimit=false)
  ]
  particleSpawnScript [
   &InitializeParticle(
    Color=<rgba 0.7 0.364 0.07 0.3>
    Lifetime=0.4
    Mass Max=2
    Mass Mode=::NewEnumerator3
    Mesh Scale=&Multiply_VectorByFloat(Vector=<1.8 1.5 1.5>, Float=1.3)
    Mesh Scale Mode=::NewEnumerator4
    Position Offset=<0 60 0>
    UsePositionOffset=true
    Write Material Random=true
   )
   /SetVariables(Particles.Color=&LerpLinearColors(StartColor=$Particles.Color, EndColor=<rgba 1.4 0.6905 0.3059 0.3>, LerpFactor=$User.Blockbreak)) {enabled=false, isSetParameters=true}
   /CRD_Melee/FX/NMS_Melee_SetColor(
    Block Break:NiagaraFloat=$User.Blockbreak
    Block Break Color:LinearColor=<rgba 1.4 0.6905 0.3059 0.3>
    Custom Color:LinearColor=$User.Color
    Custom Color Intensity:NiagaraFloat=$User.CustomColorIntensity
    Custom Color Multiplier=1.4
    Custom Color Opacity=0.3
    Non Block Break Color:LinearColor=<rgba 1 0.6449 0.1478 0.15>
    UseCustomColor:NiagaraFloat=$User.UseCustomColor
   )
   &DynamicMaterialParameters(
    Index 0 Param 1=0.7
    Index 0 Param 2=2.5
    Module.Param0WriteEnabled=true
    Module.Param1WriteEnabled=true
    Module.Param2WriteEnabled=false
    Module.Param3WriteEnabled=false
   )
   &InitialMeshOrientation(Mesh Orientation Mode=::NewEnumerator7, Rotation=<0 0 0.25>, Use Rotation Vector=true)
  ]
  particleUpdateScript [
   &ParticleState()
   &ScaleColor(
    Scale Alpha=&FloatFromCurve(
     FloatCurve=DataInterface() {propertyValues {Curve {keys [
         {arriveTangent=0.002, interpMode=RCIM_Cubic, leaveTangent=0.002, time=0, value=0}
         {arriveTangent=0.002, interpMode=RCIM_Cubic, leaveTangent=0.002, time=0.1, value=1}
         {arriveTangent=0.002, interpMode=RCIM_Cubic, leaveTangent=0.002, time=1.0011, value=0.002}
        ]}}}
     Scale Curve=20
    )
    Scale RGB=&VectorFromCurve(
     VectorCurve=DataInterface() {propertyValues {
       XCurve {keys [{interpMode=RCIM_Cubic, tangentMode=RCTM_User, time=0, value=2.875}, {interpMode=RCIM_Cubic, tangentMode=RCTM_User, time=1, value=1.4}]}
       YCurve {keys [
         {arriveTangent=0.0808, interpMode=RCIM_Cubic, leaveTangent=0.0808, tangentMode=RCTM_User, time=0, value=2.875}
         {arriveTangent=-0.0591, interpMode=RCIM_Cubic, leaveTangent=-0.0591, tangentMode=RCTM_User, time=0.991, value=1.353}
        ]}
       ZCurve {keys [{interpMode=RCIM_Cubic, tangentMode=RCTM_User, time=0, value=2.9051}, {interpMode=RCIM_Cubic, tangentMode=RCTM_User, time=1, value=1.3721}]}
      }}
     Scale Curve=&VectorFromFloat(Value=0.6)
    )
    ScaleA=true
    ScaleRGB=true
   )
   &ScaleMeshSize(Scale Factor=&VectorFromCurve(VectorCurve=DataInterface() {propertyValues {
       XCurve {keys [
         {arriveTangent=2.6319, interpMode=RCIM_Cubic, leaveTangent=2.6319, tangentMode=RCTM_User, time=0, value=0.5}
         {arriveTangent=-0.0896, interpMode=RCIM_Cubic, leaveTangent=-0.0896, tangentMode=RCTM_User, time=1, value=1.5}
        ]}
       YCurve {keys [
         {arriveTangent=2.3637, interpMode=RCIM_Cubic, leaveTangent=2.3637, tangentMode=RCTM_User, time=0, value=0.1}
         {interpMode=RCIM_Cubic, tangentMode=RCTM_User, time=1, value=1}
        ]}
       ZCurve {keys [
         {arriveTangent=2.3637, interpMode=RCIM_Cubic, leaveTangent=2.3637, tangentMode=RCTM_User, time=0, value=0.1}
         {interpMode=RCIM_Cubic, tangentMode=RCTM_User, time=1, value=1}
        ]}
      }}))
  ]
  rendererClasses [&NiagaraMeshRendererProperties]
  renderers [{rendererClass=&NiagaraMeshRendererProperties, rendererIndex=0}]
 }
 Shock_Cone_Lightweight {rendererClasses [], renderers []}
 Speedlines {
  bEnabled=false
  emitterUpdateScript [
   &EmitterState()
   &SpawnBurst_Instantaneous(Spawn Count=30, Spawn Time=0.09, Use Spawn Probability=false)
   &SpawnBurst_Instantaneous(Spawn Count=15, Spawn Time=0.05, Use Spawn Probability=false) {module=SpawnBurst_Instantaneous001}
  ]
  particleSpawnScript [
   &InitializeParticle(
    Color=<rgba 1 1 1 0.35>
    Lifetime=&RandomRangeFloat(Minimum=0, Maximum=0.3)
    Sprite Rotation Mode=::NewEnumerator0
    Sprite Size=<2 225>
    Sprite Size Mode=::NewEnumerator4
    UsePositionOffset=false
    Write Material Random=true
   ) {module=InitializeParticle001}
   &ShapeLocation(Box Size=<300 400 200>, MaskSpawnGroup=false, Override Seed=false, Shape Primitive=::NewEnumerator2, Surface Only=false, UseOwnerScale=true)
   &AddVelocity(Velocity=<0 -3743.4888 0>)
  ]
  particleUpdateScript [
   &ParticleState()
   &ScaleColor(
    Scale Alpha=&FloatFromCurve(
     FloatCurve=DataInterface() {propertyValues {Curve {keys [
         {arriveTangent=0.5269, interpMode=RCIM_Cubic, leaveTangent=0.5269, tangentMode=RCTM_User, time=0, value=0}
         {arriveTangent=-0.2619, interpMode=RCIM_Cubic, leaveTangent=-0.2619, tangentMode=RCTM_User, time=0.2607, value=0.9753}
         <at 1 0 RCIM_Cubic>
        ]}}}
     Scale Curve=0.1
    )
    Scale RGB=&VectorFromFloat(Value=5)
    ScaleA=true
    ScaleRGB=true
   )
   &Drag(Drag=7, Use Linear Drag=true, Use Rotational Drag=false)
   &SolveForcesAndVelocity(Clamp Velocity=false, Limit Acceleration=false)
   &ScaleSpriteSizeBySpeed(Max Scale Factor=<1 1>, Min Scale Factor=<0 0>)
  ]
  rendererClasses [&NiagaraSpriteRendererProperties]
  renderers [{rendererClass=&NiagaraSpriteRendererProperties}]
 }
 Spirals {
  emitterUpdateScript [
   &EmitterState()
   &SpawnBurst_Instantaneous(Spawn Count=2, Use Spawn Probability=false, UseLoopCountLimit=false)
  ]
  particleSpawnScript [
   &InitializeParticle(
    Color=&RandomRangeLinearColor(Minimum=<rgba 0.287 0.7021 1 1>, Maximum=<rgba 0.2776 0.7024 1 0.6>)
    Lifetime Max=0.4
    Lifetime Min=0.4
    Lifetime Mode=::NewEnumerator1
    Mesh Scale Max=<1 1 1>
    Mesh Scale Mode=::NewEnumerator2
    UsePositionOffset=true
    Write Material Random=true
   )
   /SetVariables(Particles.Color=&LerpLinearColors(StartColor=$Particles.Color, EndColor=<rgba 1 0.4932 0.2185 1>, LerpFactor=$User.Blockbreak)) {enabled=false, isSetParameters=true}
   &NMS_Melee_SetColor(
    Block Break:NiagaraFloat=$User.Blockbreak
    Block Break Color:LinearColor=<rgba 1 0.479 0.2185 1>
    Custom Color:LinearColor=$User.Color
    Custom Color Intensity:NiagaraFloat=$User.CustomColorIntensity
    Non Block Break Color:LinearColor=<rgba 0.3 0.1709 0.0417 8>
    UseCustomColor:NiagaraFloat=$User.UseCustomColor
   )
   &InitialMeshOrientation(
    Mesh Axis To Orient=<0 0 1>
    Mesh Orientation Mode=::NewEnumerator2
    Orientation Coordinate Space=::Local
    Orientation Vector=<0 1 0>
    Rotation=&MakeVector(
     X=0
     Y=0
     Z=&ReturnNormalizedExecIndex(Normalized Index Scale=HlslExpression() {hlslExpression="(User.SpiralCount*1.0f-1)/(User.SpiralCount*1.0f)"})
    )
    Use Rotation Vector=true
   ) {module=InitialMeshOrientation001}
   &AddVelocity(Velocity=<0 -687.8561 0>)
   &AddRotationalVelocity(Coordinate Space:ENiagaraMeshTransforms=::NewEnumerator2, Rotation Rate=<0 1 0>)
  ]
  particleUpdateScript [
   &ParticleState()
   &SolveForcesAndVelocity(Clamp Velocity=false, Limit Acceleration=false, Speed Limit=1200) {module=SolveForcesAndVelocity001}
   &SolveRotationalForcesAndVelocity()
   &ScaleColor(
    Scale Alpha=&FloatFromCurve(FloatCurve=DataInterface() {propertyValues {Curve {keys [
         {arriveTangent=-3.7424, interpMode=RCIM_Cubic, leaveTangent=-3.7424, tangentMode=RCTM_User, time=0, value=0}
         {arriveTangent=0.008, interpMode=RCIM_Cubic, leaveTangent=0.008, tangentMode=RCTM_User, time=0.175, value=1}
         {arriveTangent=-0.2882, interpMode=RCIM_Cubic, leaveTangent=-0.2882, tangentMode=RCTM_User, time=0.4, value=0.1418}
         {tangentMode=RCTM_User, time=1, value=0}
        ]}}})
    Scale RGB=&VectorFromFloat(Value=4)
    ScaleA=true
    ScaleRGB=true
   )
   &ScaleMeshSize(Scale Factor=&VectorFromCurve(
     VectorCurve=DataInterface() {propertyValues {
       XCurve {keys [{interpMode=RCIM_Cubic, tangentMode=RCTM_User, time=0, value=0.3}, {interpMode=RCIM_Cubic, tangentMode=RCTM_User, time=1, value=2.2}]}
       YCurve {keys [{interpMode=RCIM_Cubic, tangentMode=RCTM_User, time=0, value=0.3}, {interpMode=RCIM_Cubic, tangentMode=RCTM_User, time=1, value=2.2}]}
       ZCurve {keys [{arriveTangent=0.9, time=0, value=1}, {leaveTangent=0.9, time=1, value=4}]}
      }}
     Scale Curve=&Multiply_VectorByFloat(Vector=<1.5 1.5 1>, Float=0.7)
    ))
   &DynamicMaterialParameters(
    Index 0 Param 1=&RandomRangeFloat(Minimum=0, Maximum=1)
    Index 0 Param 2=2
    Index 0 Param 3=1
    Index 1 Param 1=&FloatFromCurve(
     FloatCurve=DataInterface() {propertyValues {Curve {keys [
         {arriveTangent=-1.3693, interpMode=RCIM_Cubic, leaveTangent=-1.3693, tangentMode=RCTM_User, time=0, value=1}
         {interpMode=RCIM_Cubic, tangentMode=RCTM_User, time=1, value=0.5}
        ]}}}
     Scale Curve=120
    )
    Index 1 Param 2=&FloatFromCurve(
     FloatCurve=DataInterface() {propertyValues {Curve {keys [
         {arriveTangent=-1.4344, interpMode=RCIM_Cubic, leaveTangent=-1.4344, tangentMode=RCTM_User, time=0, value=1}
         {interpMode=RCIM_Cubic, tangentMode=RCTM_User, time=1, value=0.5}
        ]}}}
     Scale Curve=28
    )
    Index 1 Param 3=500
    Module.Param0WriteEnabled=true
    Module.Param0WriteEnabled1=true
    Module.Param1WriteEnabled=true
    Module.Param1WriteEnabled1=true
    Module.Param2WriteEnabled=true
    Module.Param2WriteEnabled1=true
    Module.Param3WriteEnabled=true
    Module.Param3WriteEnabled1=true
    Write Parameter Index 1=true
   ) {module=DynamicMaterialParameters001}
  ]
  rendererClasses [&NiagaraMeshRendererProperties]
  renderers [{rendererClass=&NiagaraMeshRendererProperties, rendererIndex=0}]
 }
 Stars {
  emitterUpdateScript [
   &EmitterState(Life Cycle Mode=::NewEnumerator1, Loop Behavior=::NewEnumerator1, Loop Delay=0.1, Loop Duration=0.5667, UseLoopDelay=true)
   &SpawnBurst_Instantaneous(Spawn Count=8, Use Spawn Probability=false, UseLoopCountLimit=false)
   &SpawnBurst_Instantaneous(Spawn Count=4, Spawn Group=1, Use Spawn Probability=false, UseLoopCountLimit=false) {module=SpawnBurst_Instantaneous001}
   &SpawnRate(SpawnRate=55, Use Spawn Probability=false) {enabled=false}
  ]
  particleSpawnScript [
   &InitializeParticle(
    AdjustAlpha=false
    AdjustHue=true
    AdjustSaturation=false
    AdjustValue=false
    Color=&MaskLinearColorBySpawnGroup(Masked Linear Color=<rgba 5 2.75 0.5 1>, Passthrough Linear Color=<rgba 0.4523 2.2809 5 1>)
    Color Mode=::NewEnumerator3
    Hue Shift Range:Vector2f=<0 0.05>
    Lifetime Max=0.35
    Lifetime Min=0.3
    Lifetime Mode=::NewEnumerator1
    Sprite Rotation Angle=45
    Sprite Rotation Mode=::NewEnumerator0
    Sprite Size Mode=::NewEnumerator1
    Uniform Sprite Size Max=45
    Uniform Sprite Size Min=25
    UsePositionOffset=false
    Write Material Random=true
   )
   &ShapeLocation(MaskSpawnGroup=false, Override Seed=false, Sphere Radius=50, UseOwnerScale=true)
   &AddVelocity(
    Module.Use Falloff Distance=false
    Override Seed=false
    Velocity Mode=::NewEnumerator1
    Velocity Speed=&MaskFloatBySpawnGroup(
     Masked Float=&RandomRangeFloat(Minimum=1000, Maximum=1500)
     Passthrough Float=&RandomRangeFloat(Minimum=768.5549, Maximum=880.3445)
    )
   )
  ]
  particleUpdateScript [
   &CurlNoiseForce(Noise Frequency=15, Noise Strength=&RandomRangeFloat(Minimum=60, Maximum=120), Pan Noise=false) {enabled=false}
   &Drag(Drag=5, Use Linear Drag=true, Use Rotational Drag=true)
   &SolveForcesAndVelocity(Clamp Velocity=false, Limit Acceleration=false)
   &ParticleState()
   &ScaleSpriteSize(
    Scale Sprite Size Mode=::NewEnumerator0
    Uniform Scale Factor=&Waveform(Global Phase=$Particles.NormalizedAge, "Waveform [1]"=::NewEnumerator2, "[1] Frequency"=2, "[1] Amplitude Min/Max"=<0.2 1>)
   )
   &ScaleColor(
    Scale Alpha=&FloatFromCurve(FloatCurve=DataInterface() {propertyValues {Curve {
        defaultValue=1
        keys [
         {interpMode=RCIM_Cubic, tangentMode=RCTM_User, time=0, value=1}
         {arriveTangent=-2.5987, interpMode=RCIM_Cubic, leaveTangent=-2.5987, tangentMode=RCTM_User, time=0.4771, value=0.5344}
         {interpMode=RCIM_Cubic, tangentMode=RCTM_User, time=1, value=0}
        ]
       }}})
    ScaleA=true
    ScaleRGB=true
   )
   &DynamicMaterialParameters(
    Index 0 Param 1=0.5
    Module.Index0Float0Write=true
    Module.Index0Float1Write=true
    Module.Index0Float2Write=true
    Module.Index0Float3Write=true
   )
   /SetVariables(Particles.Glow03_Size=&Multiply_Vector2D_ByFloat(
     Float=&FloatFromCurve(
      FloatCurve=DataInterface() {propertyValues {Curve {keys [
          <at 0 0 RCIM_Cubic>
          <at 0.2 1.3 RCIM_Cubic>
          {arriveTangent=-0.1423, interpMode=RCIM_Cubic, leaveTangent=-0.1423, tangentMode=RCTM_User, time=1, value=0}
         ]}}}
      Scale Curve=&Multiply_Float(A=5, B=2)
     )
     Vector2D=<47.1239 7.5853>
    )) {isSetParameters=true, module=SetVariables_#2}
   /SetVariables(Particles.Glow03_Color=&MakeLinearColorFromVectorAndFloat(
     "Vector (RGB)"=&MakeVectorFromLinearColorRGB(Color=&MultiplyLinearColor(
       Linear Color=&MaskLinearColorBySpawnGroup(Masked Linear Color=<rgba 1.5 0.74 0.075 0.2>, Passthrough Linear Color=<rgba 0.075 1.0465 1.5 0.2>)
       Scale Factor=&MakeVector4_FromFloat(Float=1)
      ))
     "Float (Alpha)"=&Multiply_Float(
      A=&MaskFloatBySpawnGroup(Masked Float=0.2, Passthrough Float=0.1)
      B=&FloatFromCurve(FloatCurve=DataInterface() {propertyValues {Curve {keys [
           {tangentMode=RCTM_User, time=0, value=1}
           {arriveTangent=-0.0405, interpMode=RCIM_Cubic, leaveTangent=-0.0405, tangentMode=RCTM_User, time=0.4, value=1}
           {arriveTangent=0.0761, interpMode=RCIM_Cubic, leaveTangent=0.0761, tangentMode=RCTM_User, time=1, value=0}
          ]}}})
     )
    )) {isSetParameters=true, module=SetVariables_#3}
   /SetVariables(Particles.2DynamicMaterialParameter:Vector4f=<4 1 1 1>) {isSetParameters=true, module=SetVariables_#4}
  ]
  rendererClasses [&NiagaraSpriteRendererProperties]
  renderers [
   {rendererClass=&NiagaraSpriteRendererProperties}
   {rendererClass=&NiagaraSpriteRendererProperties, rendererIndex=1}
  ]
 }
 Stars001 {
  bEnabled=false
  emitterUpdateScript [
   &EmitterState(Life Cycle Mode=::NewEnumerator1, Loop Behavior=::NewEnumerator1, Loop Delay=0.1, Loop Duration=0.5667, UseLoopDelay=true)
   &SpawnBurst_Instantaneous(Spawn Count=4, Use Spawn Probability=false, UseLoopCountLimit=false)
   &SpawnRate(SpawnRate=55, Use Spawn Probability=false) {enabled=false}
  ]
  particleSpawnScript [
   &InitializeParticle(
    AdjustAlpha=false
    AdjustHue=true
    AdjustSaturation=false
    AdjustValue=false
    Color=<rgba 0.4523 2.2809 5 1>
    Color Mode=::NewEnumerator3
    Hue Shift Range:Vector2f=<0 0.05>
    Lifetime Max=0.35
    Lifetime Min=0.3
    Lifetime Mode=::NewEnumerator1
    Sprite Rotation Angle=45
    Sprite Rotation Mode=::NewEnumerator0
    Sprite Size Mode=::NewEnumerator1
    Uniform Sprite Size Max=45
    Uniform Sprite Size Min=25
    UsePositionOffset=false
    Write Material Random=true
   )
   &ShapeLocation(MaskSpawnGroup=false, Override Seed=false, Sphere Radius=50, UseOwnerScale=true)
   &AddVelocity(
    Module.Use Falloff Distance=false
    Override Seed=false
    Velocity Mode=::NewEnumerator1
    Velocity Speed=&RandomRangeFloat(Minimum=768.5549, Maximum=880.3445)
   )
  ]
  particleUpdateScript [
   &CurlNoiseForce(Noise Frequency=15, Noise Strength=&RandomRangeFloat(Minimum=60, Maximum=120), Pan Noise=false) {enabled=false}
   &Drag(Drag=5, Use Linear Drag=true, Use Rotational Drag=true)
   &SolveForcesAndVelocity(Clamp Velocity=false, Limit Acceleration=false)
   &ParticleState()
   &ScaleSpriteSize(
    Scale Sprite Size Mode=::NewEnumerator0
    Uniform Scale Factor=&Waveform(Global Phase=$Particles.NormalizedAge, "Waveform [1]"=::NewEnumerator2, "[1] Frequency"=2, "[1] Amplitude Min/Max"=<0.2 1>)
   )
   &ScaleColor(
    Scale Alpha=&FloatFromCurve(FloatCurve=DataInterface() {propertyValues {Curve {
        defaultValue=1
        keys [
         {interpMode=RCIM_Cubic, tangentMode=RCTM_User, time=0, value=1}
         {arriveTangent=-2.5987, interpMode=RCIM_Cubic, leaveTangent=-2.5987, tangentMode=RCTM_User, time=0.4771, value=0.5344}
         {interpMode=RCIM_Cubic, tangentMode=RCTM_User, time=1, value=0}
        ]
       }}})
    ScaleA=true
    ScaleRGB=true
   )
   &DynamicMaterialParameters(
    Index 0 Param 1=0.5
    Module.Index0Float0Write=true
    Module.Index0Float1Write=true
    Module.Index0Float2Write=true
    Module.Index0Float3Write=true
   )
   /SetVariables(Particles.Glow03_Size=&Multiply_Vector2D_ByFloat(
     Float=&FloatFromCurve(
      FloatCurve=DataInterface() {propertyValues {Curve {keys [
          <at 0 0 RCIM_Cubic>
          <at 0.2 1.3 RCIM_Cubic>
          {arriveTangent=-0.1423, interpMode=RCIM_Cubic, leaveTangent=-0.1423, tangentMode=RCTM_User, time=1, value=0}
         ]}}}
      Scale Curve=&Multiply_Float(A=5, B=2)
     )
     Vector2D=<47.1239 7.5853>
    )) {isSetParameters=true, module=SetVariables_#2}
   /SetVariables(Particles.Glow03_Color=&MakeLinearColorFromVectorAndFloat(
     "Vector (RGB)"=&MakeVectorFromLinearColorRGB(Color=&MultiplyLinearColor(Linear Color=<rgba 0.075 1.0465 1.5 0.2>, Scale Factor=&MakeVector4_FromFloat(Float=1)))
     "Float (Alpha)"=&Multiply_Float(
      A=0.1
      B=&FloatFromCurve(FloatCurve=DataInterface() {propertyValues {Curve {keys [
           {tangentMode=RCTM_User, time=0, value=1}
           {arriveTangent=-0.0405, interpMode=RCIM_Cubic, leaveTangent=-0.0405, tangentMode=RCTM_User, time=0.4, value=1}
           {arriveTangent=0.0761, interpMode=RCIM_Cubic, leaveTangent=0.0761, tangentMode=RCTM_User, time=1, value=0}
          ]}}})
     )
    )) {isSetParameters=true, module=SetVariables_#3}
   /SetVariables(Particles.2DynamicMaterialParameter:Vector4f=<4 1 1 1>) {isSetParameters=true, module=SetVariables_#4}
  ]
  rendererClasses [&NiagaraSpriteRendererProperties]
  renderers [
   {rendererClass=&NiagaraSpriteRendererProperties}
   {rendererClass=&NiagaraSpriteRendererProperties, rendererIndex=1}
  ]
 }
 WindCone {
  emitterUpdateScript [
   &EmitterState()
   &SpawnBurst_Instantaneous(Spawn Time=0.03, Use Spawn Probability=false, UseLoopCountLimit=false)
   &SpawnBurst_Instantaneous(Spawn Group=1, Spawn Time=0.03, Use Spawn Probability=false, UseLoopCountLimit=false) {module=SpawnBurst_Instantaneous001}
  ]
  particleSpawnScript [
   &InitializeParticle(
    Color=&MaskLinearColorBySpawnGroup(Masked Linear Color=<rgba 1.5 0.8239 0.1479 3>, Passthrough Linear Color=<rgba 0.2957 1.5022 3 1.5>)
    Lifetime=0.4
    Mesh Scale=&MaskVectorBySpawnGroup(Masked Vector=<2 1.15 1.15>, Passthrough Vector=<1 1 1>)
    Mesh Scale Mode=::NewEnumerator4
    Position Offset=<0 70 0>
    Sprite Size Mode=::NewEnumerator1
    Uniform Sprite Size Max=150
    Uniform Sprite Size Min=100
    UsePositionOffset=true
    Write Material Random=true
   ) {module=InitializeParticle001}
   &DynamicMaterialParameters(
    Index 0 Param 1=1
    Index 0 Param 2=0.6
    Index 0 Param 3=2
    Module.Param0WriteEnabled=true
    Module.Param1WriteEnabled=true
    Module.Param2WriteEnabled=true
    Module.Param3WriteEnabled=false
   )
   &InitialMeshOrientation(Mesh Orientation Mode=::NewEnumerator2, Orientation Vector=<0 1 0>, Use Rotation Vector=false)
   &CameraOffset()
  ]
  particleUpdateScript [
   &ParticleState()
   &ScaleMeshSize(Scale Factor=&VectorFromCurve(VectorCurve=DataInterface() {propertyValues {
       XCurve {keys [
         {arriveTangent=1.5385, time=0, value=0}
         {arriveTangent=0.3483, interpMode=RCIM_Cubic, leaveTangent=0.3483, tangentMode=RCTM_User, time=1, value=1}
        ]}
       YCurve {keys [
         {arriveTangent=2.3109, interpMode=RCIM_Cubic, leaveTangent=2.3109, tangentMode=RCTM_User, time=0, value=0}
         {arriveTangent=0.6124, interpMode=RCIM_Cubic, leaveTangent=0.6124, tangentMode=RCTM_User, time=1, value=1}
        ]}
       ZCurve {keys [
         {arriveTangent=2.3109, interpMode=RCIM_Cubic, leaveTangent=2.3109, tangentMode=RCTM_User, time=0, value=0}
         {arriveTangent=0.6124, interpMode=RCIM_Cubic, leaveTangent=0.6124, tangentMode=RCTM_User, time=1, value=1}
        ]}
      }}))
   &ScaleColor(
    Scale Alpha=&FloatFromCurve(FloatCurve=DataInterface() {propertyValues {Curve {keys [
         {arriveTangent=10809.126, interpMode=RCIM_Cubic, time=0, value=0}
         <at 0.15 1 RCIM_Cubic>
         {interpMode=RCIM_Cubic, leaveTangent=-3.1778, time=1, value=0}
        ]}}})
    Scale RGB=&MakeVectorFromLinearColorRGB(Color=&ColorFromCurve(ColorCurve=DataInterface() {propertyValues {
        AlphaCurve {keys []}
        BlueCurve {keys [<at 0.1026 1 RCIM_Cubic>,<at 0.959 0.321 RCIM_Cubic>]}
        GreenCurve {keys [<at 0.1026 1 RCIM_Cubic>,<at 0.959 0.4824 RCIM_Cubic>]}
        RedCurve {keys [<at 0.1026 1 RCIM_Cubic>,<at 0.959 0.5 RCIM_Cubic>]}
       }}))
    ScaleA=true
    ScaleRGB=true
   )
   &DynamicMaterialParameters(
    Index 1 Param 1=&FloatFromCurve(FloatCurve=DataInterface() {propertyValues {Curve {keys [
         {arriveTangent=2.4188, interpMode=RCIM_Cubic, leaveTangent=2.4188, tangentMode=RCTM_User, time=0, value=-0.25}
         {arriveTangent=0.444, interpMode=RCIM_Cubic, leaveTangent=0.444, tangentMode=RCTM_User, time=1, value=1}
        ]}}})
    Index 1 Param 2=&FloatFromCurve(FloatCurve=DataInterface() {propertyValues {Curve {keys [<at 0.05 1 RCIM_Cubic>,<at 0.25 0 RCIM_Cubic>]}}})
    Module.Param0WriteEnabled1=true
    Module.Param1WriteEnabled1=true
    Module.Param2WriteEnabled1=false
    Module.Param3WriteEnabled1=false
    Write Parameter Index 0=false
    Write Parameter Index 1=true
   ) {module=DynamicMaterialParameters001}
  ]
  rendererClasses [&NiagaraMeshRendererProperties]
  renderers [{rendererClass=&NiagaraMeshRendererProperties, rendererIndex=0}]
 }
 WindCone001 {bEnabled=false, rendererClasses [], renderers []}
 WindCone002 {
  bEnabled=false
  emitterUpdateScript [
   &EmitterState()
   &SpawnBurst_Instantaneous(Spawn Time=0.03, Use Spawn Probability=false, UseLoopCountLimit=false)
  ]
  particleSpawnScript [
   &InitializeParticle(
    Color=<rgba 0.2957 1.5022 3 1.5>
    Lifetime=0.4
    Mesh Scale Mode=::NewEnumerator4
    Position Offset=<0 70 0>
    Sprite Size Mode=::NewEnumerator1
    Uniform Sprite Size Max=150
    Uniform Sprite Size Min=100
    UsePositionOffset=true
    Write Material Random=true
   ) {module=InitializeParticle001}
   &DynamicMaterialParameters(
    Index 0 Param 1=1
    Index 0 Param 2=0.6
    Index 0 Param 3=2
    Module.Param0WriteEnabled=true
    Module.Param1WriteEnabled=true
    Module.Param2WriteEnabled=true
    Module.Param3WriteEnabled=false
   )
   &InitialMeshOrientation(Mesh Orientation Mode=::NewEnumerator2, Orientation Vector=<0 1 0>, Use Rotation Vector=false)
   &CameraOffset()
  ]
  particleUpdateScript [
   &ParticleState()
   &ScaleMeshSize(Scale Factor=&VectorFromCurve(VectorCurve=DataInterface() {propertyValues {
       XCurve {keys [
         {arriveTangent=1.5385, time=0, value=0}
         {arriveTangent=0.3483, interpMode=RCIM_Cubic, leaveTangent=0.3483, tangentMode=RCTM_User, time=1, value=1}
        ]}
       YCurve {keys [
         {arriveTangent=2.3109, interpMode=RCIM_Cubic, leaveTangent=2.3109, tangentMode=RCTM_User, time=0, value=0}
         {arriveTangent=0.6124, interpMode=RCIM_Cubic, leaveTangent=0.6124, tangentMode=RCTM_User, time=1, value=1}
        ]}
       ZCurve {keys [
         {arriveTangent=2.3109, interpMode=RCIM_Cubic, leaveTangent=2.3109, tangentMode=RCTM_User, time=0, value=0}
         {arriveTangent=0.6124, interpMode=RCIM_Cubic, leaveTangent=0.6124, tangentMode=RCTM_User, time=1, value=1}
        ]}
      }}))
   &ScaleColor(
    Scale Alpha=&FloatFromCurve(FloatCurve=DataInterface() {propertyValues {Curve {keys [
         {arriveTangent=10809.126, interpMode=RCIM_Cubic, time=0, value=0}
         <at 0.15 1 RCIM_Cubic>
         {interpMode=RCIM_Cubic, leaveTangent=-3.1778, time=1, value=0}
        ]}}})
    Scale RGB=&MakeVectorFromLinearColorRGB(Color=&ColorFromCurve(ColorCurve=DataInterface() {propertyValues {
        AlphaCurve {keys []}
        BlueCurve {keys [<at 0.1026 1 RCIM_Cubic>,<at 0.959 0.321 RCIM_Cubic>]}
        GreenCurve {keys [<at 0.1026 1 RCIM_Cubic>,<at 0.959 0.4824 RCIM_Cubic>]}
        RedCurve {keys [<at 0.1026 1 RCIM_Cubic>,<at 0.959 0.5 RCIM_Cubic>]}
       }}))
    ScaleA=true
    ScaleRGB=true
   )
   &DynamicMaterialParameters(
    Index 1 Param 1=&FloatFromCurve(FloatCurve=DataInterface() {propertyValues {Curve {keys [
         {arriveTangent=2.4188, interpMode=RCIM_Cubic, leaveTangent=2.4188, tangentMode=RCTM_User, time=0, value=-0.25}
         {arriveTangent=0.444, interpMode=RCIM_Cubic, leaveTangent=0.444, tangentMode=RCTM_User, time=1, value=1}
        ]}}})
    Index 1 Param 2=&FloatFromCurve(FloatCurve=DataInterface() {propertyValues {Curve {keys [<at 0.05 1 RCIM_Cubic>,<at 0.25 0 RCIM_Cubic>]}}})
    Module.Param0WriteEnabled1=true
    Module.Param1WriteEnabled1=true
    Module.Param2WriteEnabled1=false
    Module.Param3WriteEnabled1=false
    Write Parameter Index 0=false
    Write Parameter Index 1=true
   ) {module=DynamicMaterialParameters001}
  ]
  rendererClasses [&NiagaraMeshRendererProperties]
  renderers [{rendererClass=&NiagaraMeshRendererProperties, rendererIndex=0}]
 }
 particle001 {
  bEnabled=false
  emitterUpdateScript [
   &EmitterState(Life Cycle Mode=::NewEnumerator1, Loop Behavior=::NewEnumerator1, Loop Delay=0.05, Loop Duration=0.16, UseLoopDelay=true)
   &SpawnBurst_Instantaneous(Spawn Count=29, Use Spawn Probability=false, UseLoopCountLimit=false)
   &SpawnRate(SpawnRate=160, Use Spawn Probability=false)
  ]
  particleSpawnScript [
   &InitializeParticle(
    Color Maximum=<rgba 6 3 0 1>
    Color Minimum=<rgba 1 0.2893 0.0435 1>
    Color Mode=::NewEnumerator2
    Lifetime=&FloatFromCurve(
     FloatCurve=DataInterface() {propertyValues {Curve {keys [
         <at 0 0.3 RCIM_Cubic>
         {arriveTangent=1.2573, interpMode=RCIM_Cubic, leaveTangent=1.2573, tangentMode=RCTM_User, time=0.9, value=0.65}
         {arriveTangent=1.733, tangentMode=RCTM_User, time=1, value=1.3}
        ]}}}
     CurveIndex=&RandomRangeFloat(Minimum=0, Maximum=1)
    )
    Sprite Size Max=<0.15 2.1>
    Sprite Size Min=<0.15 1.5>
    Sprite Size Mode=::NewEnumerator2
    UsePositionOffset=false
    Write Material Random=true
   )
   &ShapeLocation(
    Cylinder Height=15
    Cylinder Radius=50
    MaskSpawnGroup=false
    Offset=<0 -20 0>
    Offset Coordinate Space=::Local
    Offset Mode=::NewEnumerator0
    Override Seed=false
    Rotation Axis=<0 1 0>
    Rotation Mode=::NewEnumerator1
    Shape Primitive=::NewEnumerator1
    Surface Only=false
    UseOwnerScale=true
   )
   &AddVelocity(Velocity=<0 -1 0>, Velocity Speed Scale=&RandomRangeFloat(Minimum=1749.4241, Maximum=2798.3591))
   &AddVelocity(
    Module.Use Falloff Distance=false
    Override Seed=false
    Velocity Mode=::NewEnumerator1
    Velocity Speed=&RandomRangeFloat(Minimum=1036.0082, Maximum=2407.9128)
   ) {module=AddVelocity001}
   &InheritVelocity(Speed Limit=true, UseSpeedThreshold=false) {enabled=false}
  ]
  particleUpdateScript [
   &ParticleState()
   &ScaleSpriteSize(Scale Sprite Size Mode=::NewEnumerator0, Uniform Scale Factor=10)
   &ScaleSpriteSize(
    Scale Sprite Size Mode=::NewEnumerator0
    Uniform Scale Factor=&FloatFromCurve(FloatCurve=DataInterface() {propertyValues {Curve {
        defaultValue=1
        keys [
         {arriveTangent=9.0725, interpMode=RCIM_Cubic, leaveTangent=9.0725, tangentMode=RCTM_User, time=0, value=0}
         {arriveTangent=-2.5249, interpMode=RCIM_Cubic, leaveTangent=-2.5249, tangentMode=RCTM_User, time=0.1408, value=0.9847}
         {arriveTangent=-0.0405, interpMode=RCIM_Cubic, leaveTangent=-0.0405, tangentMode=RCTM_User, time=1, value=0}
        ]
       }}})
   ) {module=ScaleSpriteSize001}
   &CurlNoiseForce(
    Bias Noise=false
    Noise Strength=&FloatFromCurve(FloatCurve=DataInterface() {propertyValues {Curve {keys [
         {arriveTangent=4732.4746, interpMode=RCIM_Cubic, leaveTangent=4732.4746, tangentMode=RCTM_User, time=0, value=0}
         {arriveTangent=81.4476, tangentMode=RCTM_User, time=1, value=2000}
        ]}}})
    Pan Noise=true
    Pan Noise Field=&TransformVector(Vector=<0 1 0>, Destination Space=::Simulation)
   )
   &Drag(Drag=3, Use Linear Drag=true, Use Rotational Drag=true)
   &VortexForce(
    Origin Pull Amount=4000
    UseOriginOffset=false
    Vortex Axis=<1 0 0>
    Vortex Axis Coordinate Space:ENiagaraCoordinateSpace=::Local
    Vortex Force Amount=7000
   ) {enabled=false}
   &SolveForcesAndVelocity(
    Clamp Velocity=true
    Limit Acceleration=false
    Speed Limit:NiagaraFloat=&FloatFromCurve(
     FloatCurve=DataInterface() {propertyValues {Curve {keys [
         {interpMode=RCIM_Cubic, tangentMode=RCTM_User, time=0, value=1}
         {arriveTangent=-0.9, interpMode=RCIM_Cubic, leaveTangent=-0.9, tangentMode=RCTM_User, time=0.7, value=0.2}
         {interpMode=RCIM_Cubic, tangentMode=RCTM_User, time=1, value=0.07}
        ]}}}
     Scale Curve=1000
    )
   )
   &ScaleSpriteSizeBySpeed(Max Scale Factor=<1 2>, Min Scale Factor=<0.5 0.5>)
   &FN_Sprite_Attributes(
    Color=&RandomRangeLinearColor(Minimum=<rgba 1 0.8402 0.1554 0.25>, Maximum=<rgba 1 0.7084 0.1255 0.25>)
    Create Color Attribute=true
    Create SpriteSize Attribute=true
    SpriteSize=&Multiply_Vector2D(A=$Particles.SpriteSize, B=<20 4>)
    Write to Module Namespace=true
   )
   &DynamicMaterialParameters(
    Index 0 Param 1=2
    Module.Index0Param0Write=true
    Module.Index0Param1Write=true
    Module.Index0Param2Write=true
    Module.Index0Param3Write=true
   )
  ]
  rendererClasses [&NiagaraSpriteRendererProperties]
  renderers [
   {rendererClass=&NiagaraSpriteRendererProperties}
   {rendererClass=&NiagaraSpriteRendererProperties, rendererIndex=1}
  ]
 }
}
summary {
 systemName=NS_DentureSay_AirDash
 emitters [
  {bEnabled=false, emitterName=Speedlines, rendererClasses [&NiagaraSpriteRendererProperties]}
  {bEnabled=false, emitterName=WindCone001, rendererClasses []}
  {bEnabled=false, emitterName=particle001, rendererClasses [&NiagaraSpriteRendererProperties]}
  {bEnabled=false, emitterName=Shock_Cone, rendererClasses [&NiagaraMeshRendererProperties]}
  {emitterName=Spirals, rendererClasses [&NiagaraMeshRendererProperties]}
  {emitterName=PP, rendererClasses [&NiagaraComponentRendererProperties]}
  {bEnabled=false, emitterName=ExhaustFlame001, rendererClasses [&NiagaraMeshRendererProperties]}
  {emitterName=Endflare, rendererClasses []}
  {emitterName=FlashMesh001, rendererClasses []}
  {emitterName=FlashMesh002, rendererClasses []}
  {emitterName=FlashMesh004, rendererClasses []}
  {emitterName=Light, rendererClasses []}
  {emitterName=WindCone, rendererClasses [&NiagaraMeshRendererProperties]}
  {emitterName=Stars, rendererClasses [&NiagaraSpriteRendererProperties]}
  {bEnabled=false, emitterName=Stars001, rendererClasses [&NiagaraSpriteRendererProperties]}
  {bEnabled=false, emitterName=WindCone002, rendererClasses [&NiagaraMeshRendererProperties]}
  {emitterName=Shock_Cone_Lightweight, rendererClasses []}
  {emitterName=ExhaustFlame001_Lightweight, rendererClasses []}
 ]
 userVariables [
  User.Blockbreak {type=NiagaraFloat, defaultValue=0, description=""}
  User.Color {type=LinearColor, description="", defaultValue {struct=LinearColor, value=<rgba 0.2287 1 0 1>}}
  User.CustomColorIntensity {type=NiagaraFloat, defaultValue=1, description=""}
  User.HasQuadExhaust {type=NiagaraBool, defaultValue=true}
  User.LocalPlayer {type=NiagaraBool}
  User.SpiralCount {type=NiagaraInt32, defaultValue=3}
  User.UseCustomColor {type=NiagaraFloat, defaultValue=0, description=""}
 ]
}
