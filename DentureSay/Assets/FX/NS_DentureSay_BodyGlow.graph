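# Niagara system for the DentureSay body glow, a brief flash drawn through sprite and mesh renderers.
# A single emitter burst-spawns particles with lifetime from GlowDuration, but they are not killed when it elapses.
# ScaleColor multiplies RGB by GlowStrength and drives alpha with a sharp spike-and-fade curve indexed by system age, not particle age.
# DisableAlphaCurve blends that curve toward constant full alpha; GlowThickness is exposed but unreferenced in the listed modules.
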NiagaraSystem /DentureSay/Assets/FX/NS_DentureSay_BodyGlow
dependencies {usedDataInterfaces [&NiagaraDataInterfaceCurve], usedDynamicInputs [], usedModules [&SystemState]}
emitters {Minimal {
  emitterUpdateScript [
   &EmitterState(Inactive Response=::NewEnumerator1, Life Cycle Mode=::NewEnumerator1, Loop Behavior=::NewEnumerator1, UseLoopDelay=true)
   &SpawnBurst_Instantaneous(Use Spawn Probability=false, UseLoopCountLimit=false)
  ]
  particleSpawnScript [
   &InitializeParticle(
    Color Mode=::NewEnumerator0
    Lifetime=$User.GlowDuration
    Mesh Scale Mode=::NewEnumerator3
    Mesh Uniform Scale=2
    UsePositionOffset=true
    Write Material Random=true
   )
   &PartitionParticles()
   &CameraOffset(Camera Offset Amount=10)
  ]
  particleUpdateScript [
   &ParticleState(Kill Particles When Lifetime Has Elapsed=false, Let Infinitely Lived Particles Die When Emitter Deactivates=true)
   &ScaleColor(
    Scale Alpha=&Lerp_Float(
     A=&FloatFromCurve(
      FloatCurve=DataInterface() {propertyValues {Curve {keys [
          {interpMode=RCIM_Cubic, leaveTangent=14.1474, tangentMode=RCTM_Break, time=0, value=0}
          {interpMode=RCIM_Cubic, tangentMode=RCTM_Break, time=0.0625, value=0.9233}
          {arriveTangent=-1.0044, interpMode=RCIM_Cubic, leaveTangent=-1.0044, tangentMode=RCTM_User, time=0.1593, value=0.1846}
          {interpMode=RCIM_Cubic, tangentMode=RCTM_Break, time=0.3248, value=0}
         ]}}}
      CurveIndex=$Engine.System.Age
     )
     B=1
     Alpha=$User.DisableAlphaCurve
    )
    Scale RGB=&VectorFromFloat(Value=$User.GlowStrength)
    ScaleA=true
    ScaleRGB=true
   )
  ]
  rendererClasses [&NiagaraSpriteRendererProperties, &NiagaraMeshRendererProperties]
  renderers [
   {rendererClass=&NiagaraSpriteRendererProperties}
   {rendererClass=&NiagaraMeshRendererProperties, rendererIndex=1}
  ]
 }}
summary {
 systemName=NS_DentureSay_BodyGlow
 emitters [{emitterName=Minimal, rendererClasses [&NiagaraSpriteRendererProperties, &NiagaraMeshRendererProperties]}]
 userVariables [
  User.DisableAlphaCurve {type=NiagaraFloat, defaultValue=0, description=""}
  User.GlowDuration {type=NiagaraFloat, defaultValue=1.5, description=""}
  User.GlowStrength {type=NiagaraFloat, defaultValue=2, description=""}
  User.GlowThickness {type=NiagaraFloat, defaultValue=4, description=""}
 ]
}
