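# Niagara system for a jab attack effect, a short orange burst thrown mostly along +Y.
# Enabled emitters burst fast JabLines sprites, GPU Stars slowed by drag, and a Minimal glow sprite with particle light.
# User SlashSize and ScaleLifetime scale that glow's size, light radius and lifetime.
# Slash meshes, swirls, streaks, energy wave and flare-with-light are disabled; Endflare and the Lightweight emitters list no modules or renderers.
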NiagaraSystem /DentureSay/Assets/FX/NS_DentureSay_JabFX_AnimTrail
dependencies {
 usedDataInterfaces [
  &NiagaraDataInterfaceColorCurve
  &NiagaraDataInterfaceCurve
  &NiagaraDataInterfaceVector2DCurve
  &NiagaraDataInterfaceArrayFloat
 ]
 usedDynamicInputs [&FN_LOD_Distance_Remap]
 usedModules [
  &SystemState
  &self:SystemSpawnScript.SystemScriptSource.SystemScriptGraph.NiagaraNodeAssignment_1.SetVariables_#1
  &self:SystemSpawnScript.SystemScriptSource.SystemScriptGraph.NiagaraNodeAssignment_0.SetVariables_#2
  &self:SystemSpawnScript.SystemScriptSource.SystemScriptGraph.NiagaraNodeAssignment_3.SetVariables_#3
  &self:SystemSpawnScript.SystemScriptSource.SystemScriptGraph.NiagaraNodeAssignment_2.SetVariables_#4
 ]
}
emitters {
 Endflare {rendererClasses [], renderers []}
 Flare_And_Light {
  bEnabled=false
  emitterUpdateScript [
   &EmitterState(Life Cycle Mode=::NewEnumerator1, Loop Behavior=::NewEnumerator1, Loop Duration=1, UseLoopDelay=false)
   &SpawnBurst_Instantaneous(Spawn Count=1, Use Spawn Probability=true)
  ]
  particleSpawnScript [
   &InitializeParticle(
    Color=<rgba 0.9551 0.98 1 1>
    Lifetime=0.3
    Sprite Size=<216.399 2360.8882>
    Sprite Size Mode=::NewEnumerator4
    UsePositionOffset=false
    Write Material Random=true
   )
   &DynamicMaterialParameters(
    Index 0 Param 1=1
    Module.Param0WriteEnabled=true
    Module.Param1WriteEnabled=true
    Module.Param2WriteEnabled=true
    Module.Param3WriteEnabled=true
   )
   &CameraOffset(Camera Offset Amount=$User.CameraO)
   &SystemLocation(MaskSpawnGroup=false, Offset=<5 200 0>)
   &AddVelocity(Velocity=<0 1 0>, Velocity Speed Scale=1614.224)
  ]
  particleUpdateScript [
   &ParticleState()
   &ScaleColor(
    Linear Color Curve=DataInterface() {propertyValues {
      AlphaCurve {keys [<at 0.0084 0>,<at 0.0821 0.5>,<at 1 0>]}
      BlueCurve {keys [<at 0.1563 0.0297>,<at 0.3995 0.0078>]}
      GreenCurve {keys [<at 0.1563 0.4502>,<at 0.3995 0.537>]}
      RedCurve {keys [<at 0.1563 1>,<at 0.3995 1>]}
     }}
    Scale Mode=::NewEnumerator2
   )
   &ScaleColor(
    Scale Alpha=&FloatFromCurve(
     FloatCurve=DataInterface() {propertyValues {Curve {keys [
         {arriveTangent=7.8189, interpMode=RCIM_Cubic, leaveTangent=7.8189, tangentMode=RCTM_User, time=0, value=0.2}
         {arriveTangent=0.0168, interpMode=RCIM_Cubic, leaveTangent=0.0168, tangentMode=RCTM_User, time=0.12, value=1}
         {arriveTangent=-1.1364, interpMode=RCIM_Cubic, leaveTangent=-1.1364, time=0.5, value=0.8}
         <at 1 0 RCIM_Cubic>
        ]}}}
     Scale Curve=&Multiply_Float(A=1, B=&FNTimeOfDayScalar(Night Scale=0.2, Morning Scale=0.5, Evening Scale=0.5))
    )
    Scale RGB=&VectorFromFloat(Value=&Lerp_Float(
      A=1
      B=2
      Alpha=&MakeFloatFromVector(
       Vector=&VectorFromFloat(Value=&FloatFromCurve(FloatCurve=DataInterface() {propertyValues {Curve {
            defaultValue=1
            keys [
             {arriveTangent=9.0725, interpMode=RCIM_Cubic, leaveTangent=9.0725, tangentMode=RCTM_User, time=0, value=0}
             {arriveTangent=-2.5249, interpMode=RCIM_Cubic, leaveTangent=-2.5249, tangentMode=RCTM_User, time=0.1, value=0.9}
             {arriveTangent=-2.25, interpMode=RCIM_Cubic, leaveTangent=-2.25, time=0.2405, value=0.3518}
             {arriveTangent=-0.0405, interpMode=RCIM_Cubic, leaveTangent=-0.0405, tangentMode=RCTM_User, time=0.5, value=0}
            ]
           }}}))
       Channel=::NewEnumerator0
      )
     ))
    ScaleA=true
    ScaleRGB=true
   ) {module=ScaleColor002}
   &FN_Light_Attributes(
    Color=$Particles.Color
    Create Color Attribute=true
    Create Radius Attribute=true
    Radius=200
    Scale Color=&Multiply_Float(
     A=&FloatFromCurve(FloatCurve=DataInterface() {propertyValues {Curve {keys [
          {arriveTangent=9.0725, interpMode=RCIM_Cubic, leaveTangent=9.0725, tangentMode=RCTM_User, time=0, value=0}
          {arriveTangent=-2.5249, interpMode=RCIM_Cubic, leaveTangent=-2.5249, tangentMode=RCTM_User, time=0.1408, value=0.9847}
          {arriveTangent=-0.0405, interpMode=RCIM_Cubic, leaveTangent=-0.0405, tangentMode=RCTM_User, time=1, value=0}
         ]}}})
     B=20
    )
   )
   &ApplyOwnerScaleToAttributes(Scale Initial Sprite Size=true)
   &SolveForcesAndVelocity(Clamp Velocity=false, Limit Acceleration=false)
  ]
  rendererClasses [&NiagaraSpriteRendererProperties, &NiagaraLightRendererProperties]
  renderers [
   {rendererClass=&NiagaraSpriteRendererProperties}
   {rendererClass=&NiagaraLightRendererProperties, rendererIndex=1}
  ]
 }
 Flare_And_Light_Lightweight {rendererClasses [], renderers []}
 JabLines {
  emitterUpdateScript [
   &EmitterState()
   &SpawnBurst_Instantaneous(Spawn Count=12, Use Spawn Probability=false, UseLoopCountLimit=false)
  ]
  particleSpawnScript [
   &InitializeParticle(
    Color Mode=::NewEnumerator0
    Lifetime Max=0.4
    Lifetime Min=0.2
    Lifetime Mode=::NewEnumerator1
    Sprite Size Max=<3 200>
    Sprite Size Min=<2 40>
    Sprite Size Mode=::NewEnumerator2
    UsePositionOffset=false
    Write Material Random=true
   )
   &AddVelocity(Velocity=<0 4319.8638 0>, Velocity Speed Scale=&RandomRangeFloat(Minimum=0.1, Maximum=1))
   &ApplyInitialForces()
   &ShapeLocation(MaskSpawnGroup=false, Override Seed=false, Sphere Radius=149.1215, UseOwnerScale=true)
  ]
  particleUpdateScript [
   &ParticleState()
   &ScaleSpriteSize(Uniform Curve Sprite Scale=DataInterface() {propertyValues {Curve {
       defaultValue=0
       keys [
        {arriveTangent=-2.3332, interpMode=RCIM_Cubic, leaveTangent=-2.3332, tangentMode=RCTM_User, time=0, value=1.5}
        {interpMode=RCIM_Cubic, tangentMode=RCTM_User, time=1, value=0}
       ]
      }}})
   &Drag(Drag=&RandomRangeFloat(Minimum=2, Maximum=4), Use Linear Drag=true, Use Rotational Drag=true)
   &SolveForcesAndVelocity(Clamp Velocity=false, Limit Acceleration=false)
   &Color(Color=&ColorFromCurve(ColorCurve=DataInterface() {propertyValues {
       AlphaCurve {keys [<at 0 1 RCIM_Cubic>,<at 1 0 RCIM_Cubic>]}
       BlueCurve {keys [<at 0 30 RCIM_Cubic>,<at 0.4878 0.3 RCIM_Cubic>]}
       GreenCurve {keys [<at 0 120.6667 RCIM_Cubic>,<at 0.4878 1.0367 RCIM_Cubic>]}
       RedCurve {keys [<at 0 200 RCIM_Cubic>,<at 0.4878 2 RCIM_Cubic>]}
      }}))
   &FN_Sprite_Attributes(
    Color=&MultiplyLinearColor(
     Linear Color=<rgba 1 0.52 0.1 0.2>
     Scale Factor=&MakeVector4(
      X=1
      Y=1
      Z=1
      W=&FloatFromCurve(FloatCurve=DataInterface() {propertyValues {Curve {keys [
           {interpMode=RCIM_Cubic, tangentMode=RCTM_Break, time=0, value=1}
           {arriveTangent=-2.4802, interpMode=RCIM_Cubic, leaveTangent=-2.4802, tangentMode=RCTM_User, time=0.4552, value=0.3333}
           <at 1 0>
          ]}}})
     )
    )
    Create Color Attribute=true
    Create SpriteSize Attribute=true
    SpriteSize=&Multiply_Vector2D_ByFloat(Float=4.1945, Vector2D=<13.2684 83.2718>)
    Write to Module Namespace=true
   )
   /SetVariables(Particles.Vector 4:Vector4f=<6 0 0 0>) {isSetParameters=true}
  ]
  rendererClasses [&NiagaraSpriteRendererProperties]
  renderers [
   {rendererClass=&NiagaraSpriteRendererProperties}
   {rendererClass=&NiagaraSpriteRendererProperties, rendererIndex=1}
  ]
 }
 Minimal {
  emitterUpdateScript [
   &EmitterState(Life Cycle Mode=::NewEnumerator1, Loop Behavior=::NewEnumerator1, Loop Duration=1, UseLoopDelay=false)
   &SpawnBurst_Instantaneous(Use Spawn Probability=false, UseLoopCountLimit=false)
  ]
  particleSpawnScript [
   &InitializeParticle(
    Color=<rgba 2.5 1.3522 0.2043 0.3>
    Lifetime=&Multiply_Float(A=0.5, B=$User.ScaleLifetime)
    Sprite Size=&Multiply_Vector2D_ByFloat(Float=$User.SlashSize, Vector2D=<1740.2727 333.6171>)
    Sprite Size Mode=::NewEnumerator4
    UsePositionOffset=false
    Write Material Random=true
   )
   &AddVelocity(Velocity=<0 4869.9277 50>)
  ]
  particleUpdateScript [
   &ParticleState()
   &ScaleColor(
    Scale Alpha=&FloatFromCurve(
     FloatCurve=DataInterface() {propertyValues {Curve {
        defaultValue=1
        keys [
         {arriveTangent=9.4247, interpMode=RCIM_Cubic, leaveTangent=9.4247, tangentMode=RCTM_User, time=0, value=0}
         {arriveTangent=-1.0597, interpMode=RCIM_Cubic, leaveTangent=-1.0597, tangentMode=RCTM_User, time=0.3874, value=1.0332}
         {arriveTangent=-1.5111, tangentMode=RCTM_User, time=0.6786, value=0}
        ]
       }}}
     Scale Curve=0.776
    )
    ScaleA=true
    ScaleRGB=true
   )
   &DynamicMaterialParameters(
    Index 0 Param 1=1
    Index 0 Param 2=1
    Index 0 Param 3=1
    Module.Index0Float0Write=true
    Module.Index0Float1Write=true
    Module.Index0Float2Write=true
    Module.Index0Float3Write=true
   )
   &FN_Light_Attributes(
    Color=<rgba 1 0.575 0.15 1>
    Create Color Attribute=true
    Create Exponent Attribute=true
    Create Radius Attribute=true
    Enable Particle Light Enable=true
    Light Enabled=true
    Radius=&Multiply_Float(A=1200, B=$User.SlashSize)
    Scale Color=&Multiply_Float(
     A=18
     B=&FloatFromCurve(FloatCurve=DataInterface() {propertyValues {Curve {
         defaultValue=1
         keys [
          {arriveTangent=9.4247, interpMode=RCIM_Cubic, leaveTangent=9.4247, tangentMode=RCTM_User, time=0, value=0}
          {arriveTangent=-0.218, interpMode=RCIM_Cubic, leaveTangent=-0.218, tangentMode=RCTM_User, time=0.5, value=1.1218}
          {arriveTangent=-1.5111, tangentMode=RCTM_User, time=1, value=0}
         ]
        }}})
    )
   )
   &SolveForcesAndVelocity(Clamp Velocity=false, Limit Acceleration=false)
  ]
  rendererClasses [&NiagaraSpriteRendererProperties, &NiagaraLightRendererProperties]
  renderers [
   {rendererClass=&NiagaraSpriteRendererProperties}
   {rendererClass=&NiagaraLightRendererProperties, rendererIndex=1}
  ]
 }
 Slash_Main {
  bEnabled=false
  emitterUpdateScript [
   &EmitterState(Life Cycle Mode=::NewEnumerator1, Loop Behavior=::NewEnumerator1, Loop Duration Mode=::NewEnumerator1, UseLoopDelay=false)
   &SpawnBurst_Instantaneous(Use Spawn Probability=false, UseLoopCountLimit=false)
  ]
  particleSpawnScript [
   &InitializeParticle(
    Color=<rgba 4 2.4696 0.213 1>
    Lifetime=&Multiply_Float(A=0.45, B=$User.ScaleLifetime)
    Mesh Scale=&Multiply_Vector(A=<1.5 1.5 1.5>, B=&VectorFromFloat(Value=$User.SlashSize))
    Mesh Scale Mode=::NewEnumerator4
    UsePositionOffset=false
    Write Material Random=true
   )
   /SetVariables(Particles.MeshIndex=&MakeIntFromFloat(Float=$User.MeshIndex)) {isSetParameters=true, module=SetVariables_#5}
   &InitialMeshOrientation(Mesh Orientation Mode=::NewEnumerator1, Use Rotation Vector=true)
  ]
  particleUpdateScript [
   &ParticleState()
   &ApplyOwnerScaleToAttributes(Scale Initial Mesh Scale=true) {enabled=false}
   &UpdateMeshOrientation(Rotation Rate=$System.RotationSpeed, Rotation Vector=<0 1 0>)
   &DynamicMaterialParameters(
    Index 0 Param 1=&RemapRange(Input Value To Remap=$System.SlashTime, Input Min=0, Input Max=1, Output Min=0, Output Max=0.7696)
    Index 0 Param 2=&FloatFromCurve(
     FloatCurve=DataInterface() {propertyValues {Curve {keys [{interpMode=RCIM_Cubic, leaveTangent=1.6601, tangentMode=RCTM_Break, time=0, value=0}, <at 0.4298 1>, <at 0.8 1>, <at 1 0>]}}}
     Scale Curve=&Multiply_Float(A=$User.SlashSize, B=1.3294)
    )
    Index 0 Param 3=&FloatFromCurve(
     FloatCurve=DataInterface() {propertyValues {Curve {keys [{interpMode=RCIM_Cubic, leaveTangent=4.0629, tangentMode=RCTM_Break, time=0.2279, value=0}, <at 1 1>]}}}
     Scale Curve=10
    )
    Index 0 Param 4=1
    Module.Index0Float0Write=true
    Module.Index0Float1Write=true
    Module.Index0Float2Write=true
    Module.Index0Float3Write=true
   )
   &ScaleColor(
    Scale Alpha=&FloatFromCurve(
     FloatCurve=DataInterface() {propertyValues {Curve {keys [
         {interpMode=RCIM_Cubic, leaveTangent=-3.4444, tangentMode=RCTM_Break, time=0.2255, value=1}
         {interpMode=RCIM_Cubic, tangentMode=RCTM_Break, time=1, value=0}
        ]}}}
     Scale Curve=&Multiply_Float(
      A=&MakeCustomFloatFromBool(Bool=$System.SoftVer, True Float=0.1, False Float=0.3)
      B=&FNTimeOfDayScalar(Night Scale=0.1, Morning Scale=0.3, Day Scale=0.9, Evening Scale=0.3)
     )
    )
    ScaleA=true
    ScaleRGB=true
   )
   /SetVariables(Particles.Position2:NiagaraPosition=&AddVectorToPosition(Position=$Particles.Position, Vector=$System.PositionOffset)) {isSetParameters=true, module=SetVariables_#6}
  ]
  rendererClasses [&NiagaraMeshRendererProperties]
  renderers [{rendererClass=&NiagaraMeshRendererProperties, rendererIndex=0}]
 }
 Slash_Main001 {
  bEnabled=false
  emitterUpdateScript [
   &EmitterState(Life Cycle Mode=::NewEnumerator1, Loop Behavior=::NewEnumerator1, Loop Duration Mode=::NewEnumerator1, UseLoopDelay=false)
   &SpawnBurst_Instantaneous(Use Spawn Probability=false, UseLoopCountLimit=false)
  ]
  particleSpawnScript [
   &InitializeParticle(
    Color=<rgba 1 0.6174 0.0533 1>
    Lifetime=&Multiply_Float(A=0.45, B=$User.ScaleLifetime)
    Mesh Scale=&Multiply_Vector(A=<1.5 1.5 1.5>, B=&VectorFromFloat(Value=$User.SlashSize))
    Mesh Scale Mode=::NewEnumerator4
    UsePositionOffset=false
    Write Material Random=true
   )
   /SetVariables(Particles.MeshIndex=&MakeIntFromFloat(Float=$User.MeshIndex)) {isSetParameters=true, module=SetVariables_#5}
   &InitialMeshOrientation(Mesh Orientation Mode=::NewEnumerator1, Use Rotation Vector=true)
  ]
  particleUpdateScript [
   &ParticleState()
   &ApplyOwnerScaleToAttributes(Scale Initial Mesh Scale=true) {enabled=false}
   &UpdateMeshOrientation(Rotation Rate=$System.RotationSpeed, Rotation Vector=<0 1 0>)
   &DynamicMaterialParameters(
    Index 0 Param 1=&RemapRange(Input Value To Remap=$System.SlashTime, Input Min=0, Input Max=1, Output Min=0, Output Max=0.7696)
    Index 0 Param 2=&FloatFromCurve(
     FloatCurve=DataInterface() {propertyValues {Curve {keys [{interpMode=RCIM_Cubic, leaveTangent=1.6601, tangentMode=RCTM_Break, time=0, value=0}, <at 0.4298 1>, <at 0.8 1>, <at 1 0>]}}}
     Scale Curve=&Multiply_Float(A=$User.SlashSize, B=4.3115)
    )
    Index 0 Param 3=&FloatFromCurve(
     FloatCurve=DataInterface() {propertyValues {Curve {keys [{interpMode=RCIM_Cubic, leaveTangent=4.0629, tangentMode=RCTM_Break, time=0.2279, value=0}, <at 1 1>]}}}
     Scale Curve=10
    )
    Index 0 Param 4=1
    Module.Index0Float0Write=true
    Module.Index0Float1Write=true
    Module.Index0Float2Write=true
    Module.Index0Float3Write=true
   )
   &ScaleColor(
    Scale Alpha=&FloatFromCurve(
     FloatCurve=DataInterface() {propertyValues {Curve {keys [
         {interpMode=RCIM_Cubic, leaveTangent=-3.4444, tangentMode=RCTM_Break, time=0.2255, value=1}
         {interpMode=RCIM_Cubic, tangentMode=RCTM_Break, time=1, value=0}
        ]}}}
     Scale Curve=&Multiply_Float(
      A=&MakeCustomFloatFromBool(Bool=$System.SoftVer, True Float=0.1, False Float=0.3)
      B=&FNTimeOfDayScalar(Night Scale=0.1, Morning Scale=0.3, Day Scale=0.9, Evening Scale=0.3)
     )
    )
    ScaleA=true
    ScaleRGB=true
   )
   /SetVariables(Particles.Position2:NiagaraPosition=&AddVectorToPosition(Position=$Particles.Position, Vector=$System.PositionOffset)) {isSetParameters=true, module=SetVariables_#6}
  ]
  rendererClasses [&NiagaraMeshRendererProperties]
  renderers [{rendererClass=&NiagaraMeshRendererProperties, rendererIndex=0}]
 }
 Slash_Main002 {
  bEnabled=false
  emitterUpdateScript [
   &EmitterState(Life Cycle Mode=::NewEnumerator1, Loop Behavior=::NewEnumerator1, Loop Duration=1, UseLoopDelay=false)
   &SpawnBurst_Instantaneous(Spawn Time=0.05, Use Spawn Probability=false, UseLoopCountLimit=false)
  ]
  particleSpawnScript [
   &InitializeParticle(
    Color=<rgba 1 0.6174 0.0533 1>
    Lifetime=0.3
    Mesh Scale Mode=::NewEnumerator3
    Mesh Uniform Scale=&Multiply_Float(A=1.2776, B=$User.SlashSize)
    Position Offset=<0 -156.7996 0>
    UsePositionOffset=true
    Write Material Random=true
   )
   /SetVariables(Particles.MeshIndex=&MakeIntFromFloat(Float=$User.MeshIndex)) {isSetParameters=true, module=SetVariables_#5}
   &InitialMeshOrientation(Mesh Orientation Mode=::NewEnumerator1, Use Rotation Vector=false) {enabled=false}
   &AddVelocity(Velocity=<0 894.373 0>)
  ]
  particleUpdateScript [
   &ParticleState()
   &ScaleMeshSize(Scale Factor=&VectorFromFloat(Value=&FloatFromCurve(FloatCurve=DataInterface() {propertyValues {Curve {
         defaultValue=1
         keys [
          {arriveTangent=4.1785, interpMode=RCIM_Cubic, leaveTangent=4.1785, tangentMode=RCTM_User, time=0, value=0.5805}
          {arriveTangent=1.1179, interpMode=RCIM_Cubic, leaveTangent=1.1179, tangentMode=RCTM_User, time=0.2042, value=0.9362}
          {arriveTangent=0.0224, interpMode=RCIM_Cubic, leaveTangent=0.0224, tangentMode=RCTM_User, time=1, value=1.0896}
         ]
        }}}))) {enabled=false}
   &ApplyOwnerScaleToAttributes(Scale Initial Mesh Scale=true) {enabled=false}
   &UpdateMeshOrientation(Rotation Rate=0.5, Rotation Vector=<0 1 0>)
   &DynamicMaterialParameters(
    Index 0 Param 1=&FloatFromCurve(FloatCurve=DataInterface() {propertyValues {Curve {keys [<at 0 1>,<at 1 -1>]}}})
    Index 0 Param 2=&FloatFromCurve(
     FloatCurve=DataInterface() {propertyValues {Curve {
        defaultValue=1
        keys [
         {interpMode=RCIM_Cubic, tangentMode=RCTM_User, time=0, value=0.6707}
         {interpMode=RCIM_Cubic, tangentMode=RCTM_User, time=0.0746, value=1.1823}
         <at 0.8281 1 RCIM_Cubic>
        ]
       }}}
     Scale Curve=&Multiply_Float(A=$User.SlashSize, B=0.7)
    )
    Index 0 Param 4=1
    Module.Index0Float0Write=true
    Module.Index0Float1Write=true
    Module.Index0Float2Write=true
    Module.Index0Float3Write=true
   )
   &ScaleColor(
    Scale Alpha=&FloatFromCurve(
     FloatCurve=DataInterface() {propertyValues {Curve {keys [
         {interpMode=RCIM_Cubic, leaveTangent=-3.4444, tangentMode=RCTM_Break, time=0.2255, value=1}
         {interpMode=RCIM_Cubic, tangentMode=RCTM_Break, time=1, value=0}
        ]}}}
     Scale Curve=&Multiply_Float(
      A=&MakeCustomFloatFromBool(Bool=$System.SoftVer, True Float=0.1, False Float=0.3)
      B=&FNTimeOfDayScalar(Night Scale=0.1, Morning Scale=0.3, Day Scale=0.9, Evening Scale=0.3)
     )
    )
    ScaleA=true
    ScaleRGB=true
   )
   &SolveForcesAndVelocity(Clamp Velocity=false, Limit Acceleration=false)
   /SetVariables(Particles.Position2:NiagaraPosition=&AddVectorToPosition(Position=$Particles.Position, Vector=$System.PositionOffset)) {isSetParameters=true, module=SetVariables_#6}
  ]
  rendererClasses [&NiagaraMeshRendererProperties]
  renderers [{rendererClass=&NiagaraMeshRendererProperties, rendererIndex=0}]
 }
 Slash_Main002_Lightweight {rendererClasses [], renderers []}
 Slash_Main_Lightweight {rendererClasses [], renderers []}
 SoftEnergyWave {
  bEnabled=false
  emitterUpdateScript [
   &EmitterState(Life Cycle Mode=::NewEnumerator1, Loop Behavior=::NewEnumerator1, Loop Duration=10, UseLoopDelay=true)
   &SpawnBurst_Instantaneous(Use Spawn Probability=false, UseLoopCountLimit=false)
  ]
  particleSpawnScript [
   &InitializeParticle(
    Color=<rgba 10 4.9333 0.5 0.4>
    Lifetime=0.3
    Mesh Scale=&VectorFromFloat(Value=0.7)
    Mesh Scale Mode=::NewEnumerator4
    Position Offset=<0 -337.5422 0>
    UsePositionOffset=true
    Write Material Random=true
   )
   &InitialMeshOrientation(Mesh Orientation Mode=::NewEnumerator1, Rotation=<0 -0.37 0>, Use Rotation Vector=true)
  ]
  particleUpdateScript [
   &ParticleState()
   &ApplyOwnerScaleToAttributes(Scale Initial Mesh Scale=true) {enabled=false}
   &UpdateMeshOrientation(Rotation Rate=$System.RotationSpeed, Rotation Vector=<0 1.25 0>) {enabled=false}
   &ScaleMeshSize(Scale Factor=&VectorFromFloat(Value=&FloatFromCurve(
      FloatCurve=DataInterface() {propertyValues {Curve {keys [{interpMode=RCIM_Cubic, tangentMode=RCTM_Break, time=0, value=0.9}, {interpMode=RCIM_Cubic, tangentMode=RCTM_Break, time=1, value=1.0976}]}}}
      Scale Curve=1.2
     )))
   &DynamicMaterialParameters(
    Index 0 Param 1=$User.AnimationTimeInput
    Index 0 Param 2=&FloatFromCurve(FloatCurve=DataInterface() {propertyValues {Curve {keys [
         {interpMode=RCIM_Cubic, leaveTangent=2.4962, tangentMode=RCTM_Break, time=0.25, value=0}
         {arriveTangent=2.6266, interpMode=RCIM_Cubic, tangentMode=RCTM_Break, time=0.8847, value=1}
        ]}}})
    Index 0 Param 3=&FloatFromCurve(
     FloatCurve=DataInterface() {propertyValues {Curve {keys [{interpMode=RCIM_Cubic, leaveTangent=4.0629, tangentMode=RCTM_Break, time=0.2279, value=0}, <at 1 1>]}}}
     Scale Curve=20
    )
    Index 0 Param 4=1
    Module.Index0Float0Write=true
    Module.Index0Float1Write=true
    Module.Index0Float2Write=true
    Module.Index0Float3Write=true
   ) {enabled=false}
   &ScaleColor(
    Scale Alpha=&FloatFromCurve(
     FloatCurve=DataInterface() {propertyValues {Curve {keys [
         {interpMode=RCIM_Cubic, leaveTangent=-0.5988, tangentMode=RCTM_Break, time=0, value=0}
         {arriveTangent=2.9332, interpMode=RCIM_Cubic, leaveTangent=2.9332, time=0.1, value=1}
         {arriveTangent=-1.851, interpMode=RCIM_Cubic, leaveTangent=-1.851, time=0.2813, value=0.8251}
         {arriveTangent=-0.2926, interpMode=RCIM_Cubic, leaveTangent=-0.2926, tangentMode=RCTM_User, time=0.5089, value=0.2432}
         {interpMode=RCIM_Cubic, tangentMode=RCTM_Break, time=1, value=0}
        ]}}}
     Scale Curve=&FNTimeOfDayScalar(Night Scale=1, Morning Scale=1, Evening Scale=1)
    )
    ScaleA=true
    ScaleRGB=true
   )
   &DynamicMaterialParameters(
    Index 0 Param 1=&FloatFromCurve(FloatCurve=DataInterface() {propertyValues {Curve {keys [<at 0 0 RCIM_Cubic>,<at 1 1>]}}})
    Index 0 Param 2=0.1
    Index 2 Param 1=100
    Module.Index0Param0Write=true
    Module.Index0Param1Write=true
    Module.Index0Param2Write=true
    Module.Index0Param3Write=true
    Module.Index2Param0Write=true
    Module.Index2Param1Write=true
    Module.Index2Param2Write=true
    Module.Index2Param3Write=true
    Write Parameter Index 2=true
   ) {module=DynamicMaterialParameters001}
  ]
  rendererClasses [&NiagaraMeshRendererProperties]
  renderers [
   {rendererClass=&NiagaraMeshRendererProperties, rendererIndex=0}
   {rendererClass=&NiagaraMeshRendererProperties, rendererIndex=1}
  ]
 }
 SoftEnergyWave_Lightweight {rendererClasses [], renderers []}
 Stars {
  simTarget=GPUComputeSim
  emitterUpdateScript [
   &EmitterState(Life Cycle Mode=::NewEnumerator1, Loop Behavior=::NewEnumerator1, Loop Delay=0.1, Loop Duration=1, UseLoopDelay=false)
   &SpawnBurst_Instantaneous(Spawn Count=8, Use Spawn Probability=false, UseLoopCountLimit=false)
  ]
  particleSpawnScript [
   &InitializeParticle(
    Color=<rgba 1 0.5093 0.08 1>
    Lifetime=&RandomRangeFloat(Minimum=0.2, Maximum=0.5)
    Sprite Rotation Mode=::NewEnumerator0
    Sprite Size Mode=::NewEnumerator1
    Uniform Sprite Size Max=20
    Uniform Sprite Size Min=12
    UsePositionOffset=false
    Write Material Random=true
   )
   &ShapeLocation(
    MaskSpawnGroup=false
    Override Seed=false
    Rotation Axis=<0 0 1>
    Rotation Mode=::NewEnumerator1
    Sphere Radius=50
    Sphere Surface Distribution=1
    UseOwnerScale=true
   )
   &AddVelocity(
    Module.Use Falloff Distance=false
    Origin Offset=&RandomRangeVector(Minimum=<0 -200 0>, Maximum=<0 -300 0>)
    Override Seed=false
    Velocity Mode=::NewEnumerator1
    Velocity Speed=&RandomRangeFloat(Minimum=2505.9387, Maximum=3586.7312)
   )
   &AddVelocity(
    Rotation Coordinate Space=::Simulation
    Velocity=<0 -200 0>
    Velocity Speed Scale=&RandomRangeFloat(Minimum=0, Maximum=0.5)
   ) {enabled=false, module=AddVelocity001}
  ]
  particleUpdateScript [
   &ParticleState()
   &ScaleSpriteSize(Uniform Curve Sprite Scale=DataInterface() {propertyValues {Curve {
       defaultValue=0
       keys [
        {tangentMode=RCTM_User, time=0, value=1}
        {arriveTangent=-0.0405, interpMode=RCIM_Cubic, leaveTangent=-0.0405, tangentMode=RCTM_User, time=0.7009, value=1}
        {arriveTangent=-8.0089, interpMode=RCIM_Cubic, leaveTangent=-8.0089, tangentMode=RCTM_User, time=1, value=0}
       ]
      }}})
   &Drag(Drag=3, Use Linear Drag=true, Use Rotational Drag=true)
   &SolveForcesAndVelocity(Clamp Velocity=false, Limit Acceleration=false)
   &DynamicMaterialParameters(
    Index 0 Param 1=&Waveform("Waveform [1]"=::NewEnumerator0, "[1] Phase Scale"=5, "[1] Frequency"=1, "[1] Amplitude Min/Max"=<3 5>)
    Module.Index0Float0Write=true
    Module.Index0Float1Write=true
    Module.Index0Float2Write=true
    Module.Index0Float3Write=true
   ) {enabled=false}
   &FN_Sprite_Attributes(
    Color=&MultiplyLinearColor(
     Linear Color=<rgba 1 0.525 0.05 0.4>
     Scale Factor=&MakeVector4(
      X=1
      Y=1
      Z=1
      W=&FloatFromCurve(FloatCurve=DataInterface() {propertyValues {Curve {keys [{interpMode=RCIM_Cubic, tangentMode=RCTM_Break, time=0, value=1}, <at 1 0>]}}})
     )
    )
    Create Color Attribute=true
    Create SpriteSize Attribute=true
    SpriteSize=&Multiply_Vector2D_ByFloat(Float=1.8067, Vector2D=<60 60>)
    Write to Module Namespace=true
   )
   /SetVariables(Particles.Position2:NiagaraPosition=&AddVectorToPosition(Position=$Particles.Position, Vector=<0 0 0>)) {isSetParameters=true}
   &DynamicMaterialParameters(
    Index 0 Param 1=&FloatFromCurve(FloatCurve=DataInterface() {propertyValues {Curve {keys [<at 0 0>,<at 0.4774 1>]}}})
    Module.Index0Float0Write=true
    Module.Index0Float1Write=true
    Module.Index0Float2Write=true
    Module.Index0Float3Write=true
   ) {module=DynamicMaterialParameters001}
   /SetVariables(Particles.Vector 4:Vector4f=<4 0 0 0>) {isSetParameters=true}
  ]
  rendererClasses [&NiagaraSpriteRendererProperties]
  renderers [
   {rendererClass=&NiagaraSpriteRendererProperties}
   {rendererClass=&NiagaraSpriteRendererProperties, rendererIndex=1}
  ]
 }
 Streaks {
  bEnabled=false
  emitterUpdateScript [
   &EmitterState()
   &SpawnBurst_Instantaneous(Spawn Count=3, Use Spawn Probability=false, UseLoopCountLimit=false)
  ]
  particleSpawnScript [
   /SetVariables(Particles.RandomFloat=&UniformRangedFloat(Minimum=0, Maximum=3)) {isSetParameters=true}
   &InitializeParticle(
    Color Maximum=<rgba 0.9565 0.4799 0.0452 3>
    Color Minimum=<rgba 0.6701 0.5575 0.0584 5>
    Color Mode=::NewEnumerator2
    Lifetime Max=0.3
    Lifetime Min=0.25
    Lifetime Mode=::NewEnumerator1
    Sprite Size Max=<18 24.9845>
    Sprite Size Min=<15 52.3129>
    Sprite Size Mode=::NewEnumerator2
    UsePositionOffset=false
    Write Material Random=true
   )
   &ShapeLocation(
    Cylinder Height=400
    Cylinder Radius=20
    MaskSpawnGroup=false
    Override Seed=false
    Rotation Axis=$User.Direction
    Rotation Mode=::NewEnumerator1
    Shape Primitive=::NewEnumerator1
    Surface Only=false
    UseOwnerScale=true
   ) {module=ShapeLocation001}
   &AddVelocity(
    Cone Angle=10
    Cone Axis=<0 1 0>
    Module.UseConeFalloff=false
    Module.UseVelDistribution=false
    Override Seed=false
    Velocity Mode=::NewEnumerator2
   ) {module=AddVelocity001}
   &CameraOffset(Camera Offset Amount=$User.CameraO)
   &SubUVAnimation(End Frame=3, SubUV Animation Mode:ENiagara_SubUVLookupMode=::NewEnumerator2)
  ]
  particleUpdateScript [
   &ParticleState() {module=UpdateAge}
   &ApplyOwnerScaleToAttributes(Scale Initial Sprite Size=true)
   &Drag(Drag=5, Use Linear Drag=true, Use Rotational Drag=true)
   &SolveForcesAndVelocity(Clamp Velocity=false, Limit Acceleration=false)
   &ScaleSpriteSize(Scale Factor=&Vector2DFromCurve(
     Vector2Curve=DataInterface() {propertyValues {XCurve {keys [<at 0.0036 1.6897>,<at 1 1>]}, YCurve {keys [<at 0 1>,<at 1.002 3.7227>]}}}
     ScaleCurve=<1.5 2.2>
    ))
   &ScaleSpriteSize(
    Scale Sprite Size Mode=::NewEnumerator0
    Uniform Scale Factor=&Multiply_Float(
     A=0.8
     B=&SelectFloatFromArray(Direct Array Index=$Particles.PartitionParticles.Partition, Float Selection Array=DataInterface() {propertyValues {FloatData [1,1]}})
    )
   ) {module=ScaleSpriteSize001}
   &ScaleColor(
    Scale Alpha=&FloatFromCurve(
     FloatCurve=DataInterface() {propertyValues {Curve {keys [{interpMode=RCIM_Cubic, tangentMode=RCTM_Break, time=0, value=1}, {interpMode=RCIM_Cubic, tangentMode=RCTM_Break, time=1, value=0}]}}}
     Scale Curve=&RandomRangeFloat(Minimum=0.15, Maximum=0.5)
    )
    Scale RGB=&VectorFromFloat(Value=&Multiply_Float(
      A=&RandomRangeFloat(Minimum=2, Maximum=15)
      B=&FloatFromCurve(FloatCurve=DataInterface() {propertyValues {Curve {keys [<at 0 1>,<at 0.2597 0.9873>,<at 0.3816 0.1931>,<at 1 0>]}}}, Scale Curve=10)
     ))
    ScaleA=true
    ScaleRGB=true
   )
   &DynamicMaterialParameters(
    Index 0 Param 2=1
    Index 0 Param 3=1
    Index 0 Param 4=&FloatFromCurve(FloatCurve=DataInterface() {propertyValues {Curve {keys [<at 0.3 1>,<at 1 0.25>]}}})
    Module.Param0WriteEnabled=true
    Module.Param1WriteEnabled=true
    Module.Param2WriteEnabled=true
    Module.Param3WriteEnabled=true
   )
  ]
  rendererClasses [&NiagaraSpriteRendererProperties]
  renderers [{rendererClass=&NiagaraSpriteRendererProperties}]
 }
 Swirls {
  bEnabled=false
  emitterUpdateScript [
   &EmitterState()
   &SpawnBurst_Instantaneous(Spawn Count=5, Use Spawn Probability=false, UseLoopCountLimit=false)
  ]
  particleSpawnScript [
   &InitializeParticle(
    Color Mode=::NewEnumerator0
    Lifetime=&RandomRangeFloat(Minimum=0.2, Maximum=0.4)
    Mesh Scale Mode=::NewEnumerator3
    Mesh Uniform Scale=&RandomRangeFloat(Minimum=0.3, Maximum=1)
    Position Offset=<0 -4.6175 0>
    UsePositionOffset=true
    Write Material Random=true
   )
   &InitialMeshOrientation(
    Mesh Orientation Mode=::NewEnumerator7
    Rotation=&RandomRangeVector(Minimum=<-0.25 0 0>, Maximum=<-0.25 0 0>)
    Rotation Coordinate Space=::NewEnumerator2
    Use Rotation Vector=true
   ) {module=InitialMeshOrientation002}
   &InitialMeshOrientation(
    Mesh Orientation Mode=::NewEnumerator7
    Rotation=&RandomRangeVector(Minimum=<0.1252 0 0>, Maximum=<0.1796 0 0>)
    Use Rotation Vector=true
   ) {enabled=false}
   &InitialMeshOrientation(
    Mesh Orientation Mode=::NewEnumerator7
    Rotation=&MakeVector(
     X=0
     Y=0
     Z=&RemapRange(Input Value To Remap=&ReturnNormalizedExecIndex(), Input Min=0, Input Max=1, Output Min=0.1, Output Max=0.9)
    )
    Use Rotation Vector=true
   ) {enabled=false, module=InitialMeshOrientation001}
  ]
  particleUpdateScript [
   &ParticleState()
   &ScaleMeshSize(Scale Factor=&MakeVector(
     X=1
     Y=1
     Z=&FloatFromCurve(
      FloatCurve=DataInterface() {propertyValues {Curve {keys [
          {interpMode=RCIM_Cubic, leaveTangent=7.8515, tangentMode=RCTM_Break, time=0, value=0}
          {interpMode=RCIM_Cubic, tangentMode=RCTM_Break, time=0.3318, value=1}
          {arriveTangent=0.7294, interpMode=RCIM_Cubic, leaveTangent=0.7294, time=0.7387, value=1.095}
          <at 0.9363 1.4409>
         ]}}}
      Scale Curve=0.4368
     )
    ))
   &UpdateMeshOrientation(Rotation Rate=&RandomRangeFloat(Minimum=-0.3, Maximum=0.3), Rotation Vector=<0 0 1>)
   &DynamicMaterialParameters(
    Index 0 Param 1=&RandomRangeFloat(Minimum=0, Maximum=100)
    Index 0 Param 2=&RandomRangeFloat(Minimum=2, Maximum=3)
    Index 0 Param 3=&FloatFromCurve(FloatCurve=DataInterface() {propertyValues {Curve {keys [
         {interpMode=RCIM_Cubic, leaveTangent=0.1046, tangentMode=RCTM_Break, time=0.3924, value=0.0986}
         {arriveTangent=-2.457, tangentMode=RCTM_User, time=1, value=-0.616}
        ]}}})
    Index 0 Param 4=&FloatFromCurve(
     FloatCurve=DataInterface() {propertyValues {Curve {keys [{interpMode=RCIM_Cubic, leaveTangent=3.2938, tangentMode=RCTM_Break, time=0, value=0}, <at 0.5545 1>]}}}
     Scale Curve=&RandomRangeFloat(Minimum=59.8838, Maximum=144.4326)
    )
    Index 1 Param 1=&FloatFromCurve(FloatCurve=DataInterface() {propertyValues {Curve {keys [{interpMode=RCIM_Cubic, leaveTangent=-3.0821, tangentMode=RCTM_Break, time=0, value=1}, <at 1 -0.3061>]}}})
    Module.Index0Float0Write=true
    Module.Index0Float1Write=true
    Module.Index0Float2Write=true
    Module.Index0Float3Write=true
    Module.Index1Float0Write=true
    Module.Index1Float1Write=true
    Module.Index1Float2Write=true
    Module.Index1Float3Write=true
    Write Parameter Index 1=::NewEnumerator1
   )
  ]
  rendererClasses [&NiagaraMeshRendererProperties]
  renderers [{rendererClass=&NiagaraMeshRendererProperties, rendererIndex=0}]
 }
}
summary {
 systemName=NS_DentureSay_JabFX_AnimTrail
 emitters [
  {
   bEnabled=false
   emitterName=Flare_And_Light
   rendererClasses [&NiagaraSpriteRendererProperties, &NiagaraLightRendererProperties]
  }
  {bEnabled=false, emitterName=Streaks, rendererClasses [&NiagaraSpriteRendererProperties]}
  {emitterName=JabLines, rendererClasses [&NiagaraSpriteRendererProperties]}
  {emitterName=Stars, simTarget=GPUComputeSim, rendererClasses [&NiagaraSpriteRendererProperties]}
  {bEnabled=false, emitterName=Swirls, rendererClasses [&NiagaraMeshRendererProperties]}
  {bEnabled=false, emitterName=Slash_Main, rendererClasses [&NiagaraMeshRendererProperties]}
  {bEnabled=false, emitterName=Slash_Main002, rendererClasses [&NiagaraMeshRendererProperties]}
  {emitterName=Endflare, rendererClasses []}
  {emitterName=Minimal, rendererClasses [&NiagaraSpriteRendererProperties, &NiagaraLightRendererProperties]}
  {bEnabled=false, emitterName=Slash_Main001, rendererClasses [&NiagaraMeshRendererProperties]}
  {bEnabled=false, emitterName=SoftEnergyWave, rendererClasses [&NiagaraMeshRendererProperties]}
  {emitterName=Flare_And_Light_Lightweight, rendererClasses []}
  {emitterName=Slash_Main_Lightweight, rendererClasses []}
  {emitterName=Slash_Main002_Lightweight, rendererClasses []}
  {emitterName=SoftEnergyWave_Lightweight, rendererClasses []}
 ]
 userVariables [
  User.AnimationTimeInput {type=NiagaraFloat, defaultValue=1.0483, description=""}
  User.CameraO {type=NiagaraFloat, defaultValue=150, description=""}
  User.Direction {type=Vector3f, description="", defaultValue {struct=Vector3f, value=<0 1 0>}}
  User.MeshIndex {type=NiagaraFloat, defaultValue=0, description=""}
  User.Radius {type=NiagaraFloat, defaultValue=1200, description=""}
  User.ScaleLifetime {type=NiagaraFloat, defaultValue=0.5, description=""}
  User.SlashSize {type=NiagaraFloat, defaultValue=0.8, description=""}
 ]
}
