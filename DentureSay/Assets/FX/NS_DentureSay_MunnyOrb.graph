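# Niagara system for a glowing munny orb effect, built from two sprite emitters that each burst a single persistent particle.
# Corona draws a pale warm glow nudged by a camera offset, while Flare_and_Light adds a larger orange flare, with no light renderer.
# Waveforms pulse each sprite's alpha at different rates, and both sprites follow owner scale.
# Front-end, camera and direction user variables are declared, but no shown module reads them.
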NiagaraSystem /DentureSay/Assets/FX/NS_DentureSay_MunnyOrb
dependencies {usedDataInterfaces [], usedDynamicInputs [], usedModules [&SystemState]}
emitters {
 Corona {
  emitterUpdateScript [
   &EmitterState(
    Inactive Response=::NewEnumerator1
    Life Cycle Mode=::NewEnumerator1
    Loop Behavior=::NewEnumerator1
    Loop Duration Mode=::NewEnumerator1
    UseLoopDelay=false
   )
   &SpawnBurst_Instantaneous(Spawn Count=1, Spawn Time=0.1, Use Spawn Probability=false)
  ]
  particleSpawnScript [
   &InitializeParticle(
    Color=<rgba 1 0.8874 0.6038 10>
    Lifetime=0.15
    Sprite Size Mode=::NewEnumerator3
    Uniform Sprite Size=97.9869
    UsePositionOffset=true
    Write Material Random=true
   )
   &DynamicMaterialParameters(
    Index 0 Param 1=0.8552
    Index 0 Param 2=1
    Index 0 Param 3=1
    Index 0 Param 4=1
    Module.Param0WriteEnabled=true
    Module.Param1WriteEnabled=true
    Module.Param2WriteEnabled=true
    Module.Param3WriteEnabled=true
   )
  ]
  particleUpdateScript [
   &ParticleState(Kill Particles When Lifetime Has Elapsed=false, Let Infinitely Lived Particles Die When Emitter Deactivates=true)
   &ApplyOwnerScaleToAttributes(Scale Initial Sprite Size=true)
   &CameraOffset(Camera Offset Amount=15)
   &ScaleColor(
    Scale Alpha=&Waveform("Waveform [1]"=::NewEnumerator12, "[1] Frequency"=3, "[1] Amplitude Min/Max"=<0.2 0.3>, "[1] Scale"=0.2)
    ScaleA=true
    ScaleRGB=true
   )
  ]
  rendererClasses [&NiagaraSpriteRendererProperties]
  renderers [{rendererClass=&NiagaraSpriteRendererProperties}]
 }
 Flare_and_Light {
  emitterUpdateScript [
   &EmitterState(
    Inactive Response=::NewEnumerator1
    Life Cycle Mode=::NewEnumerator1
    Loop Behavior=::NewEnumerator1
    Loop Duration Mode=::NewEnumerator1
    UseLoopDelay=false
   )
   &SpawnBurst_Instantaneous(Spawn Count=1, Use Spawn Probability=true)
  ]
  particleSpawnScript [
   &InitializeParticle(Color=<rgba 3 1.5652 0 0.7>, Sprite Size=<300 300>, Sprite Size Mode=::NewEnumerator4, UsePositionOffset=true, Write Material Random=true)
   &DynamicMaterialParameters(
    Index 0 Param 1=1
    Module.Param0WriteEnabled=true
    Module.Param1WriteEnabled=true
    Module.Param2WriteEnabled=true
    Module.Param3WriteEnabled=true
   )
  ]
  particleUpdateScript [
   &ParticleState(Kill Particles When Lifetime Has Elapsed=false, Let Infinitely Lived Particles Die When Emitter Deactivates=true)
   &ApplyOwnerScaleToAttributes(Scale Initial Sprite Size=true)
   &ScaleColor(
    Scale Alpha=&Waveform("Waveform [1]"=::NewEnumerator12, "[1] Frequency"=2, "[1] Amplitude Min/Max"=<0.3 0.4>)
    ScaleA=true
    ScaleRGB=true
   )
  ]
  rendererClasses [&NiagaraSpriteRendererProperties]
  renderers [{rendererClass=&NiagaraSpriteRendererProperties}]
 }
}
summary {
 systemName=NS_DentureSay_MunnyOrb
 emitters [
  {emitterName=Corona, rendererClasses [&NiagaraSpriteRendererProperties]}
  {emitterName=Flare_and_Light, rendererClasses [&NiagaraSpriteRendererProperties]}
 ]
 userVariables [
  User.bIsFrontEnd {type=NiagaraBool}
  User.bIsFrontEndPreview {type=NiagaraBool}
  User.CameraO {type=NiagaraFloat, defaultValue=150, description=""}
  User.Direction {type=Vector3f, description="", defaultValue {struct=Vector3f, value=<0 0 1>}}
 ]
}
