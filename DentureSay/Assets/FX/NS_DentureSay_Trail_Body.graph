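# Niagara body trail for a skeletal character, apparently adapted from a grind-rail effect, with ribbons, flares and a sword trail.
# Bone-attached ribbons spawn by velocity or a speed parameter, alongside foot flares with lights, a ring burst and hand particles.
# A separate anim-trail emitter follows the sword mesh.
# A smoothed system color tints ribbons and flares, time-of-day scales brightness, and jumping zeroes foot light and rail-glow alpha.
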NiagaraSystem /DentureSay/Assets/FX/NS_DentureSay_Trail_Body
dependencies {
 usedDataInterfaces [
  &NiagaraDataInterfaceCurve
  &NiagaraDataInterfaceArrayFloat
  &NiagaraDataInterfaceArrayInt32
  &NiagaraDataInterfaceArrayFloat3
  &NiagaraDataInterfaceColorCurve
  &NiagaraDataInterfaceArrayFloat2
  &NiagaraDataInterfaceArrayColor
 ]
 usedDynamicInputs [
  &self:GetEulerRot
  &ClampFloat
  &RemapRange
  &VectorLength
  &Boolean_OrOperation
  &Boolean_NotOperation
  &MakeCustomIntFromBool
  &self:Grinding_Color
  &RandomRangeVector
  &MakeVector
  /Niagara/DynamicInputs/Lerp/SmoothLerpOverTime_Color
 ]
 usedModules [
  &self:SystemSpawnScript.SystemScriptSource.SystemScriptGraph.NiagaraNodeAssignment_9.SetVariables_#1
  &SystemState
  &SampleSkeletalMesh
  &self:SystemSpawnScript.SystemScriptSource.SystemScriptGraph.NiagaraNodeAssignment_12.SetVariables_#2
  &self:SystemSpawnScript.SystemScriptSource.SystemScriptGraph.NiagaraNodeAssignment_6.SetVariables_#3
  &InterpolateOverTime
  &self:SystemSpawnScript.SystemScriptSource.SystemScriptGraph.NiagaraNodeAssignment_0.SetVariables_#4
  &self:SystemSpawnScript.SystemScriptSource.SystemScriptGraph.NiagaraNodeAssignment_1.SetVariables_#5
  &FN_Timer
  &self:SystemSpawnScript.SystemScriptSource.SystemScriptGraph.NiagaraNodeAssignment_2.SetVariables_#6
  &self:SystemSpawnScript.SystemScriptSource.SystemScriptGraph.NiagaraNodeAssignment_3.SetVariables_#7
  &self:SystemSpawnScript.SystemScriptSource.SystemScriptGraph.NiagaraNodeAssignment_15.SetVariables_#8
  &self:SystemSpawnScript.SystemScriptSource.SystemScriptGraph.NiagaraNodeAssignment_4.SetVariables_#9
 ]
}
emitters {
 Foot_Flare001 {
  emitterUpdateScript [
   &EmitterState(
    Inactive Response=::NewEnumerator1
    Life Cycle Mode=::NewEnumerator1
    Loop Behavior=::NewEnumerator1
    Loop Delay=$User.Delay
    Loop Duration Mode=::NewEnumerator1
    UseLoopDelay=true
   )
   &SpawnBurst_Instantaneous(Spawn Count=3, Use Spawn Probability=false)
  ]
  particleSpawnScript [
   &PartitionParticles(Number Of Partitions=3, Particles Per Partition=1, Partition Mode:ENiagara_PartitionMode=::NewEnumerator1)
   /Niagara/Modules/Spawn/Initialization/InitializeParticle(
    Color=$System.Color_Smoothed
    Module.Write Color=true
    Module.Write Lifetime=true
    Module.Write Mass=true
    Module.Write Material Random=true
    Module.Write Position=false
    Module.Write Scale=true
    Module.Write SpriteRotation=true
    Module.Write SpriteSize=true
    Position=&SimulationPosition()
    Sprite Rotation=&Multiply_Float(
     A=&RandomRangeFloat(Minimum=-360, Maximum=360)
     B=&SelectFloatFromArray(Direct Array Index=$Particles.PartitionParticles.Partition, Float Selection Array=DataInterface() {propertyValues {FloatData [1,1,0]}})
    )
    Sprite Size=&SelectVector2DFromArray(
     Direct Array Index=$Particles.PartitionParticles.Partition
     Vector 2D Selection Array=DataInterface() {propertyValues {FloatData [<50 50>, <50 50>, <120 200>]}}
    )
   )
   /SetVariables(
    Particles.GlowFoot_RailColor:LinearColor=&MultiplyLinearColor(
     Linear Color=&SelectLinearColorFromArray(Array Sampling Mode=::NewEnumerator1, Direct Array Index=$User.ColorIndex, Color Selection Array=$System.GrindRail_ColorArray, Random Seed=?)
     Scale Factor=&MakeFloat4FromVectorAndFloat("Vector (RGB)"=&VectorFromFloat(Value=0.6), "Float (Alpha)"=0.2)
    )
    Particles.GlowFoot_Size:Vector2f=<100 100>
   ) {isSetParameters=true}
   /SetVariables(Particles.ColorGlow:LinearColor=&MultiplyLinearColor(
     Linear Color=$System.Color_Smoothed
     Scale Factor=&MakeFloat4FromVectorAndFloat("Vector (RGB)"=&VectorFromFloat(Value=0.6), "Float (Alpha)"=0.2)
    )) {isSetParameters=true}
   &SkeletalMeshLocation(
    Bone / Socket Index=&SelectIntFromArray(Direct Array Index=$Particles.PartitionParticles.Partition, Int Selection Array=DataInterface() {propertyValues {IntData [8,9,11]}})
    Bone Sampling Mode=::NewEnumerator4
    MaskSpawnGroup=false
    Sampled Position Offset Coordinate Space:ENiagaraSkelMeshTransforms=::NewEnumerator2
    Skeletal Mesh=$User.SkeletalMesh
   ) {module=SkeletalMeshLocation001}
   /SetVariables(Particles.VisibilityTag=&SelectIntFromArray(Direct Array Index=$Particles.PartitionParticles.Partition, Int Selection Array=DataInterface() {propertyValues {IntData [0,0,1]}})) {isSetParameters=true}
   /SetVariables(
    Particles.CameraOffsetGlow=30
    Particles.VisibilityLarge:NiagaraInt32=&SelectIntFromArray(Direct Array Index=$Particles.PartitionParticles.Partition, Int Selection Array=DataInterface() {propertyValues {IntData [1,1,0]}})
   ) {isSetParameters=true}
  ]
  particleUpdateScript [
   &ParticleState(
    Kill Particles When Lifetime Has Elapsed=false
    Let Infinitely Lived Particles Die When Emitter Deactivates=true
    Loop Particles Lifetime=true
   )
   &SpriteRotationRate(Rotation Rate=&SelectFloatFromArray(Direct Array Index=$Particles.PartitionParticles.Partition, Float Selection Array=DataInterface() {propertyValues {FloatData [24,24,0]}}))
   &SkeletalMeshLocation(
    Bone / Socket Index=&SelectIntFromArray(Direct Array Index=$Particles.PartitionParticles.Partition, Int Selection Array=DataInterface() {propertyValues {IntData [8,9,11]}})
    Bone Sampling Mode=::NewEnumerator4
    MaskSpawnGroup=false
    Sampled Position Offset=&SelectVectorFromArray(
     Direct Array Index=$Particles.PartitionParticles.Partition
     Vector Selection Array=DataInterface() {propertyValues {FloatData [<0 0 0>, <0 0 0>, <0 0 -15>]}}
    )
    Sampled Position Offset Coordinate Space:ENiagaraSkelMeshTransforms=::NewEnumerator2
    Skeletal Mesh=$User.SkeletalMesh
   )
   &DynamicMaterialParameters(
    Index 0 Param 1=&FloatFromCurve(FloatCurve=DataInterface() {propertyValues {Curve {keys [<at 0 1 RCIM_Cubic>,<at 0.8 0 RCIM_Cubic>]}}})
    Index 0 Param 2=2
    Index 0 Param 3=1
    Index 0 Param 4=1
    Index 1 Param 1=&FloatFromCurve(FloatCurve=DataInterface() {propertyValues {Curve {keys [
         {arriveTangent=-2.3163, interpMode=RCIM_Cubic, time=0, value=2}
         {arriveTangent=-1.9019, interpMode=RCIM_Cubic, leaveTangent=-1.9019, tangentMode=RCTM_User, time=1, value=0.4}
        ]}}})
    Index 1 Param 2=0.4
    Index 1 Param 3=0.6
    Index 1 Param 4=1
    Index 2 Param 1=&FloatFromCurve(FloatCurve=DataInterface() {propertyValues {Curve {keys [<at 0.46 0>,<at 1 1>]}}})
    Index 2 Param 2=45
    Module.Param0WriteEnabled=true
    Module.Param0WriteEnabled1=false
    Module.Param0WriteEnabled2=true
    Module.Param1WriteEnabled=true
    Module.Param1WriteEnabled1=true
    Module.Param1WriteEnabled2=true
    Module.Param2WriteEnabled=true
    Module.Param2WriteEnabled1=true
    Module.Param2WriteEnabled2=true
    Module.Param3WriteEnabled=true
    Module.Param3WriteEnabled1=true
    Module.Param3WriteEnabled2=true
    Write Parameter Index 1=true
    Write Parameter Index 2=true
   )
   /SetVariables(Particles.DynMatGlow:Vector4f=<3 1 0 0>) {isSetParameters=true}
   &ScaleColor(
    Scale Mode=::NewEnumerator1
    Scale RGBA=&MakeVector4_FromLinearColor(LinearColor=$System.Color_Smoothed)
    ScaleRGBA=true
   )
   &ScaleColor(
    Scale RGB=&VectorFromFloat(Value=&FloatFromCurve(
      FloatCurve=DataInterface() {propertyValues {Curve {keys [
          {interpMode=RCIM_Cubic, tangentMode=RCTM_User, time=0, value=1}
          {interpMode=RCIM_Cubic, tangentMode=RCTM_User, time=0.5, value=1.5}
          {tangentMode=RCTM_User, time=1, value=1}
         ]}}}
      Scale Curve=&SelectFloatFromArray(Direct Array Index=$User.ColorIndex, Float Selection Array=DataInterface() {propertyValues {FloatData [1,1,1,1,1,1,1]}})
     ))
    ScaleA=false
    ScaleRGB=true
   ) {module=ScaleColor002}
   /SetVariables(Particles.GlowFoot_RailColor:LinearColor=&MultiplyLinearColor(
     Linear Color=$Particles.Initial.GlowFoot_RailColor
     Scale Factor=&MakeFloat4FromVectorAndFloat(
      "Vector (RGB)"=&Multiply_VectorByFloat(
       Vector=&Multiply_Vector(
        A=&VectorFromFloat(Value=&FNTimeOfDayScalar(Night Scale=0.35, Morning Scale=0.2, Day Scale=0.4, Evening Scale=0.3))
        B=&VectorFromFloat(Value=&Add_Float(A=&Multiply_Float(A=$System.FN_Timer.Value, B=1.5), B=0.75))
       )
       Float=&SelectFloatFromArray(Direct Array Index=$User.ColorIndex, Float Selection Array=DataInterface() {propertyValues {FloatData [0.025,0.15,0.2,0.1,0.1,0.3,0.1]}})
      )
      "Float (Alpha)"=&MakeCustomFloatFromBool(Bool=$User.Jumped, True Float=0, False Float=1)
     )
    )) {isSetParameters=true}
   &ScaleColor(
    Scale RGB=&Multiply_Vector(
     A=&VectorFromFloat(Value=&FNTimeOfDayScalar(Night Scale=0.2, Morning Scale=0.3, Day Scale=0.4, Evening Scale=0.4))
     B=&VectorFromFloat(Value=&Add_Float(A=&Multiply_Float(A=$System.FN_Timer.Value, B=2.5), B=0.75))
    )
    ScaleA=true
    ScaleRGB=true
   ) {module=ScaleColor001}
   &FN_Light_Attributes(
    Color=$User.Color_Default
    Create Color Attribute=true
    Enable Particle Light Enable=true
    Scale Color=&MakeCustomFloatFromBool(Bool=$User.Jumped, True Float=0, False Float=20)
   )
   /SetVariables(Particles.ColorGlow:LinearColor=&MultiplyLinearColor(
     Linear Color=$System.Color_Smoothed
     Scale Factor=&MakeFloat4FromVectorAndFloat(
      "Vector (RGB)"=&VectorFromFloat(Value=&FloatFromCurve(
        FloatCurve=DataInterface() {propertyValues {Curve {keys [
            {interpMode=RCIM_Cubic, tangentMode=RCTM_User, time=0, value=1}
            {interpMode=RCIM_Cubic, tangentMode=RCTM_User, time=0.5, value=2.5}
            {tangentMode=RCTM_User, time=1, value=1}
           ]}}}
        Scale Curve=&FNTimeOfDayScalar(Night Scale=0.5, Morning Scale=0.4, Day Scale=0.65, Evening Scale=0.6)
       ))
      "Float (Alpha)"=&Multiply_Float(A=1, B=&Add_Float(A=0.2, B=&Multiply_Float(A=$System.FN_Timer.Value, B=0.2)))
     )
    )) {isSetParameters=true}
  ]
  rendererClasses [&NiagaraSpriteRendererProperties, &NiagaraLightRendererProperties]
  renderers [
   {rendererClass=&NiagaraSpriteRendererProperties}
   {rendererClass=&NiagaraSpriteRendererProperties, rendererIndex=1}
   {rendererClass=&NiagaraLightRendererProperties, rendererIndex=2}
   {rendererClass=&NiagaraSpriteRendererProperties, rendererIndex=3}
  ]
 }
 GrindRail_Hands {
  emitterUpdateScript [
   /SetVariables(Emitter.EmitterRandom:NiagaraFloat=&RandomRangeFloat(Minimum=0, Maximum=1)) {isSetParameters=true}
   &EmitterState(Enable Distance Culling=true, MaxDistance=9000, Module.bUseMaxDistance=true, Module.bUseMinDistance=false, Scalability Mode=::NewEnumerator1)
   &FN_SpawnRibbon_v2(
    Number of Ribbons=2
    Spawn Mode=::NewEnumerator0
    SpawnRate:NiagaraFloat=&FloatFromCurve(
     FloatCurve=DataInterface() {propertyValues {Curve {keys [{arriveTangent=1459.5123, interpMode=RCIM_Cubic, leaveTangent=1459.5123, tangentMode=RCTM_User, time=0, value=0}, <at 0.01 25>, <at 1 40>]}}}
     CurveIndex=&MakeCustomFloatFromBool(
      Bool=$User.TurnON_Niagara_Velocity
      True Float=$System.NormalizedVelocityRange
      False Float=&Add_Float(A=$User.Speed_Alpha, B=0.75)
     )
     Scale Curve=2
    )
   )
  ]
  particleSpawnScript [
   &FN_InitializeRibbon_v3(
    Color:LinearColor=$System.Color_Smoothed
    Enable Skeletal Position=true
    Lifetime:NiagaraFloat=&Lerp_Float(A=&Lerp_Float(A=0.15, B=0.2, Alpha=$System.FN_Timer001.Value), B=0.2, Alpha=$System.FN_Timer.Value)
    Ribbon Width=20
    Sampled Position Coordinate Space:ENiagaraCoordinateSpace=::Local
    Sampled Rotation Coordinate Space:ENiagaraCoordinateSpace=::Local
    Sampled Velocity Coordinate Space:ENiagaraCoordinateSpace=::Local
   ) {module=FN_InitializeRibbon_v3001}
   &SkeletalMeshLocation(
    Bone / Socket Index=&SelectIntFromArray(Direct Array Index=&SpawnGroup(), Int Selection Array=DataInterface() {propertyValues {IntData [3,4]}})
    Bone Sampling Mode=::NewEnumerator4
    Interpolate Mesh Sampling=false
    MaskSpawnGroup=false
    Orientation Sampling=::NewEnumerator1
    Sampled Position Offset=<0 -2 -3>
    Sampled Position Offset Coordinate Space:ENiagaraSkelMeshTransforms=::NewEnumerator2
    Skeletal Mesh=$User.SkeletalMesh
   )
  ]
  particleUpdateScript [
   &SolveForcesAndVelocity(Clamp Velocity=false, Limit Acceleration=false)
   &DynamicMaterialParameters(
    Index 0 Param 1=&FloatFromCurve(FloatCurve=DataInterface() {propertyValues {Curve {keys [
         {arriveTangent=15.9564, interpMode=RCIM_Cubic, leaveTangent=15.9564, tangentMode=RCTM_User, time=0, value=0}
         {interpMode=RCIM_Cubic, tangentMode=RCTM_User, time=0.1, value=1}
         {interpMode=RCIM_Cubic, tangentMode=RCTM_User, time=1, value=0}
        ]}}})
    Index 0 Param 2=&FloatFromCurve(FloatCurve=DataInterface() {propertyValues {Curve {keys [
         {arriveTangent=24.5129, interpMode=RCIM_Cubic, leaveTangent=24.5129, tangentMode=RCTM_User, time=0, value=0}
         {arriveTangent=-5.3543, interpMode=RCIM_Cubic, leaveTangent=-5.3543, tangentMode=RCTM_User, time=0.1, value=1}
         {interpMode=RCIM_Cubic, tangentMode=RCTM_User, time=0.3, value=0}
        ]}}})
    Index 0 Param 3=-2
    Index 0 Param 4=1
    Module.Param0WriteEnabled=true
    Module.Param1WriteEnabled=true
    Module.Param2WriteEnabled=true
    Module.Param3WriteEnabled=true
   )
   &FN_UpdateRibbon(
    Enable Skeletal Position=true
    Initial Color:LinearColor=$Particles.Initial.Color
    Initial RibbonWidth:NiagaraFloat=$Particles.Initial.RibbonWidth
   )
   &ScaleColor(
    Scale Alpha=&FloatFromCurve(
     FloatCurve=DataInterface() {propertyValues {Curve {keys [
         {arriveTangent=15.9564, interpMode=RCIM_Cubic, leaveTangent=15.9564, tangentMode=RCTM_User, time=0, value=0}
         {arriveTangent=-3.4931, interpMode=RCIM_Cubic, leaveTangent=-3.4931, tangentMode=RCTM_User, time=0.0583, value=1.0864}
         {interpMode=RCIM_Cubic, tangentMode=RCTM_User, time=1, value=0}
        ]}}}
     Scale Curve=&SelectFloatFromArray(Direct Array Index=$System.Color_Index, Float Selection Array=DataInterface() {propertyValues {FloatData [2.5,1.5,5,2.5]}})
    )
    Scale RGB=&Add_Vector(
     A=&VectorFromFloat(Value=&FNTimeOfDayScalar(Night Scale=0.4, Morning Scale=0.4, Day Scale=0.7, Evening Scale=0.6))
     B=&SelectVectorFromArray(
      Direct Array Index=$System.Color_Index
      Vector Selection Array=DataInterface() {propertyValues {FloatData [<0 0 0>, <-0.1 1 0>, <0 0 0>, <0 0 0>]}}
     )
    )
    ScaleA=true
    ScaleRGB=true
   ) {module=ScaleColor001}
   &ScaleColor(
    Linear Color Curve=DataInterface() {propertyValues {
      AlphaCurve {keys [<at 0 0>,<at 0.0195 1>,<at 0.4459 1>,<at 0.8542 0>]}
      BlueCurve {keys [<at 0 2>,<at 0.0924 1>,<at 0.2674 0.1048>,<at 0.6912 0.0102>]}
      GreenCurve {keys [<at 0 2>,<at 0.0924 1>,<at 0.2674 0.1774>,<at 0.6912 0.0601>]}
      RedCurve {keys [<at 0 2>,<at 0.0924 1>,<at 0.2674 0.25>,<at 0.6912 0.13>]}
     }}
    Scale Mode=::NewEnumerator2
   )
  ]
  rendererClasses [&NiagaraRibbonRendererProperties]
  renderers [{rendererClass=&NiagaraRibbonRendererProperties}]
 }
 GrindRail_Main {
  emitterUpdateScript [
   &EmitterState()
   &FN_SpawnRibbon_v2(
    Spawn Mode=::NewEnumerator0
    SpawnRate:NiagaraFloat=&FloatFromCurve(
     FloatCurve=DataInterface() {propertyValues {Curve {keys [{arriveTangent=1459.5123, interpMode=RCIM_Cubic, leaveTangent=1459.5123, tangentMode=RCTM_User, time=0, value=0}, <at 0.01 25>, <at 1 40>]}}}
     CurveIndex=&MakeCustomFloatFromBool(
      Bool=$User.TurnON_Niagara_Velocity
      True Float=$System.NormalizedVelocityRange
      False Float=&Add_Float(A=$User.Speed_Alpha, B=0.75)
     )
     Scale Curve=6
    )
   )
  ]
  particleSpawnScript [
   &FN_SetSpawnGroup(Number of Spawn Groups=5)
   &InitializeRibbon(
    Color=&MultiplyLinearColor(
     Linear Color=$System.Color_Smoothed
     Scale Factor=&MakeVector4_FromLinearColor(LinearColor=&RandomRangeLinearColor(Minimum=<rgba 44.844 46.133 50 0.075>, Maximum=<rgba 50 41.5857 44.6431 0.075>))
    )
    Lifetime=&SelectFloatFromArray(Direct Array Index=$Particles.SpawnGroup, Float Selection Array=DataInterface() {propertyValues {FloatData [0.15,0.2,0.25,0.15,0.15]}})
    Module.Write Color=true
    Module.Write Lifetime=true
    Module.Write Mass=true
    Module.Write Material Random=true
    Module.Write Position=false
    Module.Write RibbonTwist=true
    Module.Write RibbonWidth=true
    Ribbon Twist:NiagaraFloat=&SelectFloatFromArray(Direct Array Index=$Particles.SpawnGroup, Float Selection Array=DataInterface() {propertyValues {FloatData [0.5,0.25,1,1,-0.75]}})
    Ribbon Width=&SelectFloatFromArray(Direct Array Index=$Particles.SpawnGroup, Float Selection Array=DataInterface() {propertyValues {FloatData [80,30,50,35,50]}})
   )
   /SetVariables(Particles.RibbonDistanceFromStart:NiagaraFloat=$Emitter.DistanceTraveled) {isSetParameters=true}
   &SkeletalMeshLocation(
    Bone / Socket Index=&SelectIntFromArray(Direct Array Index=$Particles.SpawnGroup, Int Selection Array=DataInterface() {propertyValues {IntData [7,8,9,15,16]}})
    Bone Sampling Mode=::NewEnumerator4
    Interpolate Mesh Sampling=false
    MaskSpawnGroup=false
    Orientation Sampling=::NewEnumerator1
    Sampled Position Offset=&SelectVectorFromArray(
     Direct Array Index=$Particles.SpawnGroup
     Vector Selection Array=DataInterface() {propertyValues {FloatData [<5 -5 -25>, <0 0 10>, <15 10 15>, <0 0 0>, <0 0 0>]}}
    )
    Sampled Position Offset Coordinate Space:ENiagaraSkelMeshTransforms=::NewEnumerator2
    Skeletal Mesh=$User.SkeletalMesh
   )
  ]
  particleUpdateScript [
   &ParticleState()
   &SolveForcesAndVelocity(Clamp Velocity=false, Limit Acceleration=false)
   &ScaleColor(
    Scale Alpha=&FloatFromCurve(
     FloatCurve=DataInterface() {propertyValues {Curve {keys [
         {arriveTangent=-0.0132, interpMode=RCIM_Cubic, leaveTangent=-0.0132, tangentMode=RCTM_User, time=0, value=0}
         {arriveTangent=0.811, interpMode=RCIM_Cubic, leaveTangent=0.811, time=0.083, value=0.033}
         {arriveTangent=0.5267, interpMode=RCIM_Cubic, leaveTangent=0.5267, time=0.1693, value=0.1373}
         {interpMode=RCIM_Cubic, tangentMode=RCTM_User, time=0.4, value=0.2}
         {tangentMode=RCTM_User, time=1, value=0}
        ]}}}
     Scale Curve=&SelectFloatFromArray(Direct Array Index=$System.Color_Index, Float Selection Array=DataInterface() {propertyValues {FloatData [3,2,7,10]}})
    )
    Scale RGB=&Add_Vector(
     A=&VectorFromFloat(Value=&FNTimeOfDayScalar(Night Scale=0.4, Morning Scale=0.4, Day Scale=0.7, Evening Scale=0.6))
     B=&SelectVectorFromArray(
      Direct Array Index=$System.Color_Index
      Vector Selection Array=DataInterface() {propertyValues {FloatData [<0 0 0>, <-0.1 1 0>, <0 0 0>, <0 0 0>]}}
     )
    )
    ScaleA=true
    ScaleRGB=true
   ) {module=ScaleColor001}
   &ScaleColor(
    Linear Color Curve=DataInterface() {propertyValues {
      AlphaCurve {keys [<at 0 1>,<at 1 1>]}
      BlueCurve {keys [<at 0 1>,<at 0.2314 0.1126>,<at 0.7187 0.0269>]}
      GreenCurve {keys [<at 0 1>,<at 0.2314 0.2313>,<at 0.7187 0.0657>]}
      RedCurve {keys [<at 0 1>,<at 0.2314 0.35>,<at 0.7187 0.12>]}
     }}
    Scale Mode=::NewEnumerator2
   )
  ]
  rendererClasses [&NiagaraRibbonRendererProperties]
  renderers [{rendererClass=&NiagaraRibbonRendererProperties}]
 }
 GrindRail_Main001 {
  emitterUpdateScript [
   &EmitterState()
   &FN_SpawnAnimTrail(Switch on ENiagara_AnimTrail_Source:ENiagara_AnimTrail_Source=::NewEnumerator1)
  ]
  particleSpawnScript [
   &FN_InitializeAnimTrailBoneIndecies()
   &FN_InitializeAnimTrail(SkeletalMesh=$User.SwordMesh)
  ]
  particleUpdateScript [
   &ParticleState()
   &SolveForcesAndVelocity(Clamp Velocity=false, Limit Acceleration=false)
   &ScaleColor(
    Linear Color Curve=DataInterface() {propertyValues {
      AlphaCurve {keys [<at 0 1>,<at 0.3107 0.6122>,<at 0.4283 0.0499>,<at 1 0>]}
      BlueCurve {keys [<at 0 0>,<at 0.184 0>,<at 0.3061 0>,<at 0.5916 0>,<at 0.9455 0>]}
      GreenCurve {keys [<at 0 1.1667>,<at 0.184 0.7963>,<at 0.3061 0.5413>,<at 0.5916 0.4499>,<at 0.9455 0.2259>]}
      RedCurve {keys [<at 0 2>,<at 0.184 1.2573>,<at 0.3061 1>,<at 0.5916 1>,<at 0.9455 1>]}
     }}
    Scale Mode=::NewEnumerator2
   )
   &FN_UpdateAnimTrail(SkeletalMesh=$User.SwordMesh)
  ]
  rendererClasses [&NiagaraRibbonRendererProperties, &NiagaraSpriteRendererProperties]
  renderers [
   {rendererClass=&NiagaraRibbonRendererProperties}
   {rendererClass=&NiagaraSpriteRendererProperties, rendererIndex=1}
  ]
 }
 Minimal {
  emitterUpdateScript [
   &EmitterState(
    Inactive Response=::NewEnumerator1
    Life Cycle Mode=::NewEnumerator1
    Loop Behavior=::NewEnumerator1
    Loop Delay=0.1
    Loop Duration Mode=::NewEnumerator1
    UseLoopDelay=true
   )
   &SpawnBurst_Instantaneous(Spawn Count=2, Use Spawn Probability=false, UseLoopCountLimit=false)
  ]
  particleSpawnScript [
   /Niagara/Modules/Spawn/Initialization/V2/InitializeParticle(
    Color Mode=::NewEnumerator0
    Lifetime=2.5
    Mesh Scale Mode=::NewEnumerator3
    Mesh Uniform Scale=1.5
    UsePositionOffset=false
    Write Material Random=true
   )
   &PartitionParticles()
  ]
  particleUpdateScript [
   &ParticleState(Kill Particles When Lifetime Has Elapsed=false, Let Infinitely Lived Particles Die When Emitter Deactivates=true)
   &SkeletalMeshLocation(
    Bone / Socket Index=$Particles.PartitionParticles.Partition
    Bone Sampling Mode=::NewEnumerator4
    MaskSpawnGroup=false
    Skeletal Mesh=DataInterface() {propertyValues {
      DefaultMesh=null
      ExcludeBoneName=root
      OptionalComponentClass=null
      PreviewMesh=&M_Med_Soldier_04_Preview
      SkinningMode=SkinOnTheFly
      SoftSourceActor=null
      SourceMode=Default
      UvSetIndex=0
      WholeMeshLOD=-1
      bExcludeBone=false
      bReadDeformedGeometry=true
      bRequireCurrentFrameData=true
      ComponentTags []
      FilteredBones [hand_l,hand_r]
      FilteredSockets []
      MeshUserParameter {parameter {name=None, typeDefHandle {registeredTypeIndex=1}, varData []}}
      SamplingRegions []
     }}
   )
   &ScaleColor(
    Scale Alpha=&FloatFromCurve(
     FloatCurve=DataInterface() {propertyValues {Curve {keys [
         {interpMode=RCIM_Cubic, tangentMode=RCTM_Break, time=0, value=0}
         {interpMode=RCIM_Cubic, tangentMode=RCTM_Break, time=0.2028, value=0.738}
         {interpMode=RCIM_Cubic, tangentMode=RCTM_Break, time=0.431, value=0.3209}
        ]}}}
     CurveIndex=$Engine.System.Age
     Scale Curve=0.5
    )
    ScaleA=true
    ScaleRGB=true
   )
  ]
  rendererClasses [&NiagaraSpriteRendererProperties, &NiagaraMeshRendererProperties]
  renderers [
   {rendererClass=&NiagaraSpriteRendererProperties}
   {rendererClass=&NiagaraMeshRendererProperties, rendererIndex=1}
  ]
 }
 Minimal001 {
  emitterUpdateScript [
   &EmitterState()
   &SpawnRate(SpawnRate=15, Use Spawn Probability=false)
  ]
  particleSpawnScript [
   &InitializeParticle(
    Color=&RandomRangeLinearColor(Minimum=<rgba 1 0.55 0 1>, Maximum=<rgba 1 0.3427 0 1>)
    Lifetime=0.35
    Sprite Rotation Mode=::NewEnumerator2
    Sprite Size Mode=::NewEnumerator1
    Uniform Sprite Size Max=4.75
    Uniform Sprite Size Min=3
    UsePositionOffset=false
    Write Material Random=true
   )
   &SkeletalMeshLocation(
    Bone / Socket Index=&SpawnGroup()
    Bone Sampling Mode=::NewEnumerator4
    MaskSpawnGroup=false
    Sampled Position Offset Coordinate Space:ENiagaraSkelMeshTransforms=::NewEnumerator4
    Skeletal Mesh=DataInterface() {propertyValues {
      DefaultMesh=null
      ExcludeBoneName=root
      OptionalComponentClass=null
      PreviewMesh=/Game/Characters/Player/Male/Male_Avg_Base/Fortnite_M_Avg_Player
      SkinningMode=SkinOnTheFly
      SoftSourceActor=null
      SourceMode=Default
      UvSetIndex=0
      WholeMeshLOD=-1
      bExcludeBone=false
      bReadDeformedGeometry=true
      bRequireCurrentFrameData=true
      ComponentTags []
      FilteredBones [hand_r]
      FilteredSockets []
      MeshUserParameter {parameter {name=None, typeDefHandle {registeredTypeIndex=1}, varData []}}
      SamplingRegions []
     }}
   )
   &AddVelocity(Velocity=<0 -278.6545 0>)
   &ShapeLocation(MaskSpawnGroup=false, Override Seed=false, Sphere Radius=15, UseOwnerScale=true)
  ]
  particleUpdateScript [
   &ParticleState()
   &ScaleSpriteSize(
    Uniform Curve Scale=1.209
    Uniform Curve Sprite Scale=DataInterface() {propertyValues {Curve {
       defaultValue=0
       keys [
        {arriveTangent=9.0725, interpMode=RCIM_Cubic, leaveTangent=9.0725, tangentMode=RCTM_User, time=0, value=0}
        <at 0.1408 0.9847 RCIM_Cubic>
        {arriveTangent=-0.0405, interpMode=RCIM_Cubic, leaveTangent=-0.0405, tangentMode=RCTM_User, time=1, value=0}
       ]
      }}}
   )
   &SolveForcesAndVelocity(Clamp Velocity=false, Limit Acceleration=false)
   &DynamicMaterialParameters(
    Index 0 Param 1=0.8
    Module.Index0Float0Write=true
    Module.Index0Float1Write=true
    Module.Index0Float2Write=true
    Module.Index0Float3Write=true
   )
   &ScaleColor(Scale Alpha=&RandomRangeFloat(Minimum=1, Maximum=1), ScaleA=true, ScaleRGB=true)
   &FN_Sprite_Attributes(
    Color=&MultiplyLinearColor(
     Linear Color=<rgba 1 0.55 0.1 0.4>
     Scale Factor=&MakeVector4(
      X=1
      Y=1
      Z=1
      W=&FloatFromCurve(FloatCurve=DataInterface() {propertyValues {Curve {keys [{interpMode=RCIM_Cubic, tangentMode=RCTM_Break, time=0, value=1}, <at 1 0>]}}})
     )
    )
    Create Color Attribute=true
    Create SpriteSize Attribute=true
    SpriteSize=&Multiply_Vector2D_ByFloat(Float=$User.SlashSize, Vector2D=<12.5 12.5>)
    Write to Module Namespace=true
   )
  ]
  rendererClasses [&NiagaraSpriteRendererProperties]
  renderers [
   {rendererClass=&NiagaraSpriteRendererProperties}
   {rendererClass=&NiagaraSpriteRendererProperties, rendererIndex=1}
  ]
 }
 Ring {
  emitterSpawnScript [/SetVariables(Emitter.GlowDepthFade=25) {isSetParameters=true}]
  emitterUpdateScript [
   &EmitterState(
    Enable Distance Culling=true
    Life Cycle Mode=::NewEnumerator1
    Loop Behavior=::NewEnumerator1
    Loop Delay=$User.Delay
    Loop Duration=2
    Max Distance Response=::KillImmediately
    MaxDistance=3500
    Module.bUseMaxDistance=true
    Module.bUseMinDistance=false
    Scalability Mode=::NewEnumerator1
    UseLoopDelay=true
   )
   &SpawnBurst_Instantaneous(Spawn Count=3, Use Spawn Probability=false)
  ]
  particleSpawnScript [
   &FN_SetSpawnGroup(Number of Spawn Groups=2)
   &PartitionParticles(Number Of Partitions=3)
   &InitializeParticle(
    Color=$User.Color_Default
    Lifetime=&SelectFloatFromArray(
     Direct Array Index=$Particles.PartitionParticles.Partition
     Float Selection Array=DataInterface() {propertyValues {FloatData [0.4,0.4,0.25]}}
    )
    Position Offset Coordinate Space=::Local
    Sprite Rotation Mode=::NewEnumerator2
    Sprite Size=&SelectVector2DFromArray(
     Direct Array Index=$Particles.PartitionParticles.Partition
     Vector 2D Selection Array=DataInterface() {propertyValues {FloatData [<300 300>, <300 300>, <100 150>]}}
    )
    Sprite Size Mode=::NewEnumerator4
    UsePositionOffset=true
    Write Material Random=true
   )
   &CameraOffset(Camera Offset Amount=-20)
   /SetVariables(Particles.CameraOffsetWrist=15) {isSetParameters=true}
   /SetVariables(
    Particles.VisibilityTag=&SelectIntFromArray(Direct Array Index=$Particles.PartitionParticles.Partition, Int Selection Array=DataInterface() {propertyValues {IntData [0,0,1]}})
    Particles.VisibilityWrist:NiagaraInt32=&SelectIntFromArray(Direct Array Index=$Particles.PartitionParticles.Partition, Int Selection Array=DataInterface() {propertyValues {IntData [1,1,0]}})
   ) {isSetParameters=true}
  ]
  particleUpdateScript [
   &ParticleState()
   &SkeletalMeshLocation(
    Bone / Socket Index=&SelectIntFromArray(Direct Array Index=$Particles.PartitionParticles.Partition, Int Selection Array=DataInterface() {propertyValues {IntData [8,9,2]}})
    Bone Sampling Mode=::NewEnumerator4
    MaskSpawnGroup=false
    Sampled Position Offset=<0 7.5 -7.5>
    Sampled Position Offset Coordinate Space:ENiagaraSkelMeshTransforms=::NewEnumerator2
    Skeletal Mesh=$User.SkeletalMesh
   )
   &ScaleColor(
    Linear Color Curve=DataInterface() {propertyValues {
      AlphaCurve {keys [<at 0 1>,<at 1 0>]}
      BlueCurve {keys [<at 0 1 RCIM_Cubic>]}
      GreenCurve {keys [<at 0 1 RCIM_Cubic>]}
      RedCurve {keys [<at 0 1 RCIM_Cubic>]}
     }}
    Scale Mode=::NewEnumerator2
   )
   &ScaleColor(
    Scale RGB=&Multiply_Vector(
     A=&VectorFromFloat(Value=&FNTimeOfDayScalar(Night Scale=0.35, Morning Scale=0.3, Day Scale=0.4, Evening Scale=0.4))
     B=<2 2 2>
    )
    ScaleA=true
    ScaleRGB=true
   ) {module=ScaleColor001}
   &ScaleSpriteSize(Scale Factor=&Vector2DFromFloat(Value=&FloatFromCurve(
      FloatCurve=DataInterface() {propertyValues {Curve {keys [{arriveTangent=2.7513, interpMode=RCIM_Cubic, leaveTangent=2.7513, tangentMode=RCTM_User, time=0, value=0}, <at 1 1.2>]}}}
      Scale Curve=0.5
     )))
   /SetVariables(Particles.WristSize:Vector2f=$Particles.Initial.SpriteSize) {isSetParameters=true}
   /SetVariables(
    Particles.FlareColor=&MultiplyLinearColor(Linear Color=$User.Color_Default, Scale Factor=&MakeVector4_FromFloat(Float=0.8))
    Particles.FlareSize:Vector2f=&Multiply_Vector2D_ByFloat(
     Float=&FloatFromCurve(FloatCurve=DataInterface() {propertyValues {Curve {keys [
          {interpMode=RCIM_Cubic, tangentMode=RCTM_User, time=0, value=0}
          {interpMode=RCIM_Cubic, tangentMode=RCTM_User, time=0.3, value=0.95}
          {tangentMode=RCTM_User, time=1, value=0}
         ]}}})
     Vector2D=$Particles.Initial.SpriteSize
    )
    Particles.FlashCameraOffset=20
   ) {isSetParameters=true}
  ]
  rendererClasses [&NiagaraSpriteRendererProperties]
  renderers [
   {rendererClass=&NiagaraSpriteRendererProperties}
   {rendererClass=&NiagaraSpriteRendererProperties, rendererIndex=1}
  ]
 }
}
summary {
 systemName=NS_DentureSay_Trail_Body
 emitters [
  {emitterName=GrindRail_Main, rendererClasses [&NiagaraRibbonRendererProperties]}
  {emitterName=GrindRail_Hands, rendererClasses [&NiagaraRibbonRendererProperties]}
  {emitterName=Foot_Flare001, rendererClasses [&NiagaraSpriteRendererProperties, &NiagaraLightRendererProperties]}
  {emitterName=Ring, rendererClasses [&NiagaraSpriteRendererProperties]}
  {emitterName=Minimal, rendererClasses [&NiagaraSpriteRendererProperties, &NiagaraMeshRendererProperties]}
  {emitterName=Minimal001, rendererClasses [&NiagaraSpriteRendererProperties]}
  {
   emitterName=GrindRail_Main001
   rendererClasses [&NiagaraRibbonRendererProperties, &NiagaraSpriteRendererProperties]
  }
 ]
 userVariables [
  User.Boosting {type=NiagaraBool}
  User.Color_Boosting {type=LinearColor, description="", defaultValue {struct=LinearColor, value=<rgba 0 1 0 1>}}
  User.Color_Default {type=LinearColor, description="", defaultValue {struct=LinearColor, value=<rgba 1 0.6175 0.15 1>}}
  User.Color_SlowDown {type=LinearColor, description="", defaultValue {struct=LinearColor, value=<rgba 0.2557 0.3437 1 1>}}
  User.Color_Sprinting {type=LinearColor, description="", defaultValue {struct=LinearColor, value=<rgba 0 0.7703 1 1>}}
  User.ColorIndex {type=NiagaraInt32}
  User.Delay {type=NiagaraFloat, defaultValue=0, description=""}
  User.Is180Turning {type=NiagaraFloat, defaultValue=0, description=""}
  User.Jumped {type=NiagaraBool}
  User.Leaning Direction {type=NiagaraFloat, defaultValue=0, description=""}
  User.RibbonIDOffset {type=NiagaraInt32}
  User.SkeletalMesh {
   type=NiagaraDataInterfaceSkeletalMesh
   defaultValue {
    struct=NiagaraExt_VariableValue_DataInterface
    value {dataInterface=&self:NiagaraDataInterfaceSkeletalMesh_0}
   }
  }
  User.SlashSize {type=NiagaraFloat, defaultValue=1, description=""}
  User.SlowDown {type=NiagaraBool}
  User.Speed_Alpha {type=NiagaraFloat, defaultValue=0.6, description=""}
  User.Sprinting {type=NiagaraBool}
  User.SwordMesh {
   type=NiagaraDataInterfaceSkeletalMesh
   defaultValue {
    struct=NiagaraExt_VariableValue_DataInterface
    value {dataInterface=&self:NiagaraDataInterfaceSkeletalMesh_1}
   }
  }
  User.TurnON_Niagara_Velocity {type=NiagaraBool, defaultValue=true}
 ]
}
