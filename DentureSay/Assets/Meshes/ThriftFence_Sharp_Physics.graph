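# Physics asset for the ThriftFence_Sharp mesh, modeling a handle with a jointed chain that ends in a box-shaped piece.
# A large handle capsule is rigidly locked to the small chain base on all angular axes.
# Five identical capsule links hang from the base in series, with a rotated box as the sixth and last body.
# The first link's twist is locked, while later joints share swing limits and a tighter twist limit value.
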PhysicsAsset /DentureSay/Assets/Meshes/ThriftFence_Sharp_Physics
bodies {
 base_chain [{shapeType=Capsule, length=0.5, radius=1, shapeName=base_chain_capsule}]
 dyn_chain_01 [{shapeType=Capsule, length=4, radius=3, shapeName=dyn_chain_01_capsule, center=<3 0 0>, rotation=<rot 90 0 0>}]
 dyn_chain_02 [{shapeType=Capsule, length=4, radius=3, shapeName=dyn_chain_02_capsule, center=<3 0 0>, rotation=<rot 90 0 0>}]
 dyn_chain_03 [{shapeType=Capsule, length=4, radius=3, shapeName=dyn_chain_03_capsule, center=<3 0 0>, rotation=<rot 90 0 0>}]
 dyn_chain_04 [{shapeType=Capsule, length=4, radius=3, shapeName=dyn_chain_04_capsule, center=<3 0 0>, rotation=<rot 90 0 0>}]
 dyn_chain_05 [{shapeType=Capsule, length=4, radius=3, shapeName=dyn_chain_05_capsule, center=<3 0 0>, rotation=<rot 90 0 0>}]
 dyn_chain_06 [{
   shapeType=Box
   extentX=11.0558
   extentY=1.01
   extentZ=11.7485
   shapeName=dyn_chain_06_box
   center=<4.3215 1.7665 -0.1247>
   rotation=<rot 0 97.2312 -90>
  }]
 handle [{shapeType=Capsule, length=9.5, radius=16, shapeName=handle_capsule, center=<0 0 -3.3072>}]
}
constraints [
 {
  bone1Name=base_chain
  bone2Name=handle
  swing1LimitDegrees=45
  swing1Motion=Locked
  swing2LimitDegrees=45
  swing2Motion=Locked
  twistLimitDegrees=45
  twistMotion=Locked
 }
 {bone1Name=dyn_chain_01, bone2Name=base_chain, swing1LimitDegrees=45, swing2LimitDegrees=45, twistLimitDegrees=45, twistMotion=Locked}
 {bone1Name=dyn_chain_02, bone2Name=dyn_chain_01, swing1LimitDegrees=45, swing2LimitDegrees=45, twistLimitDegrees=5}
 {bone1Name=dyn_chain_03, bone2Name=dyn_chain_02, swing1LimitDegrees=45, swing2LimitDegrees=45, twistLimitDegrees=5}
 {bone1Name=dyn_chain_04, bone2Name=dyn_chain_03, swing1LimitDegrees=45, swing2LimitDegrees=45, twistLimitDegrees=5}
 {bone1Name=dyn_chain_05, bone2Name=dyn_chain_04, swing1LimitDegrees=45, swing2LimitDegrees=45, twistLimitDegrees=5}
 {bone1Name=dyn_chain_06, bone2Name=dyn_chain_05, swing1LimitDegrees=45, swing2LimitDegrees=45, twistLimitDegrees=5}
]
