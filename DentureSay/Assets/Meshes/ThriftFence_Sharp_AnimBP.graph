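# Animation Blueprint for the ThriftFence_Sharp skeleton, adding simulated motion to its dynamic bone chain.
# When enabled, a world-space RigidBody node simulates from the ref pose using the ThriftFence_Sharp_Physics asset.
# Otherwise, a six-bone Trail controller on dyn_chain_06 runs after dyn_chain_01 is rotated to a fixed world-space orientation.
# A bool blend picks the path from the rigid body flag in the simulation detail settings, read once at initialization.
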AnimBlueprint /DentureSay/Assets/Meshes/ThriftFence_Sharp_AnimBP
target_skeleton=&/ThriftFence_Sharp_Skeleton
graphs [
 AnimGraph {nodes [
   n1 = Local Space Ref Pose()
   n2 = Trail controller - Bone: dyn_chain_06(
    ComponentPose="Transform (Modify) Bone - Bone: dyn_chain_01"(ComponentPose=Local To Component(LocalPose=@n1), Alpha=1, Translation="0,0,0", Rotation="-90,0,0", Scale="1,1,1") {rotationMode=BMM_Replace, rotationSpace=BCS_WorldSpace, boneToModify {boneName=dyn_chain_01}, rotation {pitch=-90}}
    Alpha=1
   ) {
    chainLength=6
    rotationLimits [
     {limitMax=<180 180 180>}
     {limitMax=<180 180 180>}
     {limitMax=<180 180 180>}
     {limitMax=<180 180 180>}
     {limitMax=<180 180 180>}
     {limitMax=<180 180 180>}
    ]
    rotationOffsets [<0 0 0>, <0 0 0>, <0 0 0>, <0 0 0>, <0 0 0>, <0 0 0>]
    trailBone {boneName=dyn_chain_06}
    alphaBoolBlend=~
   }
   n3 = RigidBody!(ComponentPose=Local To Component(LocalPose=Local Space Ref Pose()), Alpha, ExternalForce, bTransferBoneVelocities:bool=False) -> (Pose) {
    bClampLinearTranslationLimitToRefPose=true
    overridePhysicsAsset=&/ThriftFence_Sharp_Physics
    simulationSpace=WorldSpace
    worldSpaceMinimumScale=0.75
    baseBoneRef {boneName=base_chain}
   }
   Output Pose(Result=Blend Poses by bool(
     BlendPose_0=Component To Local(ComponentPose=@n3)
     BlendPose_1=Component To Local(ComponentPose=@n2)
     bActiveValue=Get Rigid Body Enabled() -> (Rigid Body Enabled:bool)
    ))
  ]}
 EventGraph {nodes [
   Event BlueprintInitializeAnimation()
   > n1 = GetSimulationDetailSettings!(execute, self:Fort Base Anim Instance Object Reference) -> (then, bAnimDynamicsEnabled:bool, bRigidBodyEnabled:bool, bClothEnabled:bool)
   > Set Rigid Body Enabled(Rigid Body Enabled:bool=@n1.bRigidBodyEnabled) -> (Output_Get:bool)
   TryGetPawnOwner()
   Event BlueprintUpdateAnimation()
  ]}
]
variables [Rigid Body Enabled {type=boolean}]
